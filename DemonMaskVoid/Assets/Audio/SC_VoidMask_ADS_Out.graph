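# Sound cue for the Void Mask's ADS exit foley in DemonMaskVoid, named for leaving aim-down-sights.
# A single wave player plays the VoidMask_ADS_End wave directly at half volume.
# It uses the Void Mask foley attenuation, routes to the player attacks and abilities sound class, and limits overlap through the Void Mask concurrency set alongside a low max-count override.
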SoundCue /DemonMaskVoid/Assets/Audio/SC_VoidMask_ADS_Out
root=WavePlayer_2
nodes [{id=WavePlayer_2, properties {soundWaveAssetPtr=/DemonMaskVoid/Assets/Audio/wav/VoidMask_ADS_End_01}}]
properties {
 attenuationSettings=&/SA_VoidMask_Foley
 maxDistance=2600
 soundClassObject=/Game/Sounds/FortSoundClassesAndModes/Fort_06_Player_Attacks_Abilities
 volumeMultiplier=0.5
 concurrencyOverrides {maxCount=2}
 concurrencySet [&/CS_VoidMask]
}
