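# Sound cue for the end of the Void Mask's cook loop in the DemonMaskVoid content.
# A single wave player feeds the loop-end sample straight to the output, with no randomization or modulation.
# It plays at reduced volume through Void Mask foley attenuation and concurrency settings, but is routed to the critical in-game UI sound class.
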SoundCue /DemonMaskVoid/Assets/Audio/SC_VoidMask_Cook_End
root=WavePlayer_0
nodes [{id=WavePlayer_0, properties {soundWaveAssetPtr=/DemonMaskVoid/Assets/Audio/wav/VoidMask_Cook_Loop_End_01}}]
properties {
 attenuationSettings=&/SA_VoidMask_Foley
 maxDistance=2600
 soundClassObject=&Fort_UI_InGame_Critical
 volumeMultiplier=0.2
 concurrencyOverrides {maxCount=2}
 concurrencySet [&/CS_VoidMask]
}
