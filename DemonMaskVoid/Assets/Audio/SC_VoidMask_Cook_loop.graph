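# Looping Sound Cue for the Demon Mask Void "Cook" sound, routed to the player attacks and abilities sound class.
# A mixer layers a looping wave, with volume and pitch shaped by Enveloper curves, over a half-volume stream of one-shots.
# The one-shot layer loops a short random delay into a pick of five variants, and each layer has its own Modulator.
# VoidMask foley attenuation and concurrency assets govern playback.
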SoundCue /DemonMaskVoid/Assets/Audio/SC_VoidMask_Cook_loop
root=Mixer_0
nodes [
 {id=Delay_0, children [Random_0], properties {delayMax=0.4, delayMin=0.1}}
 {
  id=Enveloper_0
  children [WavePlayer_0]
  properties {
   pitchCurve {editorCurveData {keys [
      <at 0 0 RCIM_Cubic>
      {arriveTangent=1.4302, interpMode=RCIM_Cubic, leaveTangent=1.4302, time=0.22, value=1.1865}
      {interpMode=RCIM_Cubic, time=0.6992}
     ]}}
   volumeCurve {editorCurveData {keys [<at 0 0 RCIM_Cubic>, {arriveTangent=1.0926, interpMode=RCIM_Cubic, leaveTangent=1.0926, time=0.35, value=1.2}, <at 0.7449 0.8138 RCIM_Cubic>]}}
  }
 }
 {id=Looping_0, children [Delay_0]}
 {id=Mixer_0, children [Modulator_0,Modulator_1], properties {inputVolume [1,0.5]}}
 {id=Modulator_0, children [Enveloper_0]}
 {id=Modulator_1, children [Looping_0]}
 {id=Random_0, children [WavePlayer_1,WavePlayer_2,WavePlayer_3,WavePlayer_4,WavePlayer_5], properties {weights [1,1,1,1,1]}}
 {id=WavePlayer_0, properties {bLooping=true, soundWaveAssetPtr=/DemonMaskVoid/Assets/Audio/wav/VoidMask_Cook_Loop_01}}
 {id=WavePlayer_1, properties {soundWaveAssetPtr=/DemonMaskVoid/Assets/Audio/wav/VoidMask_Cook_OS4Looping_01}}
 {id=WavePlayer_2, properties {soundWaveAssetPtr=/DemonMaskVoid/Assets/Audio/wav/VoidMask_Cook_OS4Looping_02}}
 {id=WavePlayer_3, properties {soundWaveAssetPtr=/DemonMaskVoid/Assets/Audio/wav/VoidMask_Cook_OS4Looping_03}}
 {id=WavePlayer_4, properties {soundWaveAssetPtr=/DemonMaskVoid/Assets/Audio/wav/VoidMask_Cook_OS4Looping_04}}
 {id=WavePlayer_5, properties {soundWaveAssetPtr=/DemonMaskVoid/Assets/Audio/wav/VoidMask_Cook_OS4Looping_05}}
]
properties {
 attenuationSettings=&/SA_VoidMask_Foley
 maxDistance=2600
 soundClassObject=/Game/Sounds/FortSoundClassesAndModes/Fort_06_Player_Attacks_Abilities
 volumeMultiplier=1.3
 concurrencyOverrides {maxCount=2}
 concurrencySet [&/CS_VoidMask]
}
