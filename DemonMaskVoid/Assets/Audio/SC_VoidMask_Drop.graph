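# Sound cue for the drop foley of the Void Mask item, routed through the weapon foley sound class.
# A Random node picks between two drop recordings with equal weight and feeds a Modulator at the root.
# The cue plays at reduced volume and slightly lowered pitch, using Void Mask foley attenuation and a Void Mask concurrency set.
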SoundCue /DemonMaskVoid/Assets/Audio/SC_VoidMask_Drop
root=Modulator_0
nodes [
 {id=Modulator_0, children [Random_1]}
 {id=Random_1, children [WavePlayer_0,WavePlayer_4], properties {weights [1,1]}}
 {id=WavePlayer_0, properties {soundWaveAssetPtr=/DemonMaskVoid/Assets/Audio/wav/VoidMask_Foley_Drop_01}}
 {id=WavePlayer_4, properties {soundWaveAssetPtr=/DemonMaskVoid/Assets/Audio/wav/VoidMask_Foley_Drop_02}}
]
properties {
 attenuationSettings=&/SA_VoidMask_Foley
 maxDistance=2600
 pitchMultiplier=0.95
 soundClassObject=&Fort_WeaponFoley
 volumeMultiplier=0.6
 concurrencySet [&/CS_VoidMask]
}
