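# Sound cue for the foley played when a dropped Void Mask lands.
# A Random node picks between two equally weighted drop-land wave variations and feeds a Modulator left at default settings.
# The cue routes to the Fort_WeaponFoley sound class and uses the Void Mask foley attenuation and Void Mask concurrency assets.
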SoundCue /DemonMaskVoid/Assets/Audio/SC_VoidMask_Drop_Landed
root=Modulator_0
nodes [
 {id=Modulator_0, children [Random_0]}
 {id=Random_0, children [WavePlayer_0,WavePlayer_1], properties {weights [1,1]}}
 {id=WavePlayer_0, properties {soundWaveAssetPtr=/DemonMaskVoid/Assets/Audio/wav/VoidMask_Foley_DropLand_01}}
 {id=WavePlayer_1, properties {soundWaveAssetPtr=/DemonMaskVoid/Assets/Audio/wav/VoidMask_Foley_DropLand_02}}
]
properties {
 attenuationSettings=&/SA_VoidMask_Foley
 maxDistance=2600
 soundClassObject=&Fort_WeaponFoley
 volumeMultiplier=1
 concurrencySet [&/CS_VoidMask]
}
