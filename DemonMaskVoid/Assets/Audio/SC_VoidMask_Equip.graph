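# Sound cue for equipping the Void Mask, playing an equip foley hit that leads into a looping idle foley bed.
# A mixer layers a randomly chosen one of two equip waves, with slight random volume boost, against the idle loop at slightly lower volume.
# The loop starts after a fixed delay of just over a second.
# Playback uses mask-specific attenuation and concurrency under the weapon foley sound class.
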SoundCue /DemonMaskVoid/Assets/Audio/SC_VoidMask_Equip
root=Mixer_0
nodes [
 {id=Delay_0, children [WavePlayer_1], properties {delayMax=1.21, delayMin=1.21}}
 {id=Mixer_0, children [Modulator_0,Delay_0], properties {inputVolume [1,0.8]}}
 {id=Modulator_0, children [Random_1], properties {pitchMax=1, pitchMin=1, volumeMax=1.15, volumeMin=1.05}}
 {id=Random_1, children [WavePlayer_0,WavePlayer_6], properties {weights [1,1]}}
 {id=WavePlayer_0, properties {soundWaveAssetPtr=/DemonMaskVoid/Assets/Audio/wav/VoidMask_Foley_Equip_01}}
 {id=WavePlayer_1, properties {bLooping=true, soundWaveAssetPtr=/DemonMaskVoid/Assets/Audio/wav/VoidMask_Foley_Equip_Idle_Loop_01}}
 {id=WavePlayer_6, properties {soundWaveAssetPtr=/DemonMaskVoid/Assets/Audio/wav/VoidMask_Foley_Equip_02}}
]
properties {
 attenuationSettings=&/SA_VoidMask_Foley
 maxDistance=2600
 soundClassObject=&Fort_WeaponFoley
 volumeMultiplier=0.25
 concurrencyOverrides {maxCount=2}
 concurrencySet [&/CS_VoidMask]
}
