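# Sound cue for the looping in-air sound of the Void Mask, classed as a ranged weapon projectile effect.
# A looping wave player feeds an Enveloper that ramps volume up at the start, then a Modulator and a smoothed, intensified Doppler node.
# Overridden attenuation adds distance filtering and occlusion, with a distance-scaled send to the distant reverb submix.
# A Void Mask concurrency set governs simultaneous instances.
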SoundCue /DemonMaskVoid/Assets/Audio/SC_VoidMask_InAir_Loop
root=Doppler_0
nodes [
 {id=Doppler_0, children [Modulator_0], properties {bUseSmoothing=true, dopplerIntensity=2}}
 {
  id=Enveloper_0
  children [WavePlayer_1]
  properties {
   pitchCurve {editorCurveData {keys [{time=0}]}}
   volumeCurve {editorCurveData {keys [<at 0 0 RCIM_Cubic>, {arriveTangent=5.7599, interpMode=RCIM_Cubic, leaveTangent=5.7599, tangentMode=RCTM_User, time=0.4, value=1}]}}
  }
 }
 {id=Modulator_0, children [Enveloper_0]}
 {id=WavePlayer_1, properties {bLooping=true, soundWaveAssetPtr=/DemonMaskVoid/Assets/Audio/wav/VoidMask_InAir_Loop}}
]
properties {
 attenuationSettings=/Game/Sounds/Attenuation/Fort_Weapon_Projectile_InAir_Loop_Att
 bOverrideAttenuation=true
 maxDistance=3500
 soundClassObject=&Fort_WeaponRanged_Projectile_Effects
 volumeMultiplier=3
 attenuationOverrides {
  bAttenuateWithLPF=true
  bEnableOcclusion=true
  falloffDistance=3000
  hPFFrequencyAtMax=150
  lPFFrequencyAtMax=900
  lPFRadiusMax=3000
  lPFRadiusMin=500
  occlusionLowPassFilterFrequency=1000
  occlusionTraceChannel=ECC_GameTraceChannel6
  stereoSpread=0
  attenuationShapeExtents {x=500}
 }
 concurrencySet [&/CS_VoidMask]
 soundSubmixSends [{
   maxSendDistance=3000
   maxSendLevel=0.25
   minSendDistance=500
   minSendLevel=0.1
   sendLevel=0.1
   sendLevelControlMethod=Linear
   soundSubmix=&Fortnite_Reverb_Distant
  }]
}
