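# Sound cue for the Void Mask portal appearing, set up as a non-attenuated local 2D sound in the player abilities sound class.
# A single wave player feeds the distant portal landing wave straight to the output at slightly raised pitch.
# Inline attenuation overrides disable volume falloff in place of the referenced teleport attenuation asset.
# It joins the shared Void Mask concurrency group and sends lightly to the distant reverb submix.
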SoundCue /DemonMaskVoid/Assets/Audio/SC_VoidMask_Local2D_PortalAppear
root=WavePlayer_1
nodes [{id=WavePlayer_1, properties {soundWaveAssetPtr=/DemonMaskVoid/Assets/Audio/wav/VoidMask_Portal_Land_Dist}}]
properties {
 attenuationSettings=&/SA_VoidMask_Teleport
 bOverrideAttenuation=true
 maxDistance=2097152
 pitchMultiplier=1.05
 soundClassObject=/Game/Sounds/FortSoundClassesAndModes/Fort_06_Player_Attacks_Abilities
 volumeMultiplier=1
 attenuationOverrides {bAttenuate=false}
 concurrencyOverrides {maxCount=2}
 concurrencySet [&/CS_VoidMask]
 soundSubmixSends [{maxSendDistance=5000, maxSendLevel=0.2, minSendLevel=0.015, sendLevel=0.015, soundSubmix=&Fortnite_Reverb_Distant}]
}
