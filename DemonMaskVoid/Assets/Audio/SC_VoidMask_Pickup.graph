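# Sound cue that plays pickup foley for the Demon Mask's Void Mask.
# A random node picks one of two equally weighted pickup waves and passes it through a modulator left at default settings.
# Playback uses the mask's foley attenuation asset and the weapon foley sound class.
# A mask concurrency set is assigned, along with a concurrency override that caps simultaneous instances at two.
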SoundCue /DemonMaskVoid/Assets/Audio/SC_VoidMask_Pickup
root=Modulator_0
nodes [
 {id=Modulator_0, children [Random_0]}
 {id=Random_0, children [WavePlayer_1,WavePlayer_2], properties {weights [1,1]}}
 {id=WavePlayer_1, properties {soundWaveAssetPtr=/DemonMaskVoid/Assets/Audio/wav/VoidMask_Foley_Pickup_01}}
 {id=WavePlayer_2, properties {soundWaveAssetPtr=/DemonMaskVoid/Assets/Audio/wav/VoidMask_Foley_Pickup_02}}
]
properties {
 attenuationSettings=&/SA_VoidMask_Foley
 maxDistance=2600
 soundClassObject=&Fort_WeaponFoley
 concurrencyOverrides {maxCount=2}
 concurrencySet [&/CS_VoidMask]
}
