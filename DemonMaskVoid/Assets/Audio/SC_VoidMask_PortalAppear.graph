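# Sound cue for the Void Mask portal appearing, routed to the player attacks and abilities sound class.
# A distance crossfade blends a close portal landing wave into a distant variant, which takes over in the mid range and fades out at long range.
# Teleport attenuation and Void Mask concurrency assets govern playback, with a slight pitch lift and a send to the distant reverb submix.
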SoundCue /DemonMaskVoid/Assets/Audio/SC_VoidMask_PortalAppear
root=DistanceCrossFade_0
nodes [
 {
  id=DistanceCrossFade_0
  children [WavePlayer_0,WavePlayer_1]
  properties {crossFadeInput [
    {fadeInDistanceEnd=0, fadeInDistanceStart=0, fadeOutDistanceEnd=6000, fadeOutDistanceStart=2500, volume=1}
    {fadeInDistanceEnd=6000, fadeInDistanceStart=5000, fadeOutDistanceEnd=23000, fadeOutDistanceStart=15000, volume=1}
   ]}
 }
 {id=WavePlayer_0, properties {soundWaveAssetPtr=/DemonMaskVoid/Assets/Audio/wav/VoidMask_Portal_Land_Close}}
 {id=WavePlayer_1, properties {soundWaveAssetPtr=/DemonMaskVoid/Assets/Audio/wav/VoidMask_Portal_Land_Dist}}
]
properties {
 attenuationSettings=&/SA_VoidMask_Teleport
 maxDistance=25000
 pitchMultiplier=1.05
 soundClassObject=/Game/Sounds/FortSoundClassesAndModes/Fort_06_Player_Attacks_Abilities
 volumeMultiplier=1
 concurrencyOverrides {maxCount=2}
 concurrencySet [&/CS_VoidMask]
 soundSubmixSends [{maxSendDistance=5000, maxSendLevel=0.2, minSendLevel=0.015, sendLevel=0.015, soundSubmix=&Fortnite_Reverb_Distant}]
}
