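# Looping ambience Sound Cue for the Void Mask portal, in the player attacks and abilities sound class.
# An Enveloper wraps the looping portal wave, ramping its volume from silence to full over the first second.
# The cue plays quieter and slightly pitched up, with spatialization taken from the SA_VoidMask_Portal attenuation asset.
# A concurrency override limits simultaneous instances to two.
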SoundCue /DemonMaskVoid/Assets/Audio/SC_VoidMask_Portal_Amb_Loop
root=Enveloper_0
nodes [
 {
  id=Enveloper_0
  children [WavePlayer_6]
  properties {pitchCurve {editorCurveData {keys [{time=0}]}}, volumeCurve {editorCurveData {keys [<at 0 0>,<at 0.6453 0.3647>,<at 1 1>]}}}
 }
 {id=WavePlayer_6, properties {bLooping=true, soundWaveAssetPtr=/DemonMaskVoid/Assets/Audio/wav/VoidMask_PortalAmb_loop}}
]
properties {
 attenuationSettings=&/SA_VoidMask_Portal
 bOverrideConcurrency=true
 maxDistance=4500
 pitchMultiplier=1.05
 soundClassObject=/Game/Sounds/FortSoundClassesAndModes/Fort_06_Player_Attacks_Abilities
 volumeMultiplier=0.6
 attenuationOverrides {
  bAttenuateWithLPF=true
  bEnableListenerFocus=true
  bEnableOcclusion=true
  falloffDistance=4500
  hPFFrequencyAtMax=100
  lPFFrequencyAtMax=1000
  lPFRadiusMax=5000
  lPFRadiusMin=0
  nonFocusVolumeAttenuation=0.6
  occlusionInterpolationTime=0.5
  occlusionLowPassFilterFrequency=1000
  occlusionTraceChannel=ECC_GameTraceChannel6
  occlusionVolumeAttenuation=0.8
  reverbDistanceMax=5000
  reverbDistanceMin=0
  reverbWetLevelMin=0.2
  attenuationShapeExtents {x=500}
 }
 concurrencyOverrides {maxCount=2}
}
