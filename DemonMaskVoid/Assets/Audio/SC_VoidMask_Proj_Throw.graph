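# Sound cue for the Void Mask thrown-projectile sound in DemonMaskVoid, classed as a player attack and ability sound.
# A random node picks one of five equally weighted ThrownProj wave variations, played at reduced volume and slightly raised pitch.
# It borrows the Void Mask teleport attenuation asset and uses the shared Void Mask concurrency set, with an override for maximum count.
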SoundCue /DemonMaskVoid/Assets/Audio/SC_VoidMask_Proj_Throw
root=Random_1
nodes [
 {id=Random_1, children [WavePlayer_0,WavePlayer_1,WavePlayer_2,WavePlayer_3,WavePlayer_7], properties {weights [1,1,1,1,1]}}
 {id=WavePlayer_0, properties {soundWaveAssetPtr=/DemonMaskVoid/Assets/Audio/wav/VoidMask_ThrownProj_01}}
 {id=WavePlayer_1, properties {soundWaveAssetPtr=/DemonMaskVoid/Assets/Audio/wav/VoidMask_ThrownProj_02}}
 {id=WavePlayer_2, properties {soundWaveAssetPtr=/DemonMaskVoid/Assets/Audio/wav/VoidMask_ThrownProj_03}}
 {id=WavePlayer_3, properties {soundWaveAssetPtr=/DemonMaskVoid/Assets/Audio/wav/VoidMask_ThrownProj_04}}
 {id=WavePlayer_7, properties {soundWaveAssetPtr=/DemonMaskVoid/Assets/Audio/wav/VoidMask_ThrownProj_05}}
]
properties {
 attenuationSettings=&/SA_VoidMask_Teleport
 maxDistance=25000
 pitchMultiplier=1.1
 soundClassObject=/Game/Sounds/FortSoundClassesAndModes/Fort_06_Player_Attacks_Abilities
 volumeMultiplier=0.4
 concurrencyOverrides {maxCount=2}
 concurrencySet [&/CS_VoidMask]
}
