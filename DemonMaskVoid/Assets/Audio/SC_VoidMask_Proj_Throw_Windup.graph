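# Sound cue for the windup of the Void Mask's thrown projectile.
# A Random node picks one of five equally weighted windup wave variations, played quietly and pitched up.
# It uses the Void Mask foley attenuation and concurrency set and routes to the player attacks and abilities sound class.
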SoundCue /DemonMaskVoid/Assets/Audio/SC_VoidMask_Proj_Throw_Windup
root=Random_0
nodes [
 {id=Random_0, children [WavePlayer_4,WavePlayer_5,WavePlayer_6,WavePlayer_8,WavePlayer_9], properties {weights [1,1,1,1,1]}}
 {id=WavePlayer_4, properties {soundWaveAssetPtr=/DemonMaskVoid/Assets/Audio/wav/VoidMask_ThrownProj_Windup_01}}
 {id=WavePlayer_5, properties {soundWaveAssetPtr=/DemonMaskVoid/Assets/Audio/wav/VoidMask_ThrownProj_Windup_02}}
 {id=WavePlayer_6, properties {soundWaveAssetPtr=/DemonMaskVoid/Assets/Audio/wav/VoidMask_ThrownProj_Windup_03}}
 {id=WavePlayer_8, properties {soundWaveAssetPtr=/DemonMaskVoid/Assets/Audio/wav/VoidMask_ThrownProj_Windup_04}}
 {id=WavePlayer_9, properties {soundWaveAssetPtr=/DemonMaskVoid/Assets/Audio/wav/VoidMask_ThrownProj_Windup_05}}
]
properties {
 attenuationSettings=&/SA_VoidMask_Foley
 maxDistance=2600
 pitchMultiplier=1.5
 soundClassObject=/Game/Sounds/FortSoundClassesAndModes/Fort_06_Player_Attacks_Abilities
 volumeMultiplier=0.15
 concurrencyOverrides {maxCount=2}
 concurrencySet [&/CS_VoidMask]
}
