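# Sound cue for the first-person start of the Void Mask teleport, the enter sound.
# A random node picks one of four start wave variations with equal weight.
# Playback runs at reduced volume through a dedicated first-person attenuation asset and the player abilities sound class.
# The Void Mask concurrency set is assigned, alongside a concurrency override capping simultaneous instances.
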SoundCue /DemonMaskVoid/Assets/Audio/SC_VoidMask_Teleport_Enter_1P
root=Random_1
nodes [
 {id=Random_1, children [WavePlayer_6,WavePlayer_7,WavePlayer_8,WavePlayer_9], properties {weights [1,1,1,1]}}
 {id=WavePlayer_6, properties {soundWaveAssetPtr=/DemonMaskVoid/Assets/Audio/wav/VoidMask_Teleport_1P_Start_01}}
 {id=WavePlayer_7, properties {soundWaveAssetPtr=/DemonMaskVoid/Assets/Audio/wav/VoidMask_Teleport_1P_Start_02}}
 {id=WavePlayer_8, properties {soundWaveAssetPtr=/DemonMaskVoid/Assets/Audio/wav/VoidMask_Teleport_1P_Start_03}}
 {id=WavePlayer_9, properties {soundWaveAssetPtr=/DemonMaskVoid/Assets/Audio/wav/VoidMask_Teleport_1P_Start_04}}
]
properties {
 attenuationSettings=&/SA_VoidMask_Teleport_1P
 maxDistance=2097152
 soundClassObject=/Game/Sounds/FortSoundClassesAndModes/Fort_06_Player_Attacks_Abilities
 volumeMultiplier=0.5
 concurrencyOverrides {maxCount=2}
 concurrencySet [&/CS_VoidMask]
}
