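# Sound cue for the third-person start of the Void Mask teleport, routed to the player attacks and abilities sound class.
# A distance crossfade blends two random nodes, each picking from three equally weighted waves: close variants that fade out with range, and distant variants that fade in sharply during that fade, then tail off gradually.
# It uses Void Mask teleport attenuation and concurrency assets, with a slight pitch raise.
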SoundCue /DemonMaskVoid/Assets/Audio/SC_VoidMask_Teleport_Enter_3P
root=DistanceCrossFade_1
nodes [
 {
  id=DistanceCrossFade_1
  children [Random_4,Random_0]
  properties {crossFadeInput [
    {fadeInDistanceEnd=0, fadeInDistanceStart=0, fadeOutDistanceEnd=5000, fadeOutDistanceStart=3000, volume=1}
    {fadeInDistanceEnd=4000, fadeInDistanceStart=3900, fadeOutDistanceEnd=20000, fadeOutDistanceStart=5300, volume=1}
   ]}
 }
 {id=Random_0, children [WavePlayer_7,WavePlayer_8,WavePlayer_9], properties {weights [1,1,1]}}
 {id=Random_4, children [WavePlayer_4,WavePlayer_5,WavePlayer_6], properties {weights [1,1,1]}}
 {id=WavePlayer_4, properties {soundWaveAssetPtr=/DemonMaskVoid/Assets/Audio/wav/VoidMask_Teleport_3P_Close_Start_01}}
 {id=WavePlayer_5, properties {soundWaveAssetPtr=/DemonMaskVoid/Assets/Audio/wav/VoidMask_Teleport_3P_Close_Start_03}}
 {id=WavePlayer_6, properties {soundWaveAssetPtr=/DemonMaskVoid/Assets/Audio/wav/VoidMask_Teleport_3P_Close_Start_02}}
 {id=WavePlayer_7, properties {soundWaveAssetPtr=/DemonMaskVoid/Assets/Audio/wav/VoidMask_Teleport_3P_Dist_Start_01}}
 {id=WavePlayer_8, properties {soundWaveAssetPtr=/DemonMaskVoid/Assets/Audio/wav/VoidMask_Teleport_3P_Dist_Start_02}}
 {id=WavePlayer_9, properties {soundWaveAssetPtr=/DemonMaskVoid/Assets/Audio/wav/VoidMask_Teleport_3P_Dist_Start_03}}
]
properties {
 attenuationSettings=&/SA_VoidMask_Teleport
 maxDistance=25000
 pitchMultiplier=1.05
 soundClassObject=/Game/Sounds/FortSoundClassesAndModes/Fort_06_Player_Attacks_Abilities
 volumeMultiplier=1
 concurrencyOverrides {maxCount=2}
 concurrencySet [&/CS_VoidMask]
}
