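# First-person sound cue for the exit of the Void Mask teleport ability.
# A Random node picks one of four equally weighted end-of-teleport wave variations, played at slightly raised pitch and reduced volume.
# It routes to the player attacks and abilities sound class, references a Void Mask teleport attenuation asset, and uses the shared Void Mask concurrency set.
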SoundCue /DemonMaskVoid/Assets/Audio/SC_VoidMask_Teleport_Exit_1P
root=Random_0
nodes [
 {id=Random_0, children [WavePlayer_4,WavePlayer_5,WavePlayer_6,WavePlayer_7], properties {weights [1,1,1,1]}}
 {id=WavePlayer_4, properties {soundWaveAssetPtr=/DemonMaskVoid/Assets/Audio/wav/VoidMask_Teleport_1P_End_01}}
 {id=WavePlayer_5, properties {soundWaveAssetPtr=/DemonMaskVoid/Assets/Audio/wav/VoidMask_Teleport_1P_End_02}}
 {id=WavePlayer_6, properties {soundWaveAssetPtr=/DemonMaskVoid/Assets/Audio/wav/VoidMask_Teleport_1P_End_03}}
 {id=WavePlayer_7, properties {soundWaveAssetPtr=/DemonMaskVoid/Assets/Audio/wav/VoidMask_Teleport_1P_End_04}}
]
properties {
 attenuationSettings=&/SA_VoidMask_Teleport_1P
 maxDistance=2097152
 pitchMultiplier=1.1
 soundClassObject=/Game/Sounds/FortSoundClassesAndModes/Fort_06_Player_Attacks_Abilities
 volumeMultiplier=0.45
 attenuationOverrides {bAttenuate=false}
 concurrencyOverrides {maxCount=2}
 concurrencySet [&/CS_VoidMask]
}
