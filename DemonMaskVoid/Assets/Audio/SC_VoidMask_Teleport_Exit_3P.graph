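# Third-person sound cue for the end of a Void Mask teleport, played as a player ability sound with its own attenuation and concurrency assets.
# A distance crossfade at the root blends a close layer into a distant layer, with overlapping fade ranges and the distant layer set louder.
# Each layer feeds from a random node that picks one of four equally weighted wave variations.
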SoundCue /DemonMaskVoid/Assets/Audio/SC_VoidMask_Teleport_Exit_3P
root=DistanceCrossFade_0
nodes [
 {
  id=DistanceCrossFade_0
  children [Random_1,Random_0]
  properties {crossFadeInput [
    {fadeInDistanceEnd=0, fadeInDistanceStart=0, fadeOutDistanceEnd=8000, fadeOutDistanceStart=3000, volume=1.5}
    {fadeInDistanceEnd=8000, fadeInDistanceStart=5000, fadeOutDistanceEnd=23000, fadeOutDistanceStart=12000, volume=2}
   ]}
 }
 {id=Random_0, children [WavePlayer_0,WavePlayer_1,WavePlayer_2,WavePlayer_3], properties {weights [1,1,1,1]}}
 {id=Random_1, children [WavePlayer_4, WavePlayer_5, WavePlayer_10, WavePlayer_11], properties {weights [1,1,1,1]}}
 {id=WavePlayer_0, properties {soundWaveAssetPtr=/DemonMaskVoid/Assets/Audio/wav/VoidMask_Teleport_3P_Dist_End_04}}
 {id=WavePlayer_1, properties {soundWaveAssetPtr=/DemonMaskVoid/Assets/Audio/wav/VoidMask_Teleport_3P_Dist_End_03}}
 {id=WavePlayer_10, properties {soundWaveAssetPtr=/DemonMaskVoid/Assets/Audio/wav/VoidMask_Teleport_3P_Close_End_02}}
 {id=WavePlayer_11, properties {soundWaveAssetPtr=/DemonMaskVoid/Assets/Audio/wav/VoidMask_Teleport_3P_Close_End_01}}
 {id=WavePlayer_2, properties {soundWaveAssetPtr=/DemonMaskVoid/Assets/Audio/wav/VoidMask_Teleport_3P_Dist_End_02}}
 {id=WavePlayer_3, properties {soundWaveAssetPtr=/DemonMaskVoid/Assets/Audio/wav/VoidMask_Teleport_3P_Dist_End_01}}
 {id=WavePlayer_4, properties {soundWaveAssetPtr=/DemonMaskVoid/Assets/Audio/wav/VoidMask_Teleport_3P_Close_End_04}}
 {id=WavePlayer_5, properties {soundWaveAssetPtr=/DemonMaskVoid/Assets/Audio/wav/VoidMask_Teleport_3P_Close_End_03}}
]
properties {
 attenuationSettings=&/SA_VoidMask_Teleport
 maxDistance=25000
 pitchMultiplier=1.05
 soundClassObject=/Game/Sounds/FortSoundClassesAndModes/Fort_06_Player_Attacks_Abilities
 volumeMultiplier=1
 attenuationOverrides {bAttenuate=false}
 concurrencyOverrides {maxCount=2}
 concurrencySet [&/CS_VoidMask]
}
