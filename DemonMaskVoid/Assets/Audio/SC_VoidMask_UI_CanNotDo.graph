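# Sound cue for the Void Mask's "cannot do" feedback, played at reduced volume through the player attacks and abilities sound class.
# A mixer layers Fortnite's stock UI error ping, slightly delayed and louder in the mix, with a quieter mask-specific sound run through a modulator.
# Despite the UI name, it uses the mask's foley attenuation, alongside the mask's concurrency set.
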SoundCue /DemonMaskVoid/Assets/Audio/SC_VoidMask_UI_CanNotDo
root=Mixer_0
nodes [
 {id=Delay_0, children [WavePlayer_0], properties {delayMax=0.1, delayMin=0.1}}
 {id=Mixer_0, children [Delay_0,Modulator_0], properties {inputVolume [0.5,0.3]}}
 {id=Modulator_0, children [WavePlayer_2]}
 {id=WavePlayer_0, properties {soundWaveAssetPtr=/Game/Sounds/Fort_UI_Sounds/Gameplay/UI_Error_Ping_Cant_Do_Default_01}}
 {id=WavePlayer_2, properties {soundWaveAssetPtr=/DemonMaskVoid/Assets/Audio/wav/VoidMask_CanNotDo}}
]
properties {
 attenuationSettings=&/SA_VoidMask_Foley
 maxDistance=2600
 soundClassObject=/Game/Sounds/FortSoundClassesAndModes/Fort_06_Player_Attacks_Abilities
 volumeMultiplier=0.2
 concurrencyOverrides {maxCount=3, resolutionRule=StopOldest, retriggerTime=0.35}
 concurrencySet [&/CS_VoidMask]
}
