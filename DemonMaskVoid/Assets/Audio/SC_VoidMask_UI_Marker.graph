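# Sound cue for the Void Mask UI marker, playing its appear sound in the DemonMaskVoid content.
# A fixed one-second delay node feeds a single wave player.
# The cue overrides attenuation to play without falloff, routes to the critical in-game UI sound class at slightly reduced volume, and uses the Void Mask concurrency set.
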SoundCue /DemonMaskVoid/Assets/Audio/SC_VoidMask_UI_Marker
root=Delay_0
nodes [
 {id=Delay_0, children [WavePlayer_0], properties {delayMax=1, delayMin=1}}
 {id=WavePlayer_0, properties {soundWaveAssetPtr=/DemonMaskVoid/Assets/Audio/wav/VoidMask_UI_Marker_Appear}}
]
properties {
 bOverrideAttenuation=true
 maxDistance=2097152
 soundClassObject=&Fort_UI_InGame_Critical
 volumeMultiplier=0.8
 attenuationOverrides {bAttenuate=false}
 concurrencyOverrides {maxCount=1, retriggerTime=0.3}
 concurrencySet [&/CS_VoidMask]
}
