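# Sound cue for the Demon Mask Void out-of-ammo UI sound.
# A single wave plays after a short fixed delay, then passes through a modulator that randomizes pitch and slightly varies volume.
# The cue is pitched down and quieted, routed to the weapon foley sound class with Void Mask foley attenuation, and capped by a concurrency set plus count and retrigger overrides.
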SoundCue /DemonMaskVoid/Assets/Audio/SC_VoidMask_UI_OutOfAmmo
root=Modulator_0
nodes [
 {id=Delay_0, children [WavePlayer_1], properties {delayMax=0.1, delayMin=0.1}}
 {id=Modulator_0, children [Delay_0], properties {pitchMax=1.08, pitchMin=0.93, volumeMax=1, volumeMin=0.9}}
 {id=WavePlayer_1, properties {soundWaveAssetPtr=/DemonMaskVoid/Assets/Audio/wav/VoidMask_UI_OutOfAmmo}}
]
properties {
 attenuationSettings=&/SA_VoidMask_Foley
 maxDistance=2600
 pitchMultiplier=0.93
 soundClassObject=&Fort_WeaponFoley
 volumeMultiplier=0.85
 concurrencyOverrides {maxCount=3, retriggerTime=0.1}
 concurrencySet [&/CS_VoidMask]
}
