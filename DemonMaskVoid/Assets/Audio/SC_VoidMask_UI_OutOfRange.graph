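# Sound cue for the Void Mask's out-of-range UI feedback.
# A Mixer layers a heavily reduced stock Fortnite "can't do" error ping under a boosted custom Void Mask out-of-range wave.
# It uses the critical in-game UI sound class but carries Void Mask foley attenuation.
# The shared Void Mask concurrency set is assigned, alongside override values for a small stop-oldest voice cap with a retrigger delay.
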SoundCue /DemonMaskVoid/Assets/Audio/SC_VoidMask_UI_OutOfRange
root=Mixer_0
nodes [
 {id=Mixer_0, children [WavePlayer_0,WavePlayer_1], properties {inputVolume [0.1,1.4]}}
 {id=WavePlayer_0, properties {soundWaveAssetPtr=/Game/Sounds/Fort_UI_Sounds/Gameplay/UI_Error_Ping_Cant_Do_Default_01}}
 {id=WavePlayer_1, properties {soundWaveAssetPtr=/DemonMaskVoid/Assets/Audio/wav/VoidMask_UI_OutOfRange}}
]
properties {
 attenuationSettings=&/SA_VoidMask_Foley
 maxDistance=2600
 soundClassObject=&Fort_UI_InGame_Critical
 volumeMultiplier=1
 concurrencyOverrides {maxCount=3, resolutionRule=StopOldest, retriggerTime=0.3}
 concurrencySet [&/CS_VoidMask]
}
