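# Sound cue for the Void Mask UI sound that marks the end of the teleport cooldown.
# A Random node picks one of three equally weighted wave variants and feeds a Modulator that varies pitch and volume.
# It uses the critical in-game UI sound class with Void Mask foley attenuation, and references the Void Mask concurrency set alongside count and retrigger overrides.
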SoundCue /DemonMaskVoid/Assets/Audio/SC_VoidMask_UI_TeleportCooldown_Ended
root=Modulator_0
nodes [
 {id=Modulator_0, children [Random_1], properties {pitchMin=0.92, volumeMin=0.8}}
 {id=Random_1, children [WavePlayer_1,WavePlayer_2,WavePlayer_3], properties {weights [1,1,1]}}
 {id=WavePlayer_1, properties {soundWaveAssetPtr=/DemonMaskVoid/Assets/Audio/wav/VoidMask_Cooldown_End_Teleport_03}}
 {id=WavePlayer_2, properties {soundWaveAssetPtr=/DemonMaskVoid/Assets/Audio/wav/VoidMask_Cooldown_End_Teleport_02}}
 {id=WavePlayer_3, properties {soundWaveAssetPtr=/DemonMaskVoid/Assets/Audio/wav/VoidMask_Cooldown_End_Teleport_01}}
]
properties {
 attenuationSettings=&/SA_VoidMask_Foley
 maxDistance=2600
 soundClassObject=&Fort_UI_InGame_Critical
 volumeMultiplier=1
 concurrencyOverrides {maxCount=2, retriggerTime=1}
 concurrencySet [&/CS_VoidMask]
}
