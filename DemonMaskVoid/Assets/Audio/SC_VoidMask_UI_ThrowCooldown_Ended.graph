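# Sound cue for the Void Mask UI sound that signals the end of the throw cooldown.
# A Random node picks one of three equally weighted cooldown-end waves and feeds it through a default Modulator.
# The cue plays quieter and lower in pitch, using Void Mask foley attenuation and the player attacks and abilities sound class.
# A Void Mask concurrency set is assigned, alongside overrides for voice count and retrigger time.
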SoundCue /DemonMaskVoid/Assets/Audio/SC_VoidMask_UI_ThrowCooldown_Ended
root=Modulator_0
nodes [
 {id=Modulator_0, children [Random_1]}
 {id=Random_1, children [WavePlayer_0,WavePlayer_6,WavePlayer_7], properties {weights [1,1,1]}}
 {id=WavePlayer_0, properties {soundWaveAssetPtr=/DemonMaskVoid/Assets/Audio/wav/VoidMask_Cooldown_End_Throw_01}}
 {id=WavePlayer_6, properties {soundWaveAssetPtr=/DemonMaskVoid/Assets/Audio/wav/VoidMask_Cooldown_End_Throw_02}}
 {id=WavePlayer_7, properties {soundWaveAssetPtr=/DemonMaskVoid/Assets/Audio/wav/VoidMask_Cooldown_End_Throw_03}}
]
properties {
 attenuationSettings=&/SA_VoidMask_Foley
 maxDistance=2600
 pitchMultiplier=0.8
 soundClassObject=/Game/Sounds/FortSoundClassesAndModes/Fort_06_Player_Attacks_Abilities
 volumeMultiplier=0.35
 concurrencyOverrides {maxCount=2, retriggerTime=0.4}
 concurrencySet [&/CS_VoidMask]
}
