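; Audio gameplay volume Blueprint for the Demon Mask Void teleport-back hole that drives the PortalTimeWarp audio modulation mix.
; When toggled on, it activates the Control Bus Mix if player character 0 is a locally viewed Fortnite pawn without the DemonMaskVoid TeleportAbility gameplay tag.
; Toggling off or EndPlay deactivates the mix unconditionally.
; BeginPlay, overlap and Tick events are empty.
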(event EventBeginPlay)

(event Collision|EventActorBeginOverlap
  (params (OtherActor :type Actor%20Object%20Reference)))

(event EventTick
  (params (DeltaSeconds :type Float%20%28single-precision%29)))

(event OnToggledOn(AudioGameplayVolume)
  (bind _asfort_player_pawn_athena (Utilities|Casting|CastToFortPlayerPawnAthena (Game|GetPlayerCharacter 0)))
  (if (Pawn|IsLocallyViewed _asfort_player_pawn_athena)
    (if (not (GameplayTags|HasMatchingGameplayTag _asfort_player_pawn_athena "(TagName=\"Granted.Athena.DemonMaskVoid.TeleportAbility\")"))
      (Audio|Modulation|ActivateControlBusMix (Variables|Default|GetCBM_PortalTimeWarp)))))

(event OnToggledOff(AudioGameplayVolume)
  (Audio|Modulation|DeactivateControlBusMix (Variables|Default|GetCBM_PortalTimeWarp)))

(event EventEndPlay
  (params (EndPlayReason :type EEndPlayReason%20Enum :default "Destroyed"))
  (Audio|Modulation|DeactivateControlBusMix (Variables|Default|GetCBM_PortalTimeWarp)))

(fn ConstructionScript)
