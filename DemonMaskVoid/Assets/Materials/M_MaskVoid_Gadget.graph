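# Masked material for the Mask Void gadget that dissolves it in and out with a glowing, purple-by-default edge.
# It layers the Fortnite weapon master function with optional gem effect and vertex-color hiding, then tech-art override and Creative move-tool functions.
# Two custom primitive data timestamps drive the fade through combined radial and vertical gradients in pre-skinned local space.
# Scrolling, two-plane projected fire noise breaks up the opacity mask and emissive edge.
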Material /DemonMaskVoid/Assets/Materials/M_MaskVoid_Gadget
graph {
 properties {
  bAutomaticallySetUsageInEditor=false
  bUseMaterialAttributes=true
  bUsedWithClothing=true
  bUsedWithMeshDeformer=true
  bUsedWithMorphTargets=true
  bUsedWithNanite=true
  bUsedWithSkeletalMesh=true
  blendMode=BLEND_Masked
  pixelDepthOffsetMode=PDOM_Legacy
  refractionMethod=RM_IndexOfRefraction
  usedShadingModels=Default Lit
 }
 nodes [
  n1 = TextureSampleParameter2D() {group=Skin FX, parameterName=SkinFX_Mask, samplerType=SAMPLERTYPE_Masks, texture=&T_BlackMask}
  n2 = StaticComponentMaskParameter(None=AppendVector(A=@n1.RGB, B=@n1.RGB)) {defaultR=true, group=Skin FX, parameterName=Generic Gem Effect Channel}
  n3 = MaterialFunctionCall!(In) {materialFunction=/Game/Packages/Fortress_SharedMaterials/Base_Material_Functions/BaseCustomizationFunctions/MF_FN_Weapon_MASTER}
  n4 = GetMaterialAttributes(None=StaticSwitchParameter(
    True=MaterialFunctionCall!(MASK=@n2, Color Change Sequence, Material=@n3.Material Result) {materialFunction=/Game/Weapons/WeaponSkins/MaterialFunction/GenericGem}.Result
    False=@n3.Material Result
   ) {group=Skin FX, parameterName=Use Generic Gem}) {attributeGetTypes [#1]}
  n5 = MaterialFunctionCall!(WPO) {materialFunction=/Game/Packages/Fortress_SharedMaterials/Base_Material_Functions/Fortnite_HideElementsWithVertexColor}
  n6 = MaterialFunctionCall(In=SetMaterialAttributes!(
    MaterialAttributes=@n4.MaterialAttributes
    World Position Offset=StaticSwitchParameter(True=@n5.Result, False=@n4.World Position Offset) {group=Opacity Mask, parameterName=Use Vertex Colors for Mask}
   ) {attributeSetTypes [#1]}) {materialFunction=&MF_TechArtOverrides_accessories}
  n7 = GetMaterialAttributes(None=@n6.Result) {attributeGetTypes [#2]}
  n8 = VertexInterpolator(VS=LocalPosition() {localOrigin=InstancePreSkinning}.XYZ)
  n9 = OneMinus(None=Distance(A=@n8.PS, B=Constant()))
  n10 = Constant() {r=20}
  n11 = ScalarParameter() {defaultValue=0.3164, parameterName=FadeLerpRadial1}
  n12 = ScalarParameter() {defaultValue=2.5347, parameterName=FadeLerpRadial2}
  n13 = Subtract(A=Time() {period=1}, B=ScalarParameter() {bUseCustomPrimitiveData=true, parameterName=TimeStampIn})
  n14 = ScalarParameter() {defaultValue=1, parameterName=FadeSpeedIn}
  n15 = Divide(
   A=Subtract(A=Time() {period=1}, B=ScalarParameter() {bUseCustomPrimitiveData=true, parameterName=TimeStamp1Out, primitiveDataIndex=1})
   B=ScalarParameter() {defaultValue=0.7, parameterName=FadeSpeedOut}
  )
  n16 = Frac(None=Multiply(A=Saturate(None=@n15)) {constB=1.01})
  n17 = Saturate(None=Min(A=Divide(A=@n13, B=@n14), B=OneMinus(None=@n16)))
  n18 = Min(A=@n17) {constB=0.5}
  n19 = OneMinus(None=ComponentMask(None=@n8.PS) {b=true})
  n20 = Constant() {r=20}
  n21 = ScalarParameter() {defaultValue=-0.5026, parameterName=FadeVertical1}
  n22 = ScalarParameter() {defaultValue=2.2373, parameterName=FadeVertical2}
  n23 = Max(
   A=Add(A=Divide(A=@n9, B=@n10), B=LinearInterpolate(A=@n11, B=@n12, Alpha=@n18) {constA=0.352, constB=0.69}) {constB=0.66}
   B=Add(A=Divide(A=@n19, B=@n20), B=LinearInterpolate(A=@n21, B=@n22, Alpha=@n17) {constA=0.3, constB=0.66}) {constB=0.66}
  )
  n24 = OneMinus(None=Saturate(None=Abs(None=@n23)))
  n25 = Multiply(A=@n24, B=@n24)
  n26 = LocalPosition() {localOrigin=InstancePreSkinning}
  n27 = Multiply(A=Time(), B=Constant3Vector() {constant=<rgba 0 0 -10 0>})
  n28 = Divide(A=VertexInterpolator(VS=Add(A=@n26.XYZ, B=@n27)).PS, B=Constant() {r=70})
  n29 = Max(
   A=TextureSample(UVs=ComponentMask(None=@n28) {b=true, r=true}) {texture=/Game/Global/Textures/Default/Noise/T_FireNoise_01}.R
   B=TextureSample(UVs=ComponentMask(None=@n28) {b=true, g=true}) {texture=&T_FireNoise_01}.R
  )
  n30 = Multiply(A=@n29, B=@n29)
  n31 = OneMinus(None=@n17)
  n32 = Multiply(A=Multiply(A=@n31, B=@n31)) {constB=10}
  n33 = Multiply(
   A=Multiply(A=@n25, B=Add(A=@n25, B=@n30))
   B=Multiply(A=VectorParameter() {parameterName=Emission Color, defaultValue=<rgba 13.3499 1.9098 20 1>}.RGB, B=Add(A=@n32))
  )
  n34 = Saturate(None=@n23)
  n35 = Add(A=@n34, B=@n29)
  MaterialFunctionCall(MA In=SetMaterialAttributes!(MaterialAttributes=@n7.MaterialAttributes, Emissive Color=Add(A=@n7.Emissive Color, B=@n33), Opacity Mask=Multiply(A=@n34, B=@n35)) {attributeSetTypes [#2, #3]}) {materialFunction=&MF_CreativeMoveTool}
 ]
}
