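# Unlit translucent Niagara sprite material for the Void Mask ground VFX, shading soft, smoke-like particles.
# A swirl-flow texture distorts tiled noise UVs with per-particle variation; the noise multiplies tint and time-of-day smoke colors for emissive.
# Opacity math combines sphere-masked texture alpha, depth fade, squared near-camera fade and particle alpha.
# The graph also holds speed-gated fall/gravity vector math and a sine mask, with no shown consumers.
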Material /DemonMaskVoid/Assets/VFX/Materials/M_VoidMask_Ground
graph {
 properties {
  bGenerateSphericalParticleNormals=true
  bScreenSpaceReflections=true
  bUsedWithNiagaraSprites=true
  blendMode=BLEND_Translucent
  pixelDepthOffsetMode=PDOM_Legacy
  shadingModel=MSM_Unlit
  translucencyLightingMode=TLM_VolumetricPerVertexDirectional
  usedShadingModels=Unlit
 }
 nodes [
  n1 = ParticleSpeed()
  n2 = ScalarParameter() {defaultValue=2, group=Speed, parameterName=Speed}
  n3 = Constant()
  n4 = VectorParameter() {group=Fall, parameterName=HowFarToFallZ, defaultValue=<rgba 20 20 50 0>}
  n5 = Transform(None=@n4.RGB) {transformSourceType=TRANSFORMSOURCE_Local}
  n6 = DynamicParameter() {defaultValue=<rgba 1 0.5 2 1>, paramNames [Random,Direction,Speed,Param4]}
  n7 = Subtract(A=ParticleRandom(), B=@n6.Direction) {constB=0.5}
  n8 = AppendVector(A=Multiply(A=ComponentMask(None=@n5) {g=true, r=true}, B=Multiply(A=@n7, B=@n6.Speed) {constB=2}), B=@n4.B)
  n9 = ParticleRelativeTime()
  n10 = ScalarParameter() {defaultValue=2, desc=33, group=Fall, parameterName=Splash Speed Mul}
  n11 = Multiply(A=@n6.Random) {constB=0.5}
  n12 = Multiply(
   A=Power(Base=Add(A=@n9), Exp=@n10)
   B=Multiply(A=Add(A=@n11) {constB=0.5}, B=VectorParameter() {group=Fall, parameterName=Gravity, defaultValue=<rgba 0 0 -1 0>}.RGB)
  )
  If(
   A=Saturate(None=Divide(A=@n1, B=@n2))
   B=Constant() {r=1}
   "A > B"=@n3
   "A == B"=@n3
   "A < B"=LinearInterpolate(B=Subtract(A=@n8, B=@n12), Alpha=Multiply(A=@n9, B=@n10))
  )
  MaterialFunctionCall(
   Color_Morning=ScalarParameter() {defaultValue=0.7, parameterName=HDRMult_3}
   Color_Day=ScalarParameter() {defaultValue=1, parameterName=HDRMult_2}
   Color_Evening=ScalarParameter() {defaultValue=0.7, parameterName=HDRMult_1}
   Color_Night=ScalarParameter() {defaultValue=0.5, parameterName=HDRMult}
  ) {materialFunction=&Color_Time_of_Day}
  n13 = VectorParameter() {group=Smoke Colors, parameterName=2 - Morning Mids Smoke, sortPriority=1, defaultValue=<rgba 0.445 0.3231 0.248 1>}
  n14 = VectorParameter() {group=Smoke Colors, parameterName=2 - Daytime Mids Smoke, sortPriority=2, defaultValue=<rgba 0.6719 0.7182 0.75 1>}
  n15 = VectorParameter() {group=Smoke Colors, parameterName=2 - Evening Mids Smoke, sortPriority=3, defaultValue=<rgba 0.79 0.4657 0.4032 1>}
  n16 = VectorParameter() {group=Smoke Colors, parameterName=2 - Nighttime Mids Smoke, sortPriority=4, defaultValue=<rgba 0.2808 0.3998 0.4688 1>}
  n17 = DynamicParameter() {defaultValue=<rgba 1 0.5 2 1>, paramNames [Random,Direction,Speed,Param4]}
  n18 = VectorParameter() {parameterName=Color, defaultValue=<rgba 0.0938 0.0831 0.0754 1>}
  n19 = Multiply(A=Add(A=Multiply(A=@n17.Random) {constB=0.5}) {constB=0.5}, B=@n18.RGB)
  n20 = TextureCoordinate()
  n21 = ScalarParameter() {defaultValue=0.5, parameterName=UV Scale}
  n22 = TextureCoordinate() {uTiling=1.5, vTiling=1.5}
  n23 = VectorParameter() {group="Noise ", parameterName=Noise UV, defaultValue=<rgba 1 1 0.15 -0.2>}
  n24 = MaterialFunctionCall(X=@n23.R, Y=@n23.R) {materialFunction=&MakeFloat2}
  n25 = MaterialFunctionCall(X=@n23.B, Y=@n23.B) {materialFunction=&MakeFloat2}
  n26 = Add(A=ScalarParameter() {defaultValue=1, parameterName=NoisePanningSpeed}, B=ParticleRandom())
  n27 = TextureObjectParameter() {group="Noise ", parameterName=Noise Tex, samplerType=SAMPLERTYPE_LinearColor, texture=/Game/Global/Textures/Default/Flowmaps/T_SwirlFlow_02}
  n28 = TextureSample(UVs=Add(A=Multiply(A=@n22, B=@n24.Result), B=Multiply(A=@n25.Result, B=@n26)), Tex=@n27)
  n29 = ScalarParameter() {defaultValue=0.1, group="Noise ", parameterName=Noise Strength}
  n30 = DynamicParameter() {parameterIndex=1, defaultValue=<rgba 1 0.5 0 1>, paramNames [Noise, Rand, UV Anim, 3]}
  n31 = Add(A=Multiply(A=ComponentMask(None=@n28.RGB) {g=true, r=true}, B=Multiply(A=@n29, B=@n30.Noise)), B=Multiply(A=ParticleRandom()) {constB=3.3})
  n32 = DynamicParameter() {parameterIndex=1, defaultValue=<rgba 1 0.5 0 1>, paramNames [Noise, Rand, UV Anim, 3]}
  n33 = TextureSampleParameter2D(UVs=Add(A=Multiply(A=@n20, B=@n21), B=Add(A=@n31, B=@n32.UV Anim))) {parameterName=Texture, texture=/Game/Effects/Fort_Effects/Textures/Tile/T_TileNoise_03}
  MaterialFunctionCall(Emissive=Multiply(
    A=MaterialFunctionCall(Color_Morning=@n13.RGB, Color_Day=@n14.RGB, Color_Evening=@n15.RGB, Color_Night=@n16.RGB) {desc=Mids, materialFunction=&Color_Time_of_Day}.Result
    B=Multiply(A=@n19, B=@n33.R)
   )) {materialFunction=&MF_FX_EmissiveAdjustments}
  n34 = SphereMask(A=@n20, B=Constant() {r=0.5}, Radius=Constant() {r=0.5}) {hardnessPercent=75}
  n35 = MaterialFunctionCall(Float4=Multiply(A=NamedRerouteDeclaration(None=@n34) {name=SphereMask_Result, nodeColor=<rgba 0 0.4609 1 1>}, B=@n33.RGBA)) {materialFunction=&BreakOutFloat4Components}
  n36 = StaticSwitchParameter(True=@n35.A, False=@n35.A) {parameterName=Tex Alpha R channel}
  n37 = DepthFade(
   Opacity=Constant() {r=1}
   FadeDistance=LinearInterpolate(A=ScalarParameter() {defaultValue=50, parameterName=DepthFade Max}, Alpha=@n36) {constB=0.5}
  )
  n38 = MaterialFunctionCall(Distance=ScalarParameter() {defaultValue=400, parameterName=Near Camera Fade Distance}) {materialFunction=&Near_Camera_Fade}
  n39 = Multiply(A=@n38.Result, B=@n38.Result)
  Multiply(A=Multiply(A=Multiply(A=@n18.A, B=@n36), B=Multiply(A=@n37, B=@n39)), B=ParticleColor().A)
  n40 = Multiply(A=TextureCoordinate()) {constB=0.5}
  n41 = MaterialFunctionCall(Float2=Abs(None=Sine(None=@n40))) {materialFunction=&BreakOutFloat2Components}
  n42 = Saturate(None=Multiply(A=Multiply(A=@n41.R, B=@n41.R)) {constB=2})
  Multiply(A=@n42, B=@n42)
 ]
}
