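# Translucent unlit material for smoke trails in the Void Mask VFX, usable on Niagara ribbons and mesh particles.
# A panned smoke texture is shaped by a swiping sphere mask and a U-axis sine fade, with dynamic parameters driving pan, swipe, size and erosion.
# Particle color tints the mask-brightened emissive; particle alpha, inverted Fresnel and depth fade shape opacity.
# Switches add optional SplineThicken offset and bypass depth fade on low-quality, mobile and forward paths.
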Material /DemonMaskVoid/Assets/VFX/Materials/M_VoidMask_TrailSmoke
graph {
 properties {
  bAutomaticallySetUsageInEditor=false
  bUsedWithNiagaraMeshParticles=true
  bUsedWithNiagaraRibbons=true
  blendMode=BLEND_Translucent
  shadingModel=MSM_Unlit
  usedShadingModels=Unlit
 }
 nodes [
  Constant2Vector() {g=1, r=0.5}
  Constant()
  Constant() {r=100}
  Constant() {r=100}
  n1 = NamedRerouteUsage()
  n2 = ParticleColor()
  n3 = TextureCoordinate()
  n4 = MaterialFunctionCall(Float2=@n3) {materialFunction=&BreakOutFloat2Components}
  n5 = Multiply(A=@n4.R, B=ScalarParameter() {defaultValue=1, parameterName=Tex_Tiling_U})
  n6 = Add(
   A=Multiply(A=@n4.G, B=ScalarParameter() {defaultValue=1, parameterName=TexWidth}) {constB=0.4}
   B=ScalarParameter() {parameterName=TexOffset}
  ) {constB=0.4}
  n7 = DynamicParameter() {defaultValue=<rgba 1 0 1 0>, paramNames [PanSpeed,MaskSwipe,MaskSize,Erosion]}
  n8 = Constant()
  n9 = TextureSampleParameter2D(UVs=Panner(Coordinate=AppendVector(A=@n5, B=@n6), Speed=AppendVector(A=@n7.PanSpeed, B=@n8)) {speedX=0.5}) {parameterName=Tex1, samplerType=SAMPLERTYPE_Grayscale, texture=&T_Smoke_Trail_Soft}
  n10 = Constant() {r=1}
  n11 = Multiply(A=@n7.MaskSwipe)
  n12 = SphereMask(
   A=Multiply(A=@n3, B=AppendVector(A=@n7.MaskSize, B=@n10))
   B=AppendVector(A=Add(A=@n11) {constB=0}, B=Constant() {r=0.5})
   Radius=Constant() {r=0.5}
  ) {hardnessPercent=1}
  n13 = Multiply(A=MaterialFunctionCall(Float2=@n3) {materialFunction=&BreakOutFloat2Components}.R) {constB=0.5}
  n14 = Saturate(None=Multiply(A=Multiply(A=@n9.R, B=@n12), B=Sine(None=@n13)))
  n15 = DynamicParameter() {defaultValue=<rgba 1 0 1 0>, paramNames [PanSpeed,MaskSwipe,MaskSize,Erosion]}
  n16 = Subtract(A=@n14, B=LinearInterpolate(Alpha=@n15.Erosion) {constA=-1})
  n17 = OneMinus(None=Fresnel() {baseReflectFraction=0, exponent=1})
  n18 = ScalarParameter() {defaultValue=3, parameterName=Fresnel_Power}
  n19 = Multiply(
   A=Multiply(A=@n2.A, B=Multiply(A=@n14, B=@n16)) {constA=1}
   B=StaticSwitchParameter(True=Power(Base=@n17, Exp=@n18) {constExponent=3}, False=Constant() {r=1}) {defaultValue=true, parameterName=Use_Fresnel}
  )
  n20 = ScalarParameter() {parameterName=Debug_Texture}
  LinearInterpolate(
   A=StaticSwitchParameter(True=Multiply(A=@n19, B=@n1), False=@n19) {defaultValue=true, parameterName=Use_DepthFade}
   B=Constant() {r=1}
   Alpha=@n20
  )
  n21 = LinearInterpolate(A=LinearInterpolate(Alpha=Power(Base=@n12) {constExponent=4}) {constA=0.6, constB=5}, B=Constant() {r=1}, Alpha=@n15.Erosion)
  MaterialFunctionCall(Emissive=LinearInterpolate(A=Multiply(A=@n21, B=@n2.RGB), B=@n19, Alpha=@n20)) {materialFunction=&MF_FX_EmissiveAdjustments}
  n22 = Constant() {r=1}
  n23 = QualitySwitch(Default=DepthFade(FadeDistance=ScalarParameter() {defaultValue=8, parameterName=DepthFade_Distance}), Low=@n22)
  NamedRerouteDeclaration(None=ShadingPathSwitch(Default=ShadingPathSwitch(Default=@n23, Mobile=@n22) {desc=Migrated from FeatureLevelSwitch}, Forward=@n22)) {name=Depth Fade, nodeColor=<rgba 1 0.1577 0 1>}
  StaticBool() {value=true}
  n24 = ScalarParameter() {defaultValue=200, parameterName=SplineThicken_Base}
  n25 = ScalarParameter() {defaultValue=200, parameterName=SplineThicken_Tip}
  n26 = StaticBool()
  StaticSwitchParameter(
   True=Multiply(A=MaterialFunctionCall(WidthBase=@n24, WidthTip=@n25, Expand U or V UV Channel=@n26) {materialFunction=&SplineThicken}.WorldPositionOffset)
   False=Constant()
  ) {parameterName=Use_SplineThicken}
  WorldPosition()
 ]
}
