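# Translucent, unlit, two-sided Niagara ribbon material for the Void Mask wandering boss connect effect, likely a streaking link.
# Particle color RGB passes through an emissive adjustment function for the glow.
# Two textures pan along U: a pickaxe trail shaped by a refine function, and a soft smoke trail shaped by add, multiply and power controls.
# The two are summed and scaled by particle alpha to form the translucency mask.
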Material /DemonMaskVoid/Assets/VFX/Materials/M_VoidMask_WanderingBossConnect
graph {
 properties {
  bAutomaticallySetUsageInEditor=false
  bUsedWithNiagaraRibbons=true
  blendMode=BLEND_Translucent
  shadingModel=MSM_Unlit
  twoSided=true
  usedShadingModels=Unlit
 }
 nodes [
  n1 = ParticleColor()
  MaterialFunctionCall(Emissive=@n1.RGB) {materialFunction=&MF_FX_EmissiveAdjustments}
  n2 = TextureObjectParameter() {
   group=T2
   parameterName=T2_Texture
   sortPriority=0
   texture=/Game/Weapons/FORT_Melee/Pickaxe_Gimmick_Female/FX/T_Pickaxe_Gimmick_Female_Trail_01
  }
  n3 = VectorParameter() {group=T2, parameterName=T2_UVControl, channelNames {a=VPan, b=UPan, g=VScale, r=UScale}, defaultValue=<rgba 1 1 -0.5 0>}
  n4 = MaterialFunctionCall(Texture=@n2, UVControl_RGB=@n3, UVControl_A=@n3) {materialFunction=&FXCommon_TextureSample}
  n5 = VectorParameter() {parameterName=RF, defaultValue=<rgba 2 5 0.5 0.5>}
  n6 = TextureObjectParameter() {
   group=MainText
   parameterName=MainText_Texture
   samplerType=SAMPLERTYPE_Grayscale
   sortPriority=0
   texture=&T_Smoke_Trail_Soft
  }
  n7 = VectorParameter() {group=MainText, parameterName=MainText_UVControl, channelNames {a=VPan, b=UPan, g=VScale, r=UScale}, defaultValue=<rgba 1 1 -1.5 0>}
  n8 = Add(
   A=MaterialFunctionCall!(
    Texture=@n6
    AdditionalUVOffset
    UVControl_RGB=@n7
    UVControl_A=@n7
    "[Override]TimeInput"
    "[Override]ScaleUVByCenter"=StaticBoolParameter() {defaultValue=true, parameterName="[Override]ScaleUVByCenter (SB)"}
    "[Override]UVInput"
   ) {materialFunction=/Game/Effects/Niagara/MasterMaterials/Functions/Common/FXCommon_TextureSample_V2}.R
   B=ScalarParameter() {group=MainText, parameterName=MainText_GrayAdjust_Add}
  )
  n9 = ScalarParameter() {defaultValue=1, group=MainText, parameterName=MainText_GrayAdjust_Mul}
  n10 = Power(
   Base=Saturate(None=Multiply(A=@n8, B=@n9))
   Exp=ScalarParameter() {defaultValue=0.4, group=MainText, parameterName=MainText_GrayAdjust_Power}
  )
  Multiply(
   A=@n1.A
   B=Add(
    A=MaterialFunctionCall(Texture=@n4.R, RGB=@n5.RGB, A=@n5.RGB) {materialFunction=&MF_FX_Refine}.Result
    B=Saturate(None=@n10)
   )
  )
 ]
}
