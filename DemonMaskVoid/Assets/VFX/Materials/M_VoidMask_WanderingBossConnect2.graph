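# Additive, unlit, two-sided Niagara ribbon material, named as a Void Mask wandering-boss connection effect.
# Chained UV distortion has Perlin noise warp a plasma texture, whose output then warps the main ribbon shape texture.
# The inverted shape times its boosted, saturated copy forms a mask that keeps dimmer non-black areas.
# The mask is refined and tinted by particle color for emissive adjustment, and separately multiplied by particle alpha.
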Material /DemonMaskVoid/Assets/VFX/Materials/M_VoidMask_WanderingBossConnect2
graph {
 properties {
  bAutomaticallySetUsageInEditor=false
  bUsedWithNiagaraRibbons=true
  blendMode=BLEND_Additive
  shadingModel=MSM_Unlit
  twoSided=true
  usedShadingModels=Unlit
 }
 nodes [
  n1 = ParticleColor()
  n2 = TextureObjectParameter() {
   group=MainT
   parameterName=MainT_Texture
   samplerType=SAMPLERTYPE_Grayscale
   sortPriority=0
   texture=/Game/Gadgets/Assets/VinderTech_GliderChute/Glider_Thrive_Female/FX/Texture/T_RibbonBaseShape4
  }
  n3 = TextureObjectParameter() {
   group=N1
   parameterName=N1_Texture
   samplerType=SAMPLERTYPE_LinearColor
   sortPriority=0
   texture=/Game/Effects/Fort_Effects/Emotes/UltraLight/Texture/T_Plasma_D_C
  }
  n4 = TextureObjectParameter() {
   group=N2
   parameterName=N2_Texture
   samplerType=SAMPLERTYPE_LinearGrayscale
   sortPriority=0
   texture=&T_SimplePerlin_Greyscale
  }
  n5 = VectorParameter() {group=N2, parameterName=N2_UVControl, channelNames {a=VPan, b=UPan, g=VScale, r=UScale}, defaultValue=<rgba 2 2 -0.3 0>}
  n6 = VectorParameter() {
   group=N2
   parameterName=N2_NoiseControl
   sortPriority=60
   channelNames {a=VStrength, b=UStrength, g=VOffset, r=UOffset}
   defaultValue=<rgba -0.5 -0.5 0.1 0.1>
  }
  n7 = MaterialFunctionCall!(
   NoiseValue=MaterialFunctionCall(Texture=@n4, UVControl_RGB=@n5, UVControl_A=@n5) {materialFunction=&FXCommon_TextureSample}.RG
   Control_RGB=@n6
   Control_A=@n6
  ) {materialFunction=/Game/Effects/Niagara/MasterMaterials/Functions/Common/FXCommon_NoiseControl}
  n8 = VectorParameter() {group=N1, parameterName=N1_UVControl, channelNames {a=VPan, b=UPan, g=VScale, r=UScale}, defaultValue=<rgba 1 1 -0.2 0>}
  n9 = VectorParameter() {
   group=N1
   parameterName=N1_NoiseControl
   sortPriority=60
   channelNames {a=VStrength, b=UStrength, g=VOffset, r=UOffset}
   defaultValue=<rgba -0.2976 -0.3976 0.3 0.2>
  }
  n10 = VectorParameter() {group=MainT, parameterName=MainT_UVControl, channelNames {a=VPan, b=UPan, g=VScale, r=UScale}, defaultValue=<rgba 1 1 0 0>}
  n11 = MaterialFunctionCall(
   Texture=@n2
   AdditionalUVOffset=MaterialFunctionCall(
    NoiseValue=MaterialFunctionCall(Texture=@n3, AdditionalUVOffset=@n7.NoiseValue, UVControl_RGB=@n8, UVControl_A=@n8) {materialFunction=&FXCommon_TextureSample}.RG
    Control_RGB=@n9
    Control_A=@n9
   ) {materialFunction=&FXCommon_NoiseControl}.NoiseValue
   UVControl_RGB=@n10
   UVControl_A=@n10
  ) {materialFunction=&FXCommon_TextureSample}
  n12 = ScalarParameter() {defaultValue=10, parameterName=fwe}
  n13 = Multiply(A=OneMinus(None=@n11.R), B=Saturate(None=Multiply(A=@n11.R, B=@n12)))
  n14 = VectorParameter() {parameterName=FWEF, defaultValue=<rgba 25 2 0.05 0.5>}
  MaterialFunctionCall(Emissive=Multiply(
    A=@n1.RGB
    B=MaterialFunctionCall(Texture=@n13, RGB=@n14.RGB, A=@n14.RGB) {materialFunction=&MF_FX_Refine}.Result
   )) {materialFunction=&MF_FX_EmissiveAdjustments}
  Multiply(A=@n1.A, B=@n13)
 ]
}
