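# Niagara system for the Void Mask charge effect: purple flash, smoke and energy on a skeletal body mesh.
# Flash (sprites plus light) and looping smoke wisps follow one bone at the mask offset, while nine partitioned decal sprites sit on separate bones.
# System velocity range drives a decal material counter.
# Speed-stretched streaks spawn offset from the system under negative point attraction; an idle smoke emitter is disabled.
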NiagaraSystem /DemonMaskVoid/Assets/VFX/NS_VoidMask_Charge
dependencies {
 usedDataInterfaces [
  &NiagaraDataInterfaceCurve
  &NiagaraDataInterfaceColorCurve
  &NiagaraDataInterfaceSpriteRendererInfo
  &NiagaraDataInterfaceSkeletalMesh
 ]
 usedDynamicInputs []
 usedModules [
  &self:SystemSpawnScript.SystemScriptSource.SystemScriptGraph.NiagaraNodeAssignment_0.SetVariables_#1
  &self:SystemSpawnScript.SystemScriptSource.SystemScriptGraph.NiagaraNodeAssignment_1.SetVariables_#2
  &SystemState
 ]
}
emitters {
 Decals {
  emitterUpdateScript [
   &EmitterState(Life Cycle Mode=::NewEnumerator1, Loop Behavior=::NewEnumerator1, Loop Delay=0.07, Loop Duration=2, UseLoopDelay=true)
   &SpawnBurst_Instantaneous(Spawn Count=9, Use Spawn Probability=false, UseLoopCountLimit=false)
  ]
  particleSpawnScript [
   &PartitionParticles(Number Of Partitions=9)
   &InitializeParticle(
    Color=<rgba 0.4333 0 1.3 1.5>
    Lifetime=0.5
    Sprite Size Mode=::NewEnumerator3
    Uniform Sprite Size=40
    UsePositionOffset=false
    Write Material Random=true
   )
   &CameraOffset(Camera Offset Amount=10)
  ]
  particleUpdateScript [
   &ParticleState()
   &SkeletalMeshLocation(
    Bone / Socket Index=$Particles.PartitionParticles.Partition
    Bone Sampling Mode=::NewEnumerator4
    MaskSpawnGroup=false
    Orientation Sampling=::NewEnumerator1
    Skeletal Mesh=$System.SkeletalMesh_Body
   )
   /Niagara/Modules/Update/Utility/IncrementOverTime(Rate of Change:NiagaraFloat=&Lerp_Float(A=-1.1, B=2, Alpha=$System.NormalizedVelocityRange))
   &DynamicMaterialParameters(
    Index 0 Param 1=&FloatFromCurve(
     FloatCurve=DataInterface() {propertyValues {Curve {keys [{arriveTangent=3.2804, interpMode=RCIM_Cubic, leaveTangent=3.2804, tangentMode=RCTM_User, time=0, value=1}, <at 0.7 2>]}}}
     CurveIndex=$System.Age
    )
    Index 0 Param 2=1
    Index 0 Param 3=&Add_Float(
     A=$Particles.IncrementOverTime.Counter
     B=&FloatFromCurve(
      FloatCurve=DataInterface() {propertyValues {Curve {keys [{arriveTangent=-2.1645, interpMode=RCIM_Cubic, leaveTangent=-2.1645, tangentMode=RCTM_User, time=0, value=1}, <at 1.25 0>]}}}
      CurveIndex=$System.Age
      Scale Curve=-0.5
     )
    )
    Index 0 Param 4=&RandomRangeFloat(Minimum=0.75, Maximum=0.9)
    Index 1 Param 1=50
    Index 1 Param 2=10
    Module.Param0WriteEnabled=true
    Module.Param0WriteEnabled1=true
    Module.Param1WriteEnabled=true
    Module.Param1WriteEnabled1=true
    Module.Param2WriteEnabled=true
    Module.Param2WriteEnabled1=false
    Module.Param3WriteEnabled=true
    Module.Param3WriteEnabled1=false
    Write Parameter Index 1=true
   )
   &ScaleColor(
    Scale Alpha=&Multiply_Float(
     A=$User.GlobalOpacity
     B=&FloatFromCurve(FloatCurve=DataInterface() {propertyValues {Curve {keys [
          {arriveTangent=20.9312, interpMode=RCIM_Cubic, leaveTangent=20.9312, tangentMode=RCTM_User, time=0, value=0}
          {arriveTangent=1.1179, interpMode=RCIM_Cubic, leaveTangent=1.1179, tangentMode=RCTM_User, time=0.2, value=1.125}
          {arriveTangent=0.0224, interpMode=RCIM_Cubic, leaveTangent=0.0224, tangentMode=RCTM_User, time=0.5, value=0.35}
         ]}}})
    )
    Scale RGB=&VectorFromFloat(Value=&Lerp_Float(
      A=1
      B=6
      Alpha=&FloatFromCurve(FloatCurve=DataInterface() {propertyValues {Curve {keys [
           {interpMode=RCIM_Cubic, tangentMode=RCTM_User, time=0, value=0}
           {arriveTangent=-4.8045, interpMode=RCIM_Cubic, leaveTangent=-4.8045, tangentMode=RCTM_User, time=0.2, value=1.05}
           {tangentMode=RCTM_User, time=1, value=0}
          ]}}})
     ))
    ScaleA=true
    ScaleRGB=true
   )
  ]
  rendererClasses [&NiagaraSpriteRendererProperties]
  renderers [{rendererClass=&NiagaraSpriteRendererProperties}]
 }
 Flash {
  emitterUpdateScript [
   &EmitterState()
   &SpawnBurst_Instantaneous(Use Spawn Probability=false, UseLoopCountLimit=true)
   /SetVariables(Emitter.EquipCurve:NiagaraFloat=&FloatFromCurve(
     FloatCurve=DataInterface() {propertyValues {Curve {keys [
         {interpMode=RCIM_Cubic, tangentMode=RCTM_User, time=0, value=0}
         {arriveTangent=4.25, interpMode=RCIM_Cubic, leaveTangent=4.25, time=0.1, value=0.2}
         {arriveTangent=-0.0083, interpMode=RCIM_Cubic, leaveTangent=-0.0083, tangentMode=RCTM_User, time=0.2, value=0.85}
         {tangentMode=RCTM_User, time=0.5, value=0}
        ]}}}
     CurveIndex=$Emitter.Age
    )) {isSetParameters=true}
  ]
  particleSpawnScript [
   &PartitionParticles(ID to Partition:NiagaraInt32=&ReturnExecIndex()) {enabled=false}
   &InitializeParticle(
    Lifetime=0.5
    Mesh Scale=&Multiply_Vector(A=<1 1 1.2>, B=&VectorFromFloat(Value=1.4))
    Mesh Scale Mode=::NewEnumerator4
    Sprite Size Mode=::NewEnumerator3
    Uniform Sprite Size=6
    UsePositionOffset=false
    Write Material Random=true
   )
   &DynamicMaterialParameters(
    Index 0 Param 1=&RandomRangeFloat(Minimum=0.4, Maximum=0.6)
    Index 0 Param 2=1
    Index 0 Param 3=1
    Index 0 Param 4=1
    Module.Index0Param0Write=true
    Module.Index0Param1Write=true
    Module.Index0Param2Write=true
    Module.Index0Param3Write=true
   )
   /SetVariables() {enabled=false, isSetParameters=true}
   &InitialMeshOrientation(Mesh Orientation Mode=::NewEnumerator7, Rotation=<0 0.25 0>, Use Rotation Vector=true)
   /SetVariables(Particles.VisibilityTag=$Particles.PartitionParticles.Partition) {enabled=false, isSetParameters=true}
  ]
  particleUpdateScript [
   &SkeletalMeshLocation(
    Bone / Socket Index=9
    Bone Sampling Mode=::NewEnumerator4
    MaskSpawnGroup=false
    Sampled Position Offset=$System.MaskOffsetPosition
    Skeletal Mesh=$System.SkeletalMesh_Body
   )
   &ParticleState()
   &SolveForcesAndVelocity(Clamp Velocity=false, Limit Acceleration=false)
   &ScaleColor(
    Scale Alpha=&Multiply_Float(
     A=1
     B=&FloatFromCurve(FloatCurve=DataInterface() {propertyValues {Curve {keys [<at 0 0>,<at 0.2 1>,<at 0.8 1>,<at 1 0>]}}})
    )
    ScaleA=true
    ScaleRGB=true
   ) {enabled=false}
   &ScaleSpriteSize(Scale Factor=&Vector2DFromFloat(Value=&FloatFromCurve(FloatCurve=DataInterface() {propertyValues {Curve {keys [<at 0 1.5 RCIM_Cubic>,<at 1 0>]}}}))) {enabled=false}
   &CameraOffset(Camera Offset Amount=&FloatFromCurve(
     FloatCurve=DataInterface() {propertyValues {Curve {keys [
         {interpMode=RCIM_Cubic, tangentMode=RCTM_User, time=0, value=0.1}
         {arriveTangent=0.4369, interpMode=RCIM_Cubic, leaveTangent=0.4369, tangentMode=RCTM_User, time=0.5, value=1}
        ]}}}
     Scale Curve=35
    )) {module=CameraOffset001}
   /SetVariables(Particles.Flash size:Vector2f=&Multiply_Vector2D_ByFloat(
     Float=&FloatFromCurve(
      FloatCurve=DataInterface() {propertyValues {Curve {keys [
          {arriveTangent=3.2405, interpMode=RCIM_Cubic, leaveTangent=3.2405, tangentMode=RCTM_User, time=0, value=0}
          {arriveTangent=0.0224, interpMode=RCIM_Cubic, leaveTangent=0.0224, tangentMode=RCTM_User, time=1, value=1}
         ]}}}
      Scale Curve=200
     )
     Vector2D=<1 0.7>
    )) {isSetParameters=true}
   /SetVariables(Particles.Flash color:LinearColor=&MakeLinearColorFromVectorAndFloat(
     "Vector (RGB)"=&MakeVectorFromLinearColorRGB(Color=<rgba 0.2833 0.011 1.1 0.55>)
     "Float (Alpha)"=&FloatFromCurve(FloatCurve=DataInterface() {propertyValues {Curve {keys [
          {interpMode=RCIM_Cubic, tangentMode=RCTM_User, time=0, value=0}
          {arriveTangent=-2.5948, interpMode=RCIM_Cubic, leaveTangent=-2.5948, tangentMode=RCTM_User, time=0.124, value=1}
          {interpMode=RCIM_Cubic, tangentMode=RCTM_User, time=1, value=0}
         ]}}})
    )) {isSetParameters=true}
   &FN_Light_Attributes(
    Color=<rgba 0.3149 0.0865 1 1>
    Create Color Attribute=true
    Create Exponent Attribute=true
    Create Radius Attribute=true
    Exponent=7
    Radius=50
    Scale Color=&FloatFromCurve(
     FloatCurve=DataInterface() {propertyValues {Curve {keys [
         {interpMode=RCIM_Cubic, tangentMode=RCTM_User, time=0, value=0}
         {interpMode=RCIM_Cubic, tangentMode=RCTM_User, time=0.5, value=1}
         {tangentMode=RCTM_User, time=1, value=0}
        ]}}}
     Scale Curve=200
    )
   )
   /SetVariables(
    Particles.ColorDark:LinearColor=&ColorFromCurve(ColorCurve=DataInterface() {propertyValues {
       AlphaCurve {keys [<at 0 0>,<at 0.0846 0.7 RCIM_Cubic>,<at 1 0 RCIM_Cubic>]}
       BlueCurve {keys [<at 0 0.01 RCIM_Cubic>]}
       GreenCurve {keys [<at 0 0.007 RCIM_Cubic>]}
       RedCurve {keys [<at 0 0.008 RCIM_Cubic>]}
      }})
    Particles.SpriteSizeDark:Vector2f=&Vector2DFromFloat(Value=&Lerp_Float(
      A=0
      B=55
      Alpha=&FloatFromCurve(FloatCurve=DataInterface() {propertyValues {Curve {keys [
           {interpMode=RCIM_Cubic, tangentMode=RCTM_User, time=0, value=0}
           {interpMode=RCIM_Cubic, tangentMode=RCTM_User, time=0.1, value=1}
           {interpMode=RCIM_Cubic, tangentMode=RCTM_User, time=0.7, value=0}
          ]}}})
     ))
   ) {isSetParameters=true}
  ]
  rendererClasses [&NiagaraSpriteRendererProperties, &NiagaraLightRendererProperties]
  renderers [
   {rendererClass=&NiagaraSpriteRendererProperties}
   {rendererClass=&NiagaraLightRendererProperties, rendererIndex=1}
   {rendererClass=&NiagaraSpriteRendererProperties, rendererIndex=2}
  ]
 }
 SmokeLoopIdle002 {
  bEnabled=false
  emitterSpawnScript [/SetVariables(Emitter.Position:NiagaraPosition=&TransformVector(Vector=Linked() {linkedVariable {name=User.Offset Pos, type=NiagaraPosition}}, Destination Space=::World, Transform Vector as Position=true)) {enabled=false, isSetParameters=true}]
  emitterUpdateScript [
   &EmitterState(Life Cycle Mode=::NewEnumerator1, Loop Behavior=::NewEnumerator1, Loop Delay=0.5, Loop Duration=0.5, UseLoopDelay=false)
   &SpawnBurst_Instantaneous(Use Spawn Probability=false, UseLoopCountLimit=false)
   /SetVariables(Emitter.SharedCurve1=DataInterface() {propertyValues {Curve {keys [
        {arriveTangent=1.1977, interpMode=RCIM_Cubic, leaveTangent=1.1977, tangentMode=RCTM_User, time=0, value=0}
        {arriveTangent=1.914, interpMode=RCIM_Cubic, leaveTangent=0.736, tangentMode=RCTM_Break, time=0.1432, value=0.1703}
        {arriveTangent=-0.7631, interpMode=RCIM_Cubic, leaveTangent=-0.0298, tangentMode=RCTM_Break, time=0.2949, value=0.8528}
        {arriveTangent=-5.5008, interpMode=RCIM_Cubic, leaveTangent=-5.5008, tangentMode=RCTM_User, time=1, value=0}
       ]}}}) {isSetParameters=true}
   /SetVariables(
    Emitter.CameraOffset=-25
    Emitter.Color:LinearColor=&ColorFromCurve(
     ColorCurve=DataInterface() {propertyValues {
       AlphaCurve {keys [<at 0 0 RCIM_Cubic>,<at 0.0826 2 RCIM_Cubic>]}
       BlueCurve {keys [<at 0.4494 0 RCIM_Cubic>]}
       GreenCurve {keys [<at 0.4494 0 RCIM_Cubic>]}
       RedCurve {keys [<at 0.4494 0 RCIM_Cubic>]}
      }}
     CurveIndex=$Emitter.NormalizedLoopAge
    )
   ) {isSetParameters=true}
   /SetVariables(
    Emitter.DynamicParam0_inky:Vector4f=&MakeVector4(
     X=&MakeFloatFromVector4(
      Vector 4=&MakeVector4(
       X=&FloatFromCurve(
        FloatCurve=DataInterface() {propertyValues {Curve {keys [
            {arriveTangent=4.1069, interpMode=RCIM_Cubic, leaveTangent=4.1069, tangentMode=RCTM_User, time=0, value=0}
            {interpMode=RCIM_Cubic, tangentMode=RCTM_User, time=0.4, value=1.1}
            {interpMode=RCIM_Cubic, leaveTangent=-1.8333, time=1, value=1}
           ]}}}
        CurveIndex=$Emitter.NormalizedLoopAge
        Scale Curve=0.5
       )
       Y=&RandomRangeFloat(Minimum=0, Maximum=1)
       Z=&FloatFromCurve(
        FloatCurve=DataInterface() {propertyValues {Curve {keys [{tangentMode=RCTM_User, time=0, value=0}, {tangentMode=RCTM_User, time=1, value=1}]}}}
        CurveIndex=$Emitter.NormalizedLoopAge
       )
       W=&FloatFromCurve(
        FloatCurve=DataInterface() {propertyValues {Curve {keys [<at 0.45 0 RCIM_Cubic>, {arriveTangent=11.9729, tangentMode=RCTM_User, time=0.99, value=10}]}}}
        CurveIndex=$Emitter.NormalizedLoopAge
       )
      )
      Channel=::NewEnumerator0
     )
     Y=&RandomRangeFloat(Minimum=0, Maximum=1)
     Z=&FloatFromCurve(
      FloatCurve=DataInterface() {propertyValues {Curve {keys [
          {arriveTangent=1.7411, interpMode=RCIM_Cubic, leaveTangent=1.7411, tangentMode=RCTM_User, time=0.1, value=0}
          {tangentMode=RCTM_User, time=0.8, value=1}
         ]}}}
      CurveIndex=$Emitter.NormalizedLoopAge
     )
     W=&FloatFromCurve(
      FloatCurve=DataInterface() {propertyValues {Curve {keys [<at 0.2 0 RCIM_Cubic>, {arriveTangent=36.7024, tangentMode=RCTM_User, time=0.99, value=10}]}}}
      CurveIndex=$Emitter.NormalizedLoopAge
     )
    )
    Emitter.Scale:Vector3f=&Multiply_VectorByFloat(
     Vector=&Multiply_Vector(
      A=&VectorFromFloat(Value=&FloatFromCurve(FloatCurve=$Emitter.SharedCurve1, CurveIndex=$Emitter.NormalizedLoopAge, Scale Curve=$User.Global Scale))
      B=<1.12 1.12 0.4>
     )
     Float=1
    )
    Emitter.SpriteSize:Vector2f=&Multiply_Vector2D_ByFloat(
     Float=&MakeFloatFromVector(
      Vector=&Multiply_Vector(
       A=&VectorFromFloat(Value=&FloatFromCurve(
         FloatCurve=DataInterface() {propertyValues {Curve {keys [
             {arriveTangent=0.3782, interpMode=RCIM_Cubic, leaveTangent=0.3782, tangentMode=RCTM_User, time=0, value=0.2}
             {arriveTangent=0.3146, interpMode=RCIM_Cubic, leaveTangent=-0.8436, tangentMode=RCTM_Break, time=1.1, value=1}
            ]}}}
         CurveIndex=$Emitter.NormalizedLoopAge
        ))
       B=<1.4 1.4 0.4>
      )
      Channel=::NewEnumerator0
     )
     Vector2D=&Multiply_Vector2D_ByFloat(Float=$User.Global Scale, Vector2D=<85 95>)
    )
   ) {isSetParameters=true}
   &SampleSkeletalMesh(
    AdditionalYawPitchRoll=false
    Bone / Socket Index=9
    Bone Sampling Mode:ENiagaraBoneSamplingMode=::NewEnumerator4
    Skeletal Mesh:NiagaraDataInterfaceSkeletalMesh=$System.SkeletalMesh_Body
   ) {enabled=false}
  ]
  rendererClasses [&NiagaraSpriteRendererProperties]
  renderers [{rendererClass=&NiagaraSpriteRendererProperties}]
 }
 SmokeWhisps {
  emitterUpdateScript [
   &EmitterState(Life Cycle Mode=::NewEnumerator1, Loop Behavior=::NewEnumerator1, Loop Delay=0.02, Loop Duration=0.1, UseLoopDelay=true)
   &SpawnBurst_Instantaneous(Spawn Count=4, Use Spawn Probability=false, UseLoopCountLimit=false)
   &SpawnRate(SpawnRate=10, Use Spawn Probability=false) {enabled=false}
  ]
  particleSpawnScript [
   &InitializeParticle(
    Color Maximum=<rgba 1 0.4531 0.9055 1>
    Color Minimum=<rgba 0.294 0.2185 0.6 1>
    Color Mode=::NewEnumerator2
    Lifetime Max=1
    Lifetime Min=0.6
    Lifetime Mode=::NewEnumerator1
    Mass Mode=::NewEnumerator3
    Sprite Rotation Mode=::NewEnumerator2
    Sprite Size Mode=::NewEnumerator1
    Uniform Sprite Size Max=70
    Uniform Sprite Size Min=50
    UsePositionOffset=false
    Write Material Random=true
   )
   &ShapeLocation(MaskSpawnGroup=false, Offset=<0 -15 0>, Offset Mode=::NewEnumerator0, Override Seed=false, Sphere Radius=3, UseOwnerScale=true)
   &AddVelocity(Velocity=&RandomRangeVector(Minimum=<-10 -10 15>, Maximum=<10 10 35>)) {enabled=false}
   &InheritVelocity(Inherited Velocity Speed Limit=1000, Speed Limit=true, UseSpeedThreshold=false) {enabled=false}
   &SkeletalMeshLocation(
    Bone / Socket Index=9
    Bone Sampling Mode=::NewEnumerator4
    MaskSpawnGroup=false
    Sampled Position Offset=&AddVectorToPosition(Position=Linked() {linkedVariable {name=System.MaskOffsetPosition, type=Vector3f}}, Vector=<-4 -7 0>)
    Skeletal Mesh=$System.SkeletalMesh_Body
   )
  ]
  particleUpdateScript [
   &ParticleState()
   &ScaleSpriteSize(
    Scale Sprite Size Mode=::NewEnumerator0
    Uniform Scale Factor=&Lerp_Float(
     A=0.9
     B=1
     Alpha=&FloatFromCurve(FloatCurve=DataInterface() {propertyValues {Curve {keys [
          {arriveTangent=0.0489, interpMode=RCIM_Cubic, leaveTangent=0.0489, tangentMode=RCTM_User, time=0, value=0}
          {interpMode=RCIM_Cubic, tangentMode=RCTM_User, time=0.5, value=1}
         ]}}})
    )
   )
   &ScaleColor(
    Scale Alpha=&FloatFromCurve(FloatCurve=DataInterface() {propertyValues {Curve {keys [<at 0 0 RCIM_Cubic>,<at 0.1 1 RCIM_Cubic>,<at 1 0 RCIM_Cubic>]}}}, Scale Curve=3)
    Scale RGB=&MakeVectorFromLinearColorRGB(Color=&ColorFromCurve(ColorCurve=DataInterface() {propertyValues {
        AlphaCurve {keys [<at 0 1 RCIM_Cubic>]}
        BlueCurve {keys [{arriveTangent=-2.3815, interpMode=RCIM_Cubic, time=0.1322, value=0.7}, <at 0.4771 0.1>, <at 0.8877 0 RCIM_Cubic>]}
        GreenCurve {keys [{arriveTangent=-0.4763, interpMode=RCIM_Cubic, time=0.1322, value=0.161}, <at 0.4771 0.03>, <at 0.8877 0 RCIM_Cubic>]}
        RedCurve {keys [{arriveTangent=-1.7925, interpMode=RCIM_Cubic, time=0.1322, value=0.4074}, <at 0.4771 0.033>, <at 0.8877 0 RCIM_Cubic>]}
       }}))
    ScaleA=true
    ScaleRGB=true
   )
   &AccelerationForce(Acceleration=<0 0 60>)
   &Drag(Use Linear Drag=true, Use Rotational Drag=true)
   &InheritSourceMovement(
    Applied Position Scale=&VectorFromFloat(Value=&FloatFromCurve(FloatCurve=DataInterface() {propertyValues {Curve {keys [<at 0 1>,<at 0.7 0>]}}}))
    ApplyForce=false
    ApplyPosition=true
    ApplyPositionFalloff=false
   )
   &SolveForcesAndVelocity(Clamp Velocity=false, Limit Acceleration=false)
   &SubUVAnimation(
    End Frame Range Override=32
    "Sprite Renderer "=DataInterface() {propertyValues {SpriteRenderer=&self:SmokeWhisps_0.NiagaraSpriteRendererProperties_0}}
    UseEndFrame=true
    UseStartFrame=true
   )
   &DynamicMaterialParameters(
    Index 0 Param 1=&FloatFromCurve(FloatCurve=DataInterface() {propertyValues {Curve {keys [
         {interpMode=RCIM_Cubic, tangentMode=RCTM_User, time=0.2, value=0}
         {arriveTangent=2.1174, interpMode=RCIM_Cubic, leaveTangent=2.1174, tangentMode=RCTM_User, time=1, value=1}
        ]}}})
    Index 0 Param 2=40
    Module.Index0Param0Write=true
    Module.Index0Param1Write=true
    Module.Index0Param2Write=false
    Module.Index0Param3Write=false
   )
   &CameraOffset(Camera Offset Amount=&Lerp_Float(
     A=0
     B=35
     Alpha=&FloatFromCurve(FloatCurve=DataInterface() {propertyValues {Curve {keys [{interpMode=RCIM_Cubic, tangentMode=RCTM_User, time=0, value=0}, {interpMode=RCIM_Cubic, tangentMode=RCTM_User, time=0.7, value=1}]}}})
    ))
  ]
  rendererClasses [&NiagaraSpriteRendererProperties]
  renderers [{rendererClass=&NiagaraSpriteRendererProperties}]
 }
 Streaks {
  emitterUpdateScript [
   &EmitterState(Life Cycle Mode=::NewEnumerator1, Loop Behavior=::NewEnumerator1, Loop Delay=0.03, Loop Duration=0.25, UseLoopDelay=false)
   &SpawnRate(
    SpawnRate=&FloatFromCurve(FloatCurve=DataInterface() {propertyValues {Curve {keys [<at 0.2 1>,<at 1 0>]}}}, CurveIndex=$Emitter.NormalizedLoopAge, Scale Curve=600)
    Use Spawn Probability=false
   )
  ]
  particleSpawnScript [
   &InitializeParticle(
    Color=&RandomRangeLinearColor(Minimum=<rgba 0.9689 0 1 1>, Maximum=<rgba 0.2479 0 1 1>)
    Lifetime Max=0.15
    Lifetime Min=0.1
    Lifetime Mode=::NewEnumerator1
    Sprite Size=&MaskVector2dBySpawnGroup(
     Masked Vector2D=&UniformRangedVector2D(Minimum=<0.5 1>, Maximum=<1 1>)
     Passthrough Vector2D=&UniformRangedVector2D(Minimum=<1 1>, Maximum=<1.5 1>)
     Spawn Group Mask=0
    )
    Sprite Size Mode=::NewEnumerator4
    UsePositionOffset=false
    Write Material Random=true
   ) {module=InitializeParticle001}
   &SystemLocation(
    MaskSpawnGroup=false
    Offset=&Add_Vector(A=<0 -200 0>, B=&UniformRangedVector(Minimum=<-100 -100 -75>, Maximum=<100 100 75>))
   )
  ]
  particleUpdateScript [
   &ParticleState()
   &SkeletalMeshLocation(
    Bone / Socket Index=2
    Bone Sampling Mode=::NewEnumerator4
    "Inherit Velocity (Scale)"=-1
    MaskSpawnGroup=false
    Orientation Sampling=::NewEnumerator1
    Sampled Position Offset=&RandomRangeVector(Minimum=<-200 -200 25>, Maximum=<200 200 125>)
    Skeletal Mesh=$System.Skeletal Mesh
    Velocity Sampling=::NewEnumerator0
    VelocitySpeedLimit=false
   ) {enabled=false}
   &PointAttractionForce(Attraction Radius=5000, Attraction Strength=-800, Kill Within Radius=false, Module.Use Falloff=false, UseAttractorOffset=false)
   &SolveForcesAndVelocity(Clamp Velocity=false, Limit Acceleration=false)
   &ScaleSpriteSizeBySpeed(Max Scale Factor=<1 70>, Min Scale Factor=<0.5 50>)
   &ScaleColor(
    Scale Alpha=&FloatFromCurve(FloatCurve=DataInterface() {propertyValues {Curve {keys [
         {arriveTangent=6.3626, interpMode=RCIM_Cubic, time=0, value=0}
         <at 0.5 0.9779 RCIM_Cubic>
         {interpMode=RCIM_Cubic, leaveTangent=-0.6839, time=1, value=0}
        ]}}})
    ScaleA=true
    ScaleRGB=false
   )
  ]
  rendererClasses [&NiagaraSpriteRendererProperties]
  renderers [{rendererClass=&NiagaraSpriteRendererProperties}]
 }
}
summary {
 systemName=NS_VoidMask_Charge
 emitters [
  {emitterName=Decals, rendererClasses [&NiagaraSpriteRendererProperties]}
  {emitterName=Streaks, rendererClasses [&NiagaraSpriteRendererProperties]}
  {bEnabled=false, emitterName=SmokeLoopIdle002, rendererClasses [&NiagaraSpriteRendererProperties]}
  {emitterName=Flash, rendererClasses [&NiagaraSpriteRendererProperties, &NiagaraLightRendererProperties]}
  {emitterName=SmokeWhisps, rendererClasses [&NiagaraSpriteRendererProperties]}
 ]
 userVariables [
  User.Global Scale {type=NiagaraFloat, defaultValue=1, description=""}
  User.GlobalOpacity {type=NiagaraFloat, defaultValue=1, description=""}
  User.Offset Pos {type=NiagaraPosition, description="", defaultValue {}}
 ]
}
