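# Niagara effect for equipping the Void Mask, layering a purple void burst with a mask flame, eye trail and smoke.
# Vortex, chromatic refraction and curl-noise bits form the bursts, alongside socket-following wisps and two-tone smoke dimmed by time of day.
# Flame and eye-trail meshes persist by resampling the skeletal mesh each update.
# The flame passes smoothed position into a dynamic parameter and fades along the system timeline.
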NiagaraSystem /DemonMaskVoid/Assets/VFX/NS_VoidMask_Equip
dependencies {
 usedDataInterfaces [
  &NiagaraDataInterfaceSpriteRendererInfo
  &NiagaraDataInterfaceCurve
  &NiagaraDataInterfaceColorCurve
  &NiagaraDataInterfaceSkeletalMesh
 ]
 usedDynamicInputs []
 usedModules [
  &self:SystemSpawnScript.SystemScriptSource.SystemScriptGraph.NiagaraNodeAssignment_0.SetVariables_#1
  &SystemState
 ]
}
emitters {
 Bits_Up_ {
  emitterUpdateScript [
   &EmitterState(Life Cycle Mode=::NewEnumerator1, Loop Behavior=::NewEnumerator1, Loop Delay=0.23, Loop Duration=0.5, UseLoopDelay=true)
   &SpawnBurst_Instantaneous(Spawn Count=30, Use Spawn Probability=false)
  ]
  particleSpawnScript [
   &InitializeParticle(
    Color=<rgba 0.026 0.026 0.026 2>
    Lifetime Max=0.35
    Lifetime Min=0.2
    Lifetime Mode=::NewEnumerator1
    Sprite Size Max=<5 25>
    Sprite Size Min=<5 14>
    Sprite Size Mode=::NewEnumerator2
    UsePositionOffset=false
    Write Material Random=true
   ) {module=InitializeParticle001}
   &SkeletalMeshLocation(Bone Sampling Mode=::NewEnumerator4, MaskSpawnGroup=false, Orientation Sampling=::NewEnumerator1, Skeletal Mesh=$System.Skeletal Mesh)
   &AddVelocity(Velocity=&RandomRangeVector(Minimum=<-1 -1 -1>, Maximum=<1 1 1>), Velocity Speed Scale=500)
   &CameraOffset(Camera Offset Amount=35) {enabled=false, module=CameraOffset001}
  ]
  particleUpdateScript [
   &ParticleState()
   &CurlNoiseForce(
    Noise Frequency=5
    Noise Quality / Cost=::NewEnumerator0
    Noise Strength=&Multiply_Float(A=900, B=1)
    Pan Noise=false
   )
   &Drag(Drag=&UniformRangedFloat(Minimum=5, Maximum=9), Use Linear Drag=true, Use Rotational Drag=false)
   &GravityForce(Gravity=&RandomRangeVector(Minimum=<0 0 -250>, Maximum=<0 0 -400>))
   &SolveForcesAndVelocity(Clamp Velocity=false, Limit Acceleration=false)
   &ScaleSpriteSize(Scale Factor=&Vector2DFromFloat(Value=&FloatFromCurve(FloatCurve=DataInterface() {propertyValues {Curve {keys [<at 0 1>,<at 1 0>]}}})))
   &ScaleSpriteSizeBySpeed(Max Scale Factor=<0.3 3>, Min Scale Factor=<0.7 0.5>)
   &ScaleColor(
    Scale RGB=&MakeVectorFromLinearColorRGB(Color=&ColorFromCurve(ColorCurve=DataInterface() {propertyValues {
        AlphaCurve {keys []}
        BlueCurve {keys [<at 0.0421 1>,<at 0.1987 32>,<at 0.4817 0>]}
        GreenCurve {keys [<at 0.0421 0>,<at 0.1987 0>,<at 0.4817 0>]}
        RedCurve {keys [<at 0.0421 0.1669>,<at 0.1987 3.3116>,<at 0.4817 0>]}
       }}))
    ScaleA=true
    ScaleRGB=true
   )
  ]
  rendererClasses [&NiagaraSpriteRendererProperties]
  renderers [{rendererClass=&NiagaraSpriteRendererProperties}]
 }
 ChromaticRefraction {
  emitterUpdateScript [
   &EmitterState(Life Cycle Mode=::NewEnumerator1, Loop Behavior=::NewEnumerator1, Loop Duration Mode=::NewEnumerator1, UseLoopDelay=false)
   &SpawnBurst_Instantaneous(Use Spawn Probability=false, UseLoopCountLimit=false)
  ]
  particleSpawnScript [
   &InitializeParticle(
    Lifetime=0.35
    Mesh Scale Mode=::NewEnumerator3
    Sprite Size Mode=::NewEnumerator3
    Uniform Sprite Size=140
    UsePositionOffset=false
    Write Material Random=true
   )
   &CameraOffset(Camera Offset Amount=20)
  ]
  particleUpdateScript [
   &ParticleState()
   &Timeline(
    Play Rate=&FloatFromCurve(
     FloatCurve=DataInterface() {propertyValues {Curve {keys [
         {interpMode=RCIM_Cubic, leaveTangent=-12.5334, tangentMode=RCTM_Break, time=0, value=4}
         {interpMode=RCIM_Cubic, tangentMode=RCTM_Break, time=0.7, value=1}
        ]}}}
     Scale Curve=0.5
    )
    Timeline End=1.0000000200408773e+20
   )
   &ScaleSpriteSize(
    Scale Sprite Size Mode=::NewEnumerator0
    Uniform Scale Factor=&FloatFromCurve(FloatCurve=DataInterface() {propertyValues {Curve {keys [
         {interpMode=RCIM_Cubic, tangentMode=RCTM_User, time=0, value=0}
         {interpMode=RCIM_Cubic, tangentMode=RCTM_User, time=0.3, value=1}
         {arriveTangent=-0.0632, tangentMode=RCTM_User, time=1, value=0}
        ]}}})
   )
   &DynamicMaterialParameters(
    Index 0 Param 1=1
    Index 0 Param 2=0.104
    Index 0 Param 4=0.2
    Index 1 Param 2=500
    Index 1 Param 4=1
    Index 2 Param 1=$Particles.Timeline.Time
    Index 2 Param 2=$Particles.Timeline.Time
    Index 2 Param 3=$Particles.Timeline.Time
    Index 3 Param 1=&FloatFromCurve(FloatCurve=DataInterface() {propertyValues {Curve {keys [<at 0.1 1>,<at 0.4 0>]}}})
    Index 3 Param 2=30
    Module.Index0Param0Write=true
    Module.Index0Param1Write=true
    Module.Index0Param2Write=true
    Module.Index0Param3Write=true
    Module.Index1Param0Write=true
    Module.Index1Param1Write=true
    Module.Index1Param2Write=true
    Module.Index1Param3Write=true
    Module.Index2Param0Write=true
    Module.Index2Param1Write=true
    Module.Index2Param2Write=true
    Module.Index2Param3Write=true
    Module.Index3Param0Write=true
    Module.Index3Param1Write=true
    Module.Index3Param2Write=true
    Module.Index3Param3Write=true
    Write Parameter Index 1=true
    Write Parameter Index 2=true
    Write Parameter Index 3=true
   ) {module=DynamicMaterialParameters001}
   &ScaleColor(
    Scale RGB=&MakeVectorFromLinearColorRGB(Color=&ColorFromCurve(ColorCurve=DataInterface() {propertyValues {
        AlphaCurve {keys [<at 0 1 RCIM_Cubic>]}
        BlueCurve {keys [<at 0.1752 2 RCIM_Cubic>,<at 0.9394 0.22 RCIM_Cubic>]}
        GreenCurve {keys [<at 0.1752 1.2 RCIM_Cubic>,<at 0.9394 0.132 RCIM_Cubic>]}
        RedCurve {keys [<at 0.1752 1.6677 RCIM_Cubic>,<at 0.9394 0.1725 RCIM_Cubic>]}
       }}))
    ScaleA=true
    ScaleRGB=true
   ) {module=ScaleColor001}
  ]
  rendererClasses [&NiagaraSpriteRendererProperties]
  renderers [{rendererClass=&NiagaraSpriteRendererProperties}]
 }
 EyeTrail {
  emitterUpdateScript [
   &EmitterState(Life Cycle Mode=::NewEnumerator1, Loop Behavior=::NewEnumerator1, Loop Duration=1, UseLoopDelay=false)
   &SpawnBurst_Instantaneous(Use Spawn Probability=false, UseLoopCountLimit=false)
  ]
  particleSpawnScript [
   &InitializeParticle(
    Color=<rgba 3.561 10 9.0009 1>
    Lifetime Max=2.5
    Lifetime Min=1.5
    Lifetime Mode=::NewEnumerator1
    Mesh Scale Mode=::NewEnumerator3
    Position Offset=<5 5 0>
    Sprite Size Mode=::NewEnumerator3
    Uniform Sprite Size=5
    UsePositionOffset=false
    Write Material Random=true
   )
   &InitialMeshOrientation(Orientation Vector=<0 0 -1>, Use Orientation Vector=true, Use Rotation Vector=false) {enabled=false}
   &BoxLocation(Box Offset=<-152 3 0>, Box Size=<0.1 0.1 0.1>, MaskSpawnGroup=false, Module.Override Seed=false, Rotate Box=false, Surface Only=false) {enabled=false}
   &CameraOffset(Camera Offset Amount=1) {enabled=false}
  ]
  particleUpdateScript [
   &ParticleState(Kill Particles When Lifetime Has Elapsed=false)
   &DynamicMaterialParameters(
    Index 0 Param 3=1
    Module.Index0Param0Write=true
    Module.Index0Param1Write=true
    Module.Index0Param2Write=true
    Module.Index0Param3Write=true
    Module.Index1Param0Write=true
    Module.Index1Param1Write=true
    Module.Index1Param2Write=true
    Module.Index1Param3Write=true
    Module.Index2Param0Write=true
    Module.Index2Param1Write=true
    Module.Index2Param2Write=true
    Module.Index2Param3Write=true
    Write Parameter Index 1=true
    Write Parameter Index 2=true
   )
   &SkeletalMeshLocation(
    Bone Sampling Mode=::NewEnumerator4
    MaskSpawnGroup=false
    Orientation Sampling=::NewEnumerator1
    Sampled Position Offset=<-0.5 -4 0>
    Skeletal Mesh=$System.Skeletal Mesh
   )
  ]
  rendererClasses [&NiagaraMeshRendererProperties]
  renderers [{rendererClass=&NiagaraMeshRendererProperties, rendererIndex=0}]
 }
 Main_Flame {
  emitterUpdateScript [
   &EmitterState(
    Inactive Response=::NewEnumerator1
    Life Cycle Mode=::NewEnumerator1
    Loop Behavior=::NewEnumerator1
    Loop Duration Mode=::NewEnumerator1
    UseLoopDelay=false
   )
   &SpawnBurst_Instantaneous(Spawn Count=1, Use Spawn Probability=false)
  ]
  particleSpawnScript [
   &InitializeParticle(
    Color=<rgba 1 1 1 4>
    Mesh Scale=&Multiply_Vector(A=<1 1.2 1.5>, B=&VectorFromFloat(Value=0.95))
    Mesh Scale Mode=::NewEnumerator4
    Sprite Size Mode=::NewEnumerator3
    Uniform Sprite Size=58
    UsePositionOffset=false
    Write Material Random=true
   ) {module=InitializeParticle001}
   /SetVariables(Particles.Mesh02_Size:Vector3f=<1 1 1.1>) {enabled=false, isSetParameters=true}
   /SetVariables(Particles.Mesh02_Color:LinearColor=<rgba 0.9438 1.0157 1.2 1.5>) {enabled=false, isSetParameters=true}
  ]
  particleUpdateScript [
   &ParticleState(Kill Particles When Lifetime Has Elapsed=false, Let Infinitely Lived Particles Die When Emitter Deactivates=true)
   &ScaleMeshSize(Scale Factor=&VectorFromFloat(Value=&FloatFromCurve(FloatCurve=DataInterface() {propertyValues {Curve {keys [
          {arriveTangent=9.0725, interpMode=RCIM_Cubic, leaveTangent=9.0725, tangentMode=RCTM_User, time=0, value=0}
          {arriveTangent=0.6298, interpMode=RCIM_Cubic, leaveTangent=0.6298, tangentMode=RCTM_User, time=0.1, value=1.45}
          {arriveTangent=-0.2064, interpMode=RCIM_Cubic, leaveTangent=-0.2064, tangentMode=RCTM_User, time=0.7, value=1.125}
          {arriveTangent=-0.0405, interpMode=RCIM_Cubic, leaveTangent=-0.0405, tangentMode=RCTM_User, time=1, value=1}
         ]}}})))
   /SetVariables(Particles.MeshSize2:Vector3f=&Multiply_VectorByFloat(
     Vector=<1 1 1>
     Float=&FloatFromCurve(
      FloatCurve=DataInterface() {propertyValues {Curve {keys [
          {arriveTangent=0.5843, interpMode=RCIM_Cubic, leaveTangent=0.5843, tangentMode=RCTM_User, time=0, value=0}
          {arriveTangent=-1.1475, interpMode=RCIM_Cubic, leaveTangent=-1.1475, tangentMode=RCTM_User, time=0.3, value=1.2}
          {arriveTangent=-0.0405, interpMode=RCIM_Cubic, leaveTangent=-0.0405, tangentMode=RCTM_User, time=1, value=1}
         ]}}}
      CurveIndex=$Emitter.Age
     )
    )) {isSetParameters=true}
   &InterpolateOverTime(Rate Of Change=10, Target Vector Value:NiagaraPosition=$Particles.Position, Variable Type=::NewEnumerator2)
   &SkeletalMeshLocation(
    Bone Sampling Mode=::NewEnumerator4
    MaskSpawnGroup=false
    Sampled Position Offset=<0 -1 2>
    Sampled Position Offset Coordinate Space:ENiagaraSkelMeshTransforms=::NewEnumerator2
    Skeletal Mesh=$System.Skeletal Mesh
   )
   /SetVariables(Particles.Interpolate Dynamic=&MakeFloat4FromVectorAndFloat("Vector (RGB)"=$Output.InterpolateOverTime.Moving Average, "Float (Alpha)"=0.4)) {isSetParameters=true}
   &SolveForcesAndVelocity(Clamp Velocity=false, Limit Acceleration=false)
   /SetVariables(Particles.Mesh02_Cameraoffset=3) {enabled=false, isSetParameters=true}
   &ScaleColor(
    Scale Alpha=&Multiply_Float(
     A=&FloatFromCurve(
      FloatCurve=DataInterface() {propertyValues {Curve {keys [
          {arriveTangent=-2.5993, interpMode=RCIM_Cubic, leaveTangent=-2.5993, tangentMode=RCTM_User, time=0, value=1}
          {arriveTangent=-0.2961, interpMode=RCIM_Cubic, leaveTangent=-0.2961, tangentMode=RCTM_User, time=1, value=0.05}
         ]}}}
      CurveIndex=$System.Timeline.NormalizedTime
     )
     B=&FloatFromCurve(FloatCurve=DataInterface() {propertyValues {Curve {keys [
          {interpMode=RCIM_Cubic, tangentMode=RCTM_User, time=0.2, value=0}
          {arriveTangent=2.7824, interpMode=RCIM_Cubic, leaveTangent=2.7824, tangentMode=RCTM_User, time=0.8, value=1}
         ]}}})
    )
    Scale RGB=&VectorFromFloat(Value=&FloatFromCurve(FloatCurve=DataInterface() {propertyValues {Curve {keys [
          {arriveTangent=3.4804, interpMode=RCIM_Cubic, leaveTangent=3.4804, tangentMode=RCTM_User, time=0.05, value=0.5}
          {arriveTangent=-27.8495, interpMode=RCIM_Cubic, leaveTangent=-27.8495, tangentMode=RCTM_User, time=0.1, value=5}
          <at 0.5 1>
         ]}}}))
    ScaleA=true
    ScaleRGB=true
   )
   /SetVariables(Particles.ColorDark:LinearColor=&MultiplyLinearColor(
     Linear Color=$Particles.Color
     Scale Factor=&MakeFloat4FromVectorAndFloat("Vector (RGB)"=<0.1 0.1 0.1>, "Float (Alpha)"=1)
    )) {isSetParameters=true}
   /SetVariables(Particles.FakeLightColor:LinearColor=&ColorFromCurve(ColorCurve=DataInterface() {propertyValues {
       AlphaCurve {keys [<at 0 0 RCIM_Cubic>,<at 0.5283 1 RCIM_Cubic>]}
       BlueCurve {keys [<at 0 1 RCIM_Cubic>,<at 1 1 RCIM_Cubic>]}
       GreenCurve {keys [<at 0 0 RCIM_Cubic>,<at 1 0 RCIM_Cubic>]}
       RedCurve {keys [<at 0 0.5875 RCIM_Cubic>,<at 1 0.4554 RCIM_Cubic>]}
      }})) {isSetParameters=true}
  ]
  rendererClasses [&NiagaraMeshRendererProperties, &NiagaraSpriteRendererProperties]
  renderers [
   {rendererClass=&NiagaraMeshRendererProperties, rendererIndex=0}
   {rendererClass=&NiagaraMeshRendererProperties, rendererIndex=1}
   {rendererClass=&NiagaraSpriteRendererProperties, rendererIndex=2}
  ]
 }
 Smoke {
  emitterUpdateScript [
   &EmitterState()
   &SpawnBurst_Instantaneous(Use Spawn Probability=false, UseLoopCountLimit=true) {enabled=false}
   &SpawnRate(SpawnRate=4, Use Spawn Probability=false)
   &SpawnRate(SpawnGroup=1, SpawnRate=4, Use Spawn Probability=false) {module=SpawnRate001}
  ]
  particleSpawnScript [
   &InitializeParticle(
    Color=&MaskLinearColorBySpawnGroup(Masked Linear Color=<rgba 0 0 0 2>, Passthrough Linear Color=<rgba 0.733 0 3 2>)
    Lifetime=&MaskFloatBySpawnGroup(Masked Float=1.5, Passthrough Float=1)
    Position Offset=&MaskVectorBySpawnGroup(Masked Vector=<0 -3 0>, Passthrough Vector=<0 -5 -2>)
    Position Offset Coordinate Space=::Local
    Sprite Rotation Mode=::NewEnumerator2
    Sprite Size Mode=::NewEnumerator1
    Sprite UV Mode=::NewEnumerator3
    Uniform Sprite Size Max=60
    Uniform Sprite Size Min=50
    UsePositionOffset=true
    Write Material Random=true
   )
   &SubUVAnimation(
    End Frame Range Override=4
    Evaluation Type=::NewEnumerator0
    "Sprite Renderer "=DataInterface() {propertyValues {SpriteRenderer=&self:Smoke_0.NiagaraSpriteRendererProperties_1}}
    SubUV Animation Mode:ENiagara_SubUVLookupModeV2=::NewEnumerator2
    UseEndFrame=true
    UseStartFrame=true
   )
   &CameraOffset(Camera Offset Amount=4)
   &AddVelocity(Velocity=<0 1 0>, Velocity Speed Scale=-10)
  ]
  particleUpdateScript [
   &ParticleState()
   &SpriteRotationRate(Rotation Rate=&RandomRangeFloat(Minimum=-10, Maximum=10))
   &Drag(Use Linear Drag=true, Use Rotational Drag=false)
   &SolveForcesAndVelocity(Clamp Velocity=false, Limit Acceleration=false)
   &ScaleSpriteSize(
    Uniform Curve Scale=&Multiply_Float(
     A=&FloatFromCurve(
      FloatCurve=DataInterface() {propertyValues {Curve {keys [
          {arriveTangent=-0.5767, interpMode=RCIM_Cubic, leaveTangent=-0.5767, tangentMode=RCTM_User, time=0, value=1}
          {arriveTangent=-0.0083, interpMode=RCIM_Cubic, leaveTangent=-0.0083, tangentMode=RCTM_User, time=1, value=0.8}
         ]}}}
      CurveIndex=$System.Timeline.NormalizedTime
     )
     B=&MaskFloatBySpawnGroup(Masked Float=1, Passthrough Float=0.5)
    )
    Uniform Curve Sprite Scale=DataInterface() {propertyValues {Curve {keys [
        {arriveTangent=4.1785, interpMode=RCIM_Cubic, leaveTangent=4.1785, tangentMode=RCTM_User, time=0, value=0}
        {arriveTangent=1.3326, interpMode=RCIM_Cubic, leaveTangent=1.3326, tangentMode=RCTM_User, time=0.2042, value=0.5}
        {arriveTangent=0.0224, interpMode=RCIM_Cubic, leaveTangent=0.0224, tangentMode=RCTM_User, time=1, value=1}
       ]}}}
   )
   &DynamicMaterialParameters(
    Index 0 Param 1=&RandomRangeFloat(Minimum=0, Maximum=111)
    Index 0 Param 4=0.2
    Module.Param0WriteEnabled=true
    Module.Param1WriteEnabled=false
    Module.Param2WriteEnabled=false
    Module.Param3WriteEnabled=false
   ) {module=DynamicMatParams}
   &ScaleColor(
    Scale Alpha=&Multiply_Float(
     A=1
     B=&FloatFromCurve(FloatCurve=DataInterface() {propertyValues {Curve {keys [
          {interpMode=RCIM_Cubic, tangentMode=RCTM_User, time=0, value=0}
          {interpMode=RCIM_Cubic, tangentMode=RCTM_User, time=0.2, value=1}
          {interpMode=RCIM_Cubic, tangentMode=RCTM_User, time=0.5, value=1}
          {tangentMode=RCTM_User, time=1, value=0}
         ]}}})
    )
    Scale RGB=&VectorFromFloat(Value=&FNTimeOfDayScalar(Night Scale=0.6, Morning Scale=0.8, Evening Scale=0.8))
    ScaleA=true
    ScaleRGB=true
   )
  ]
  rendererClasses [&NiagaraSpriteRendererProperties]
  renderers [{rendererClass=&NiagaraSpriteRendererProperties}]
 }
 SmokeWhisps {
  emitterUpdateScript [
   &EmitterState(Life Cycle Mode=::NewEnumerator1, Loop Behavior=::NewEnumerator1, Loop Delay=0.045, Loop Duration=0.21, UseLoopDelay=true)
   &SpawnBurst_Instantaneous(Spawn Count=5, Use Spawn Probability=false, UseLoopCountLimit=false)
   &SpawnRate(SpawnRate=8, Use Spawn Probability=false)
  ]
  particleSpawnScript [
   &InitializeParticle(
    Color Maximum=<rgba 1 0.4531 0.9055 1>
    Color Minimum=<rgba 0.294 0.2185 0.6 1>
    Color Mode=::NewEnumerator2
    Lifetime Max=1.3
    Lifetime Min=0.9
    Lifetime Mode=::NewEnumerator1
    Mass Mode=::NewEnumerator3
    Sprite Rotation Mode=::NewEnumerator2
    Sprite Size Mode=::NewEnumerator1
    Uniform Sprite Size Max=72
    Uniform Sprite Size Min=55
    UsePositionOffset=false
    Write Material Random=true
   )
   &ShapeLocation(MaskSpawnGroup=false, Offset=<0 -15 0>, Offset Mode=::NewEnumerator0, Override Seed=false, Sphere Radius=9, UseOwnerScale=true)
   &AddVelocity(Velocity=&RandomRangeVector(Minimum=<-10 -10 15>, Maximum=<10 10 50>)) {enabled=false}
   &InheritVelocity(Inherited Velocity Speed Limit=1000, Speed Limit=true, UseSpeedThreshold=false)
  ]
  particleUpdateScript [
   &SkeletalMeshLocation(
    "Inherit Velocity (Scale)":NiagaraFloat=&Lerp_Float(
     A=0
     B=-0.3
     Alpha=&FloatFromCurve(FloatCurve=DataInterface() {propertyValues {Curve {keys [{interpMode=RCIM_Cubic, tangentMode=RCTM_User, time=0.1, value=0}, {interpMode=RCIM_Cubic, tangentMode=RCTM_User, time=0.8, value=1}]}}})
    )
    MaskSpawnGroup=false
    Mesh Sampling Type=::NewEnumerator5
    Sampled Position Offset=&MakeVector(
     X=0
     Y=0
     Z=&Lerp_Float(
      A=-5
      B=&RandomRangeFloat(Minimum=20, Maximum=30)
      Alpha=&FloatFromCurve(FloatCurve=DataInterface() {propertyValues {Curve {keys [
           {interpMode=RCIM_Cubic, tangentMode=RCTM_User, time=0, value=0}
           {arriveTangent=1.4599, interpMode=RCIM_Cubic, leaveTangent=1.4599, tangentMode=RCTM_User, time=0.9, value=1}
          ]}}})
     )
    )
    Skeletal Mesh=$System.Skeletal Mesh
    Socket Sampling Mode=::NewEnumerator5
    Velocity Sampling=::NewEnumerator0
    VelocitySpeedLimit=false
   )
   &ParticleState()
   &ScaleSpriteSize(
    Scale Sprite Size Mode=::NewEnumerator0
    Uniform Scale Factor=&Lerp_Float(
     A=0.2
     B=1
     Alpha=&FloatFromCurve(FloatCurve=DataInterface() {propertyValues {Curve {keys [
          {arriveTangent=0.0489, interpMode=RCIM_Cubic, leaveTangent=0.0489, tangentMode=RCTM_User, time=0, value=0}
          {interpMode=RCIM_Cubic, tangentMode=RCTM_User, time=0.6, value=1}
         ]}}})
    )
   )
   &ScaleColor(
    Scale Alpha=&FloatFromCurve(FloatCurve=DataInterface() {propertyValues {Curve {keys [<at 0 0 RCIM_Cubic>,<at 0.2 1 RCIM_Cubic>,<at 1 0 RCIM_Cubic>]}}}, Scale Curve=7)
    Scale RGB=&MakeVectorFromLinearColorRGB(Color=&ColorFromCurve(ColorCurve=DataInterface() {propertyValues {
        AlphaCurve {keys [<at 0 1 RCIM_Cubic>]}
        BlueCurve {keys [
          <at 0.0278 2>
          {arriveTangent=-2.463, interpMode=RCIM_Cubic, leaveTangent=-2.463, time=0.1695, value=0.7}
          <at 0.4893 0.1>
          <at 0.8398 0 RCIM_Cubic>
         ]}
        GreenCurve {keys [
          <at 0.0278 0.4>
          {arriveTangent=-0.4926, interpMode=RCIM_Cubic, leaveTangent=-0.4926, time=0.1695, value=0.161}
          <at 0.4893 0.03>
          <at 0.8398 0 RCIM_Cubic>
         ]}
        RedCurve {keys [
          <at 0.0278 1.5053>
          {arriveTangent=-1.8538, interpMode=RCIM_Cubic, leaveTangent=-1.8538, time=0.1695, value=0.4074}
          <at 0.4893 0.033>
          <at 0.8398 0 RCIM_Cubic>
         ]}
       }}))
    ScaleA=true
    ScaleRGB=true
   )
   &SolveForcesAndVelocity(Clamp Velocity=false, Limit Acceleration=false)
   &SubUVAnimation(
    End Frame Range Override=32
    "Sprite Renderer "=DataInterface() {propertyValues {SpriteRenderer=&self:SmokeWhisps_0.NiagaraSpriteRendererProperties_0}}
    UseEndFrame=true
    UseStartFrame=true
   )
   &DynamicMaterialParameters(
    Index 0 Param 1=&FloatFromCurve(FloatCurve=DataInterface() {propertyValues {Curve {keys [
         {interpMode=RCIM_Cubic, tangentMode=RCTM_User, time=0, value=0}
         {arriveTangent=1.2659, interpMode=RCIM_Cubic, leaveTangent=1.2659, tangentMode=RCTM_User, time=1, value=1}
        ]}}})
    Index 0 Param 2=40
    Module.Index0Param0Write=true
    Module.Index0Param1Write=true
    Module.Index0Param2Write=false
    Module.Index0Param3Write=false
   )
   &CameraOffset(Camera Offset Amount=10)
  ]
  rendererClasses [&NiagaraSpriteRendererProperties]
  renderers [{rendererClass=&NiagaraSpriteRendererProperties}]
 }
 Void_VortexSpawn {
  emitterUpdateScript [
   &EmitterState(Life Cycle Mode=::NewEnumerator1, Loop Behavior=::NewEnumerator1, Loop Delay=0.1, Loop Duration=2, UseLoopDelay=true)
   &SpawnBurst_Instantaneous(Use Spawn Probability=false, UseLoopCountLimit=false)
  ]
  particleSpawnScript [
   &InitializeParticle(
    Color=<rgba 2.6667 0 8 1>
    Lifetime=0.36
    Sprite Rotation Mode=::NewEnumerator2
    Sprite Size Mode=::NewEnumerator3
    Uniform Sprite Size=90
    UsePositionOffset=false
    Write Material Random=true
   )
   &CameraOffset(Camera Offset Amount=16)
  ]
  particleUpdateScript [
   &ParticleState()
   &SpriteRotationRate(Rotation Rate=-200)
   &ScaleSpriteSize(Uniform Curve Sprite Scale=DataInterface() {propertyValues {Curve {keys [
        {arriveTangent=-0.9951, interpMode=RCIM_Cubic, leaveTangent=-0.9951, tangentMode=RCTM_User, time=0, value=1}
        {arriveTangent=-0.1359, interpMode=RCIM_Cubic, leaveTangent=-0.1359, tangentMode=RCTM_User, time=1, value=0.5}
       ]}}})
   &SolveForcesAndVelocity(Clamp Velocity=false, Limit Acceleration=false)
   &DynamicMaterialParameters(
    Index 0 Param 1=&Lerp_Float(
     A=-1
     B=0.6
     Alpha=&FloatFromCurve(FloatCurve=DataInterface() {propertyValues {Curve {keys [
          {arriveTangent=3.1331, interpMode=RCIM_Cubic, leaveTangent=3.1331, tangentMode=RCTM_User, time=0, value=0}
          {arriveTangent=2.5519, interpMode=RCIM_Cubic, leaveTangent=2.5519, tangentMode=RCTM_User, time=1, value=1}
         ]}}})
    )
    Index 0 Param 2=1.3
    Index 0 Param 3=&Lerp_Float(
     A=-0.25
     B=0.15
     Alpha=&FloatFromCurve(FloatCurve=DataInterface() {propertyValues {Curve {keys [<at 0 0>,<at 1 1>]}}})
    )
    Index 0 Param 4=0.1
    Module.Index0Param0Write=true
    Module.Index0Param1Write=true
    Module.Index0Param2Write=true
    Module.Index0Param3Write=true
   )
   &ScaleColor(
    Scale Alpha=&FloatFromCurve(
     FloatCurve=DataInterface() {propertyValues {Curve {keys [
         <at 0 0 RCIM_Cubic>
         {arriveTangent=-1.0608, interpMode=RCIM_Cubic, leaveTangent=-1.0608, tangentMode=RCTM_User, time=0.1, value=1}
         {arriveTangent=-0.924, interpMode=RCIM_Cubic, leaveTangent=-0.924, tangentMode=RCTM_User, time=1, value=0}
        ]}}}
     Scale Curve=1.3
    )
    Scale RGB=&MakeVectorFromLinearColorRGB(Color=&ColorFromCurve(ColorCurve=DataInterface() {propertyValues {
        AlphaCurve {keys [<at 0 1 RCIM_Cubic>]}
        BlueCurve {keys [<at 0.0316 1>,<at 0.1853 2 RCIM_Cubic>,<at 0.4824 1>,<at 0.9383 1>]}
        GreenCurve {keys [<at 0.0316 1>, {arriveTangent=-1.8367, interpMode=RCIM_Cubic, time=0.1853, value=0.754}, <at 0.4824 0.114>, <at 0.9383 0.1332>]}
        RedCurve {keys [<at 0.0316 1>, {arriveTangent=-0.8565, interpMode=RCIM_Cubic, time=0.1853, value=1.8039}, <at 0.4824 0.5869>, <at 0.9383 0.2719>]}
       }}))
    ScaleA=true
    ScaleRGB=true
   )
   /SetVariables(
    Particles.ColorDark:LinearColor=&MultiplyLinearColor(
     Linear Color=<rgba 0 0 0 1>
     Scale Factor=&MakeFloat4FromVectorAndFloat(
      "Vector (RGB)"=<1 1 1>
      "Float (Alpha)"=&FloatFromCurve(
       FloatCurve=DataInterface() {propertyValues {Curve {keys [
           <at 0 0 RCIM_Cubic>
           {arriveTangent=-2.1159, interpMode=RCIM_Cubic, leaveTangent=-2.1159, tangentMode=RCTM_User, time=0.1, value=1}
           {interpMode=RCIM_Cubic, tangentMode=RCTM_User, time=1, value=0}
          ]}}}
       Scale Curve=3
      )
     )
    )
    Particles.DynamicMaterialParameterDark:Vector4f=&MakeVector4(
     X=&Lerp_Float(
      A=-0.1
      B=0.2
      Alpha=&FloatFromCurve(FloatCurve=DataInterface() {propertyValues {Curve {keys [
           {interpMode=RCIM_Cubic, tangentMode=RCTM_User, time=0.4, value=0}
           {arriveTangent=2.5519, interpMode=RCIM_Cubic, leaveTangent=2.5519, tangentMode=RCTM_User, time=1, value=1}
          ]}}})
     )
     Y=1
     Z=&Lerp_Float(
      A=-0.5
      B=0.2
      Alpha=&FloatFromCurve(FloatCurve=DataInterface() {propertyValues {Curve {keys [<at 0 0>,<at 1 1>]}}})
     )
     W=0.4
    )
    Particles.SpriteSizeDark:Vector2f=&Multiply_Vector2D_ByFloat(Float=1.1, Vector2D=$Particles.SpriteSize)
   ) {isSetParameters=true}
   /SetVariables(
    Particles.ColorFlare:LinearColor=&ColorFromCurve(ColorCurve=DataInterface() {propertyValues {
       AlphaCurve {keys [<at 0 0 RCIM_Cubic>,<at 0.1365 1>,<at 0.5368 0 RCIM_Cubic>]}
       BlueCurve {keys [<at 0 1.6 RCIM_Cubic>,<at 0.6419 0.1 RCIM_Cubic>]}
       GreenCurve {keys [<at 0 0.096 RCIM_Cubic>,<at 0.6419 0 RCIM_Cubic>]}
       RedCurve {keys [<at 0 0.5973 RCIM_Cubic>,<at 0.6419 0 RCIM_Cubic>]}
      }})
    Particles.SpriteSizeFlare:Vector2f=&MakeVector2D(
     X=&FloatFromCurve(FloatCurve=DataInterface() {propertyValues {Curve {keys [<at 0 0.7>,<at 0.7 1>]}}}, Scale Curve=220)
     Y=&FloatFromCurve(FloatCurve=DataInterface() {propertyValues {Curve {keys [<at 0 1>,<at 0.8 0.2>]}}}, Scale Curve=145)
    )
   ) {isSetParameters=true}
  ]
  rendererClasses [&NiagaraSpriteRendererProperties]
  renderers [
   {rendererClass=&NiagaraSpriteRendererProperties}
   {rendererClass=&NiagaraSpriteRendererProperties, rendererIndex=1}
   {rendererClass=&NiagaraSpriteRendererProperties, rendererIndex=2}
  ]
 }
}
summary {
 systemName=NS_VoidMask_Equip
 emitters [
  {emitterName=Smoke, rendererClasses [&NiagaraSpriteRendererProperties]}
  {emitterName=Main_Flame, rendererClasses [&NiagaraMeshRendererProperties, &NiagaraSpriteRendererProperties]}
  {emitterName=EyeTrail, rendererClasses [&NiagaraMeshRendererProperties]}
  {emitterName=Bits_Up_, rendererClasses [&NiagaraSpriteRendererProperties]}
  {emitterName=SmokeWhisps, rendererClasses [&NiagaraSpriteRendererProperties]}
  {emitterName=ChromaticRefraction, rendererClasses [&NiagaraSpriteRendererProperties]}
  {emitterName=Void_VortexSpawn, rendererClasses [&NiagaraSpriteRendererProperties]}
 ]
 userVariables [
  User.AbilityOnCooldown {type=NiagaraBool}
  User.Global Scale {type=NiagaraFloat, defaultValue=1, description=""}
  User.Offset Pos {type=NiagaraPosition, description="", defaultValue {}}
 ]
}
