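# Niagara muzzle effect for the Void Mask: a short purple burst that darkens as it fades.
# Parallel burst emitters layer a mesh distortion, a slightly delayed mesh-and-sprite flash, spinning vortex tendrils and a drag-slowed wave.
# User.AimVec orients the tendrils and drives the wave's velocity and both their spawn offsets.
# Curves animate scale, color and dynamic material parameters, with distortion alpha scaled by time of day; Distortion_LW has no renderers.
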NiagaraSystem /DemonMaskVoid/Assets/VFX/NS_VoidMask_Muzzle
dependencies {
 usedDataInterfaces [&NiagaraDataInterfaceArrayFloat, &NiagaraDataInterfaceCurve, &NiagaraDataInterfaceColorCurve]
 usedDynamicInputs []
 usedModules [&SystemState]
}
emitters {
 Distortion_LW {rendererClasses [], renderers []}
 Distrotion {
  emitterUpdateScript [
   &EmitterState(Life Cycle Mode=::NewEnumerator1, Loop Behavior=::NewEnumerator1, Loop Delay=0.1, Loop Duration=0.1, UseLoopDelay=false)
   &SpawnBurst_Instantaneous(Spawn Group=1, Use Spawn Probability=false, UseLoopCountLimit=false) {module=SpawnBurst_Instantaneous001}
  ]
  particleSpawnScript [
   /SetVariables(Particles.VisibilityTag=$Emitter.SpawnGroup) {enabled=false, isSetParameters=true}
   &InitializeParticle(
    Color=<rgba 3.1037 0 5 1>
    Lifetime=&SelectFloatFromArray(Direct Array Index=$Emitter.SpawnGroup, Float Selection Array=DataInterface() {propertyValues {FloatData [0.5,0.2,0.3]}})
    Mesh Scale Mode=::NewEnumerator3
    UsePositionOffset=false
    Write Material Random=true
   )
   &InitialMeshOrientation(
    Rotation=&RandomRangeVector(Minimum=<0 0 0>, Maximum=<1 1 1>)
    Use Orientation Vector=true
    Use Rotation Vector=true
   )
   /SetVariables(Particles.Color2=<rgba 50 37.4589 6.8714 2>) {isSetParameters=true}
  ]
  particleUpdateScript [
   &ParticleState()
   &ScaleSpriteSize(Uniform Curve Sprite Scale=DataInterface() {propertyValues {Curve {keys [
        {arriveTangent=4.1785, interpMode=RCIM_Cubic, leaveTangent=4.1785, tangentMode=RCTM_User, time=0, value=0.1}
        {arriveTangent=1.1179, interpMode=RCIM_Cubic, leaveTangent=1.1179, tangentMode=RCTM_User, time=0.2042, value=0.6228}
        {arriveTangent=0.0224, interpMode=RCIM_Cubic, leaveTangent=0.0224, tangentMode=RCTM_User, time=1, value=1}
       ]}}})
   &ScaleColor(
    Scale Alpha=&FloatFromCurve(
     FloatCurve=DataInterface() {propertyValues {Curve {keys [{arriveTangent=45, interpMode=RCIM_Cubic, time=0, value=0}, <at 0.3 1.75 RCIM_Cubic>, <at 1 0 RCIM_Cubic>]}}}
     Scale Curve=&FNTimeOfDayScalar(Night Scale=0.1, Morning Scale=0.2, Day Scale=0.25, Evening Scale=0.15)
    )
    Scale RGB=<3 3 3>
    ScaleA=true
    ScaleRGB=true
   )
   &ScaleMeshSize(Scale Factor=&VectorFromFloat(Value=&FloatFromCurve(FloatCurve=DataInterface() {propertyValues {Curve {keys [
          {arriveTangent=6.5452, interpMode=RCIM_Cubic, leaveTangent=6.5452, tangentMode=RCTM_User, time=0, value=0.05}
          {arriveTangent=6.229, interpMode=RCIM_Cubic, leaveTangent=6.229, tangentMode=RCTM_User, time=0.1, value=0.75}
          {arriveTangent=2.2956, interpMode=RCIM_Cubic, leaveTangent=2.2956, tangentMode=RCTM_User, time=0.1698, value=0.1791}
          {arriveTangent=-0.0405, interpMode=RCIM_Cubic, leaveTangent=-0.0405, tangentMode=RCTM_User, time=0.6, value=0.7688}
         ]}}})))
   &DynamicMaterialParameters(
    Index 0 Param 1=&FloatFromCurve(FloatCurve=DataInterface() {propertyValues {Curve {keys [
         {arriveTangent=-0.3119, interpMode=RCIM_Cubic, leaveTangent=-0.3119, tangentMode=RCTM_User, time=0, value=0.9375}
         {interpMode=RCIM_Cubic, tangentMode=RCTM_User, time=0.3, value=0.875}
         {arriveTangent=0.5138, interpMode=RCIM_Cubic, leaveTangent=0.5138, tangentMode=RCTM_User, time=0.5, value=0.9125}
         {interpMode=RCIM_Cubic, tangentMode=RCTM_User, time=1, value=1}
        ]}}})
    Index 0 Param 2=&FloatFromCurve(FloatCurve=DataInterface() {propertyValues {Curve {keys [<at 0 15 RCIM_Cubic>, {arriveTangent=-40.0295, interpMode=RCIM_Cubic, leaveTangent=-40.0295, tangentMode=RCTM_User, time=1, value=0}]}}})
    Module.Param0WriteEnabled=true
    Module.Param1WriteEnabled=true
    Module.Param2WriteEnabled=true
    Module.Param3WriteEnabled=true
   )
   /SetVariables(
    Particles.Color1:LinearColor=&ColorFromCurve(ColorCurve=DataInterface() {propertyValues {
       AlphaCurve {keys [<at 0.4948 0.5>,<at 1 0 RCIM_Cubic>]}
       BlueCurve {keys [<at 0.0735 1 RCIM_Cubic>,<at 0.4828 0.0365 RCIM_Cubic>]}
       GreenCurve {keys [<at 0.0735 0 RCIM_Cubic>,<at 0.4828 0 RCIM_Cubic>]}
       RedCurve {keys [<at 0.0735 0.5 RCIM_Cubic>,<at 0.4828 0.0164 RCIM_Cubic>]}
      }})
    Particles.DMP0A:Vector4f=&MakeVector4(
     X=1
     Y=&FloatFromCurve(FloatCurve=DataInterface() {propertyValues {Curve {keys [{interpMode=RCIM_Cubic, tangentMode=RCTM_User, time=0, value=0.25}, {interpMode=RCIM_Cubic, tangentMode=RCTM_User, time=1, value=-1}]}}})
     Z=0.5
     W=50
    )
    Particles.DMP2A:Vector4f=<1 1 1 1>
    Particles.MeshSize1:Vector3f=&Multiply_VectorByFloat(
     Vector=<9 9 9>
     Float=&FloatFromCurve(
      FloatCurve=DataInterface() {propertyValues {Curve {keys [
          {arriveTangent=5.7182, interpMode=RCIM_Cubic, leaveTangent=5.7182, tangentMode=RCTM_User, time=0, value=0.1}
          {arriveTangent=1.9532, interpMode=RCIM_Cubic, leaveTangent=1.9532, tangentMode=RCTM_User, time=0.2, value=0.9}
          {arriveTangent=-0.0405, interpMode=RCIM_Cubic, leaveTangent=-0.0405, tangentMode=RCTM_User, time=1, value=1.25}
         ]}}}
      Scale Curve=0.3
     )
    )
   ) {isSetParameters=true}
  ]
  rendererClasses [&NiagaraMeshRendererProperties]
  renderers [{rendererClass=&NiagaraMeshRendererProperties, rendererIndex=0}]
 }
 Flash {
  emitterSpawnScript [/SetVariables(Emitter.CameraOffset=50, Emitter.SpriteSize:Vector2f=&Multiply_Vector2D_ByFloat(Float=1, Vector2D=<100 90>)) {enabled=false, isSetParameters=true}]
  emitterUpdateScript [
   &EmitterState(Life Cycle Mode=::NewEnumerator1, Loop Behavior=::NewEnumerator1, Loop Delay=0.06, Loop Duration=0.15, UseLoopDelay=true)
   &SpawnBurst_Instantaneous(Spawn Count=1, Use Spawn Probability=false)
   /SetVariables(
    Emitter.Color:LinearColor=<rgba 0.197 0 0.5208 1>
    Emitter.SpriteSize:Vector2f=&Multiply_Vector2D_ByFloat(
     Float=&FloatFromCurve(
      FloatCurve=DataInterface() {propertyValues {Curve {keys [
          {arriveTangent=9.901, interpMode=RCIM_Cubic, leaveTangent=9.901, tangentMode=RCTM_User, time=0, value=1}
          {arriveTangent=12.2238, interpMode=RCIM_Cubic, leaveTangent=12.2238, tangentMode=RCTM_User, time=0.1, value=3.5}
          <at 0.3 0>
         ]}}}
      CurveIndex=$Emitter.Age
     )
     Vector2D=$Emitter.Initial.SpriteSize
    )
   ) {enabled=false, isSetParameters=true}
  ]
  particleSpawnScript [
   /Niagara/Modules/Spawn/Initialization/InitializeParticle(
    Color=<rgba 1.9406 1.8679 2 0.3>
    Lifetime=0.3
    Mesh Scale=&VectorFromFloat(Value=&Lerp_Float(A=1, B=0.9, Alpha=&MakeFloatFromInt(Int=&SpawnGroup())))
    Module.Write Color=true
    Module.Write Lifetime=true
    Module.Write Mass=false
    Module.Write Material Random=true
    Module.Write Position=false
    Module.Write Scale=true
    Module.Write SpriteRotation=false
    Module.Write SpriteSize=false
    Position=&SimulationPosition()
    Sprite Size=<5 5>
   ) {enabled=false}
   /Niagara/Modules/Spawn/Initialization/V2/InitializeParticle(
    Color=<rgba 0.4167 0 1 1>
    Lifetime=0.26
    Mesh Scale=&Multiply_VectorByFloat(Vector=&VectorFromFloat(Value=0.25), Float=1)
    Mesh Scale Mode=::NewEnumerator4
    Position Offset=<40 0 0>
    Sprite Size Mode=::NewEnumerator3
    Uniform Sprite Size=60
    UsePositionOffset=false
    Write Material Random=true
   ) {module=InitializeParticle001}
   &CameraOffset(Camera Offset Amount=-20)
   /SetVariables(Particles.Position=&AddVectorToPosition(Position=<0 0 0>, Vector=&Multiply_VectorByFloat(Vector=$User.AimVec, Float=28))) {enabled=false, isSetParameters=true}
  ]
  particleUpdateScript [
   &SolveForcesAndVelocity(Clamp Velocity=false, Limit Acceleration=false)
   &ParticleState()
   &ScaleColor(
    Scale Alpha=&FloatFromCurve(
     FloatCurve=DataInterface() {propertyValues {Curve {keys [
         {arriveTangent=-1.8753, interpMode=RCIM_Cubic, leaveTangent=-1.8753, tangentMode=RCTM_User, time=0.25, value=1}
         {interpMode=RCIM_Cubic, tangentMode=RCTM_User, time=1, value=0}
        ]}}}
     Scale Curve=30
    )
    Scale RGB=&VectorFromFloat(Value=10)
    ScaleA=true
    ScaleRGB=true
   )
   &ScaleMeshSize(Scale Factor=&VectorFromFloat(Value=&FloatFromCurve(
      FloatCurve=DataInterface() {propertyValues {Curve {keys [
          {arriveTangent=4.1785, interpMode=RCIM_Cubic, leaveTangent=4.1785, tangentMode=RCTM_User, time=0.0247, value=0}
          <at 0.4247 0.95>
          {arriveTangent=-0.0405, interpMode=RCIM_Cubic, leaveTangent=-0.0405, tangentMode=RCTM_User, time=1, value=0.05}
         ]}}}
      Scale Curve=1.6
     )))
   /SetVariables(Particles.Color2=&ColorFromCurve(ColorCurve=DataInterface() {propertyValues {
       AlphaCurve {keys [<at 0 1 RCIM_Cubic>,<at 1 0 RCIM_Cubic>]}
       BlueCurve {keys [<at 0 1 RCIM_Cubic>]}
       GreenCurve {keys [<at 0 0.7725 RCIM_Cubic>]}
       RedCurve {keys [<at 0 0.9467 RCIM_Cubic>]}
      }})) {isSetParameters=true}
   /SetVariables(Particles.SpriteSizeFlare:Vector2f=&Multiply_Vector2D_ByFloat(
     Float=&FloatFromCurve(FloatCurve=DataInterface() {propertyValues {Curve {keys [
          {interpMode=RCIM_Cubic, tangentMode=RCTM_User, time=0, value=0.2}
          {arriveTangent=-2.2468, interpMode=RCIM_Cubic, leaveTangent=-2.2468, tangentMode=RCTM_User, time=0.3, value=1}
          {interpMode=RCIM_Cubic, tangentMode=RCTM_User, time=1, value=0}
         ]}}})
     Vector2D=<100 100>
    )) {isSetParameters=true}
   /SetVariables(Particles.CameraOffsetFlare=45, Particles.ColorFlare:LinearColor=<rgba 2.1333 0 4 1>) {isSetParameters=true}
   /SetVariables(
    Particles.ColorDark:LinearColor=&ColorFromCurve(ColorCurve=DataInterface() {propertyValues {
       AlphaCurve {keys [<at 0 0>,<at 0.1665 3 RCIM_Cubic>,<at 1 0 RCIM_Cubic>]}
       BlueCurve {keys [<at 0 0.01 RCIM_Cubic>]}
       GreenCurve {keys [<at 0 0.007 RCIM_Cubic>]}
       RedCurve {keys [<at 0 0.008 RCIM_Cubic>]}
      }})
    Particles.SpriteSizeDark:Vector2f=&Vector2DFromFloat(Value=&Lerp_Float(
      A=0
      B=200
      Alpha=&FloatFromCurve(FloatCurve=DataInterface() {propertyValues {Curve {keys [
           {interpMode=RCIM_Cubic, tangentMode=RCTM_User, time=0, value=0}
           {interpMode=RCIM_Cubic, tangentMode=RCTM_User, time=0.1, value=1}
           {interpMode=RCIM_Cubic, tangentMode=RCTM_User, time=0.7, value=0}
          ]}}})
     ))
   ) {enabled=false, isSetParameters=true}
  ]
  rendererClasses [&NiagaraMeshRendererProperties, &NiagaraSpriteRendererProperties]
  renderers [
   {rendererClass=&NiagaraMeshRendererProperties, rendererIndex=0}
   {rendererClass=&NiagaraSpriteRendererProperties, rendererIndex=1}
   {rendererClass=&NiagaraSpriteRendererProperties, rendererIndex=2}
   {rendererClass=&NiagaraSpriteRendererProperties, rendererIndex=3}
  ]
 }
 Tendrils {
  emitterUpdateScript [
   &EmitterState(Life Cycle Mode=::NewEnumerator1, Loop Behavior=::NewEnumerator1, Loop Duration=2, UseLoopDelay=true)
   &SpawnBurst_Instantaneous(Use Spawn Probability=false, UseLoopCountLimit=false)
  ]
  particleSpawnScript [
   &InitializeParticle(
    Lifetime=1.3
    Mass Mode=::NewEnumerator3
    Mesh Scale=&Multiply_VectorByFloat(Vector=<0.6 2 2>, Float=1.3)
    Mesh Scale Mode=::NewEnumerator4
    Sprite Rotation Mode=::NewEnumerator2
    Sprite Size Mode=::NewEnumerator1
    Uniform Sprite Size Max=12
    Uniform Sprite Size Min=6
    UsePositionOffset=false
    Write Material Random=true
   )
   &InitialMeshOrientation(
    Mesh Axis To Orient=<-1 0 0>
    Mesh Orientation Mode=::NewEnumerator2
    Orientation Vector=$User.AimVec
    Rotation=&MakeVector(X=&RandomRangeFloat(Minimum=-1, Maximum=1), Y=0, Z=0)
    Use Rotation Vector=true
   )
   &SystemLocation(
    MaskSpawnGroup=false
    Offset=&AddVectorToPosition(Position=&Multiply_VectorByFloat(Vector=$User.AimVec, Float=-10), Vector=<0 0 25>)
   )
   /SetVariables(Particles.MeshOrientationVortex:Quat4f=$Particles.MeshOrientation) {isSetParameters=true}
  ]
  particleUpdateScript [
   &ParticleState()
   &ScaleMeshSize(Scale Factor=&MakeVector(
     X=&FloatFromCurve(FloatCurve=DataInterface() {propertyValues {Curve {keys [
          {arriveTangent=4.5334, interpMode=RCIM_Cubic, leaveTangent=4.5334, tangentMode=RCTM_User, time=0, value=0.2}
          {interpMode=RCIM_Cubic, tangentMode=RCTM_User, time=0.4, value=1}
         ]}}})
     Y=&FloatFromCurve(FloatCurve=DataInterface() {propertyValues {Curve {keys [{interpMode=RCIM_Cubic, tangentMode=RCTM_User, time=0, value=1}, {interpMode=RCIM_Cubic, tangentMode=RCTM_User, time=1, value=0.25}]}}})
     Z=&FloatFromCurve(FloatCurve=DataInterface() {propertyValues {Curve {keys [{interpMode=RCIM_Cubic, tangentMode=RCTM_User, time=0, value=1}, {interpMode=RCIM_Cubic, tangentMode=RCTM_User, time=1, value=0.25}]}}})
    )) {module=ScaleMeshSize001}
   &UpdateMeshOrientation(
    Rotation Rate=&Lerp_Float(
     A=0.7
     B=0.2
     Alpha=&FloatFromCurve(FloatCurve=DataInterface() {propertyValues {Curve {keys [
          {arriveTangent=1.7929, interpMode=RCIM_Cubic, leaveTangent=1.7929, tangentMode=RCTM_User, time=0, value=0}
          {interpMode=RCIM_Cubic, tangentMode=RCTM_User, time=1, value=1}
         ]}}})
    )
    Rotation Vector=<-1 0 0>
   )
   &ScaleColor(
    Scale Alpha=&FloatFromCurve(FloatCurve=DataInterface() {propertyValues {Curve {keys [
         {arriveTangent=10.1427, interpMode=RCIM_Cubic, leaveTangent=10.1427, tangentMode=RCTM_User, time=0, value=0}
         {arriveTangent=-0.3382, interpMode=RCIM_Cubic, leaveTangent=-0.3382, tangentMode=RCTM_User, time=0.1, value=1}
         {arriveTangent=-3.4639, interpMode=RCIM_Cubic, leaveTangent=-3.4639, tangentMode=RCTM_User, time=1, value=0}
        ]}}})
    Scale RGB=&MakeVectorFromLinearColorRGB(Color=&ColorFromCurve(ColorCurve=DataInterface() {propertyValues {
        AlphaCurve {keys [<at 0 1 RCIM_Cubic>]}
        BlueCurve {keys [<at 0 0.5 RCIM_Cubic>,<at 0.9918 0.04>]}
        GreenCurve {keys [<at 0 0.01 RCIM_Cubic>,<at 0.9918 0.006>]}
        RedCurve {keys [<at 0 0.1325 RCIM_Cubic>,<at 0.9918 0.0088>]}
       }}))
    ScaleA=true
    ScaleRGB=true
   )
   /SetVariables(Particles.ColorDark:LinearColor=&MakeLinearColorFromVectorAndFloat(
     "Vector (RGB)"=&MakeVectorFromLinearColorRGB(Color=&ColorFromCurve(ColorCurve=DataInterface() {propertyValues {
         AlphaCurve {keys [<at 0 1 RCIM_Cubic>]}
         BlueCurve {keys [<at 0 0.3 RCIM_Cubic>,<at 0.3152 0>]}
         GreenCurve {keys [<at 0 0 RCIM_Cubic>,<at 0.3152 0>]}
         RedCurve {keys [<at 0 0.05 RCIM_Cubic>,<at 0.3152 0>]}
        }}))
     "Float (Alpha)"=&FloatFromCurve(FloatCurve=DataInterface() {propertyValues {Curve {keys [
          {interpMode=RCIM_Cubic, tangentMode=RCTM_User, time=0.1, value=0}
          {interpMode=RCIM_Cubic, tangentMode=RCTM_User, time=0.2, value=1}
          {interpMode=RCIM_Cubic, tangentMode=RCTM_User, time=1, value=0}
         ]}}})
    )) {isSetParameters=true}
   &SolveForcesAndVelocity(Clamp Velocity=false, Limit Acceleration=false)
   &DynamicMaterialParameters(
    Index 0 Param 1=0.1
    Index 0 Param 2=&Lerp_Float(
     A=0.3
     B=2.5
     Alpha=&FloatFromCurve(FloatCurve=DataInterface() {propertyValues {Curve {keys [<at 0 0>,<at 1 1>]}}})
    )
    Index 0 Param 3=&Lerp_Float(
     A=0.5
     B=2
     Alpha=&FloatFromCurve(FloatCurve=DataInterface() {propertyValues {Curve {keys [<at 0 0>,<at 1 1>]}}})
    )
    Index 0 Param 4=&Lerp_Float(
     A=0.2
     B=1
     Alpha=&FloatFromCurve(FloatCurve=DataInterface() {propertyValues {Curve {keys [
          {arriveTangent=0.2526, interpMode=RCIM_Cubic, leaveTangent=0.2526, tangentMode=RCTM_User, time=0, value=0}
          {arriveTangent=1.7785, interpMode=RCIM_Cubic, leaveTangent=1.7785, tangentMode=RCTM_User, time=1, value=1}
         ]}}})
    )
    Module.Index0Param0Write=true
    Module.Index0Param1Write=true
    Module.Index0Param2Write=true
    Module.Index0Param3Write=true
   )
   /SetVariables(Particles.DynamicMaterialParameterDark:Vector4f=&MakeVector4(
     X=1
     Y=&Lerp_Float(
      A=0.2
      B=2
      Alpha=&FloatFromCurve(FloatCurve=DataInterface() {propertyValues {Curve {keys [<at 0 0>,<at 1 1>]}}})
     )
     Z=&Lerp_Float(
      A=0.5
      B=1
      Alpha=&FloatFromCurve(FloatCurve=DataInterface() {propertyValues {Curve {keys [<at 0 0>,<at 0.7 1>]}}})
     )
     W=&Lerp_Float(
      A=0
      B=1
      Alpha=&FloatFromCurve(FloatCurve=DataInterface() {propertyValues {Curve {keys [
           {arriveTangent=0.2526, interpMode=RCIM_Cubic, leaveTangent=0.2526, tangentMode=RCTM_User, time=0.2, value=0}
           {arriveTangent=1.7785, interpMode=RCIM_Cubic, leaveTangent=1.7785, tangentMode=RCTM_User, time=1, value=1}
          ]}}})
     )
    )) {isSetParameters=true}
   /SetVariables(Particles.DynamicMaterialParameterVortex:Vector4f=&MakeVector4(
     X=&Lerp_Float(
      A=0.4
      B=-2
      Alpha=&FloatFromCurve(FloatCurve=DataInterface() {propertyValues {Curve {keys [
           {arriveTangent=5.2247, interpMode=RCIM_Cubic, leaveTangent=5.2247, tangentMode=RCTM_User, time=0, value=0}
           {interpMode=RCIM_Cubic, tangentMode=RCTM_User, time=0.4, value=1}
          ]}}})
     )
     Y=&Lerp_Float(
      A=1
      B=4
      Alpha=&FloatFromCurve(FloatCurve=DataInterface() {propertyValues {Curve {keys [<at 0 0>,<at 0.5 1>]}}})
     )
     Z=&Lerp_Float(
      A=0
      B=1
      Alpha=&FloatFromCurve(FloatCurve=DataInterface() {propertyValues {Curve {keys [<at 0 0>,<at 0.3 1>]}}})
     )
     W=-0.1
    )) {isSetParameters=true}
   /SetVariables(Particles.ScaleVortex:Vector3f=&Multiply_Vector(A=<1 1 1>, B=<3 4 4>)) {isSetParameters=true}
   /SetVariables(Particles.ColorVortex:LinearColor=&ColorFromCurve(ColorCurve=DataInterface() {propertyValues {
       AlphaCurve {keys [<at 0 0>, {arriveTangent=1.0121, interpMode=RCIM_Cubic, leaveTangent=1.0121, time=0.046, value=2}, <at 0.988 1>]}
       BlueCurve {keys [<at 0.1136 2 RCIM_Cubic>,<at 0.5005 0.02 RCIM_Cubic>]}
       GreenCurve {keys [<at 0.1136 0 RCIM_Cubic>,<at 0.5005 0.01 RCIM_Cubic>]}
       RedCurve {keys [<at 0.1136 0.5413 RCIM_Cubic>,<at 0.5005 0.0112 RCIM_Cubic>]}
      }})) {isSetParameters=true}
  ]
  rendererClasses [&NiagaraMeshRendererProperties]
  renderers [
   {rendererClass=&NiagaraMeshRendererProperties, rendererIndex=0}
   {rendererClass=&NiagaraMeshRendererProperties, rendererIndex=1}
  ]
 }
 Wave {
  emitterUpdateScript [
   &EmitterState(Life Cycle Mode=::NewEnumerator1, Loop Behavior=::NewEnumerator1, Loop Delay=0.03, Loop Duration=2, UseLoopDelay=true)
   &SpawnBurst_Instantaneous(Use Spawn Probability=false, UseLoopCountLimit=false)
  ]
  particleSpawnScript [
   &InitializeParticle(
    Color=<rgba 0.2567 0 1 1>
    Lifetime=0.35
    Mass Mode=::NewEnumerator3
    Mesh Scale=&Multiply_VectorByFloat(Vector=<0.6 0.6 0.6>, Float=&RandomRangeFloat(Minimum=0.5, Maximum=0.6))
    Mesh Scale Mode=::NewEnumerator4
    Sprite Rotation Mode=::NewEnumerator2
    Sprite Size Mode=::NewEnumerator3
    Uniform Sprite Size=120
    UsePositionOffset=false
    Write Material Random=true
   )
   &AddVelocity(Velocity=&Multiply_VectorByFloat(Vector=$User.AimVec, Float=450))
   &SystemLocation(
    MaskSpawnGroup=false
    Offset=&AddVectorToPosition(Position=&Multiply_VectorByFloat(Vector=$User.AimVec, Float=-10), Vector=<0 0 0>)
   )
   &CameraOffset(Camera Offset Amount=-20)
   &InitialMeshOrientation(
    Mesh Orientation Mode=::NewEnumerator2
    Orientation Vector=<-1 0 0>
    Rotation=&RandomRangeVector(Minimum=<-1 0 0>, Maximum=<1 0 0>)
    Use Rotation Vector=true
   )
  ]
  particleUpdateScript [
   &ParticleState()
   &ScaleMeshSize(Scale Factor=&MakeVector(
     X=&Lerp_Float(
      A=1
      B=2
      Alpha=&FloatFromCurve(FloatCurve=DataInterface() {propertyValues {Curve {keys [<at 0 0>,<at 1 1>]}}})
     )
     Y=&FloatFromCurve(FloatCurve=DataInterface() {propertyValues {Curve {keys [
          {arriveTangent=2.2436, interpMode=RCIM_Cubic, leaveTangent=2.2436, tangentMode=RCTM_User, time=0, value=0.4}
          {interpMode=RCIM_Cubic, tangentMode=RCTM_User, time=0.7, value=1}
         ]}}})
     Z=&FloatFromCurve(FloatCurve=DataInterface() {propertyValues {Curve {keys [
          {arriveTangent=2.2436, interpMode=RCIM_Cubic, leaveTangent=2.2436, tangentMode=RCTM_User, time=0, value=0.4}
          {interpMode=RCIM_Cubic, tangentMode=RCTM_User, time=0.7, value=1}
         ]}}})
    ))
   &Drag(Drag=7, Use Linear Drag=true, Use Rotational Drag=true)
   &SolveForcesAndVelocity(Clamp Velocity=false, Limit Acceleration=false)
   &ScaleColor(
    Scale Alpha=&FloatFromCurve(
     FloatCurve=DataInterface() {propertyValues {Curve {keys [
         {arriveTangent=-2.0124, interpMode=RCIM_Cubic, leaveTangent=-2.0124, tangentMode=RCTM_User, time=0.1, value=1}
         {arriveTangent=-0.4437, interpMode=RCIM_Cubic, leaveTangent=-0.4437, tangentMode=RCTM_User, time=1, value=0}
        ]}}}
     Scale Curve=8
    )
    Scale RGB=&MakeVectorFromLinearColorRGB(Color=&ColorFromCurve(ColorCurve=DataInterface() {propertyValues {
        AlphaCurve {keys [<at 0 1 RCIM_Cubic>]}
        BlueCurve {keys [<at 0.0647 20 RCIM_Cubic>,<at 0.2128 3>,<at 1 0.1 RCIM_Cubic>]}
        GreenCurve {keys [<at 0.0647 0.2 RCIM_Cubic>,<at 0.2128 0.3>,<at 1 0.03 RCIM_Cubic>]}
        RedCurve {keys [<at 0.0647 9.11 RCIM_Cubic>,<at 0.2128 1.425>,<at 1 0.0533 RCIM_Cubic>]}
       }}))
    ScaleA=true
    ScaleRGB=true
   )
   &DynamicMaterialParameters(
    Index 0 Param 1=1
    Index 0 Param 2=0.4
    Index 1 Param 2=1
    Module.Index0Param0Write=true
    Module.Index0Param1Write=true
    Module.Index0Param2Write=false
    Module.Index0Param3Write=false
    Module.Index1Param0Write=false
    Module.Index1Param1Write=true
    Module.Index1Param2Write=false
    Module.Index1Param3Write=false
    Write Parameter Index 1=true
   )
   /SetVariables(Particles.ColorGlow:LinearColor=&ColorFromCurve(ColorCurve=DataInterface() {propertyValues {
       AlphaCurve {keys [<at 0.0009 0>,<at 0.1538 2 RCIM_Cubic>,<at 0.6798 0 RCIM_Cubic>]}
       BlueCurve {keys [<at 0 1 RCIM_Cubic>,<at 0.2739 0.1 RCIM_Cubic>]}
       GreenCurve {keys [<at 0 0 RCIM_Cubic>,<at 0.2739 0.05 RCIM_Cubic>]}
       RedCurve {keys [<at 0 0.4167 RCIM_Cubic>,<at 0.2739 0.0583 RCIM_Cubic>]}
      }})) {isSetParameters=true}
  ]
  rendererClasses [&NiagaraMeshRendererProperties, &NiagaraSpriteRendererProperties]
  renderers [
   {rendererClass=&NiagaraMeshRendererProperties, rendererIndex=0}
   {rendererClass=&NiagaraSpriteRendererProperties, rendererIndex=1}
  ]
 }
}
summary {
 systemName=NS_VoidMask_Muzzle
 emitters [
  {emitterName=Distrotion, rendererClasses [&NiagaraMeshRendererProperties]}
  {emitterName=Flash, rendererClasses [&NiagaraMeshRendererProperties, &NiagaraSpriteRendererProperties]}
  {emitterName=Distortion_LW, rendererClasses []}
  {emitterName=Tendrils, rendererClasses [&NiagaraMeshRendererProperties]}
  {emitterName=Wave, rendererClasses [&NiagaraMeshRendererProperties, &NiagaraSpriteRendererProperties]}
 ]
 userVariables [
  User.AimVec {type=Vector3f, description="", defaultValue {struct=Vector3f, value=<0 1 1>}}
  User.Global Scale {type=NiagaraFloat, defaultValue=1, description=""}
 ]
}
