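# Niagara system for the Demon Mask void projectile: a purple void orb with smoke and a trail.
# VoidOrb layers flipbook sprites and a mesh, joined by a light with glow sprites, chromatic refraction, smoke wisps and a dark ribbon.
# GroundSandWarp spawns mesh dust at User.DustLocation when User.GroundDust is set, adjusting material by time of day and size by camera distance.
# System timelines taper spawning, orb scale and alpha to fade out.
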NiagaraSystem /DemonMaskVoid/Assets/VFX/NS_VoidMask_Projectile
dependencies {
 usedDataInterfaces [
  &NiagaraDataInterfaceCurve
  &NiagaraDataInterfaceColorCurve
  &NiagaraDataInterfaceVectorCurve
  &NiagaraDataInterfaceArrayFloat
  &NiagaraDataInterfaceSpriteRendererInfo
 ]
 usedDynamicInputs []
 usedModules [
  &self:SystemSpawnScript.SystemScriptSource.SystemScriptGraph.NiagaraNodeAssignment_0.SetVariables_#1
  &SystemState
  &self:SystemSpawnScript.SystemScriptSource.SystemScriptGraph.NiagaraNodeAssignment_1.SetVariables_#2
 ]
}
emitters {
 ChromaticRefraction {
  emitterUpdateScript [
   &EmitterState(
    Inactive Response=::NewEnumerator1
    Life Cycle Mode=::NewEnumerator1
    Loop Behavior=::NewEnumerator1
    Loop Delay=0.5
    Loop Duration Mode=::NewEnumerator1
    UseLoopDelay=true
   )
   &SpawnBurst_Instantaneous(Use Spawn Probability=false, UseLoopCountLimit=false)
  ]
  particleSpawnScript [&InitializeParticle(Sprite Size=<5 5>, Sprite Size Mode=::NewEnumerator4, UsePositionOffset=false, Write Material Random=true)]
  particleUpdateScript [
   &ParticleState(Kill Particles When Lifetime Has Elapsed=false, Let Infinitely Lived Particles Die When Emitter Deactivates=true)
   &Timeline(
    Play Rate=&FloatFromCurve(
     FloatCurve=DataInterface() {propertyValues {Curve {keys [
         {interpMode=RCIM_Cubic, leaveTangent=-12.5334, tangentMode=RCTM_Break, time=0, value=4}
         {interpMode=RCIM_Cubic, tangentMode=RCTM_Break, time=0.7, value=1}
        ]}}}
     Scale Curve=0.5
    )
    Timeline End=1.0000000200408773e+20
   )
   &ScaleSpriteSize(
    Scale Sprite Size Mode=::NewEnumerator0
    Uniform Scale Factor=&FloatFromCurve(
     FloatCurve=DataInterface() {propertyValues {Curve {keys [{interpMode=RCIM_Cubic, tangentMode=RCTM_User, time=0.5, value=0}, {interpMode=RCIM_Cubic, tangentMode=RCTM_User, time=1, value=1}]}}}
     CurveIndex=$Emitter.Age
     Scale Curve=70
    )
   )
   &DynamicMaterialParameters(
    Index 0 Param 1=2
    Index 1 Param 2=500
    Index 1 Param 4=1
    Index 2 Param 1=$Particles.Timeline.Time
    Index 2 Param 2=$Particles.Timeline.Time
    Index 2 Param 3=$Particles.Timeline.Time
    Index 2 Param 4=&SelectFloatFromArray(Direct Array Index=$Particles.PartitionParticles.Partition, Float Selection Array=DataInterface() {propertyValues {FloatData [1,0.3,0.3]}})
    Index 3 Param 1=&FloatFromCurve(FloatCurve=DataInterface() {propertyValues {Curve {keys [<at 0.1 1>,<at 0.4 0>]}}})
    Index 3 Param 2=200
    Module.Index0Param0Write=true
    Module.Index0Param1Write=true
    Module.Index0Param2Write=true
    Module.Index0Param3Write=true
    Module.Index1Param0Write=true
    Module.Index1Param1Write=true
    Module.Index1Param2Write=true
    Module.Index1Param3Write=true
    Module.Index2Param0Write=true
    Module.Index2Param1Write=true
    Module.Index2Param2Write=true
    Module.Index2Param3Write=true
    Module.Index3Param0Write=true
    Module.Index3Param1Write=true
    Module.Index3Param2Write=true
    Module.Index3Param3Write=true
    Write Parameter Index 1=true
    Write Parameter Index 2=true
    Write Parameter Index 3=true
   ) {module=DynamicMaterialParameters001}
   &ScaleColor(Scale RGB=&VectorFromFloat(Value=0.5), ScaleA=true, ScaleRGB=true) {module=ScaleColor001}
  ]
  rendererClasses [&NiagaraSpriteRendererProperties]
  renderers [{rendererClass=&NiagaraSpriteRendererProperties}]
 }
 GroundSandWarp {
  emitterUpdateScript [
   &EmitterState()
   &SpawnRate(
    Spawn Probability=&Multiply_Float(
     A=&MakeCustomFloatFromBool(Bool=$User.GroundDust, True Float=1, False Float=0)
     B=$System.FadeOutFastTimeline
    )
    SpawnRate=40
    Use Spawn Probability=true
   )
  ]
  particleSpawnScript [
   &InitializeParticle(
    Color Maximum=<rgba 0.1905 0.09 0.3 1>
    Color Minimum=<rgba 0.615 0.24 0.8 1>
    Color Mode=::NewEnumerator2
    Lifetime Max=0.6
    Lifetime Min=0.5
    Lifetime Mode=::NewEnumerator1
    Mesh Scale Mode=::NewEnumerator3
    Mesh Uniform Scale=&Multiply_Float(A=&RandomRangeFloat(Minimum=7, Maximum=9), B=0.65)
    Position Offset=<0 0 50>
    Sprite Size Mode=::NewEnumerator3
    Uniform Sprite Size=300
    UsePositionOffset=true
    Write Material Random=true
   )
   &ShapeLocation(
    MaskSpawnGroup=false
    Offset=<0 0 20>
    Offset Mode=::NewEnumerator0
    Override Seed=false
    Rotation Coordinate Space=::World
    Shape Origin=$User.DustLocation
    Sphere Radius=10
    UseOwnerScale=true
   )
   &InitialMeshOrientation(
    Mesh Axis To Orient=&RandomRangeVector(Minimum=<1 0 0>, Maximum=<1 0 0>)
    Mesh Orientation Mode=::NewEnumerator2
    Orientation Vector=&RandomRangeVector(Minimum=<0 1 0>, Maximum=<0 1 0>)
    Rotation=&RandomRangeVector(Minimum=<-0.1 0 0>, Maximum=<0.1 0 1>)
    Use Rotation Vector=true
   )
   &SpriteFacingAndAlignment(Sprite Facing=<0 0 1>, WriteAlignment=false, WriteFacing=true)
  ]
  particleUpdateScript [
   &ParticleState()
   &DynamicMaterialParameters(
    Index 0 Param 1=&FloatFromCurve(FloatCurve=DataInterface() {propertyValues {Curve {keys [{interpMode=RCIM_Cubic, tangentMode=RCTM_User, time=0.5, value=0}, {interpMode=RCIM_Cubic, tangentMode=RCTM_User, time=1, value=1}]}}})
    Index 0 Param 2=&MakeCustomFloatFromBool(
     Bool=false
     True Float=0
     False Float=&FNTimeOfDayScalar(Night Scale=0.05, Morning Scale=-0.26, Day Scale=0, Evening Scale=-0.35)
    )
    Index 0 Param 3=&MakeCustomFloatFromBool(
     Bool=false
     True Float=2
     False Float=&FNTimeOfDayScalar(Night Scale=0.5, Morning Scale=0.23, Evening Scale=0.2)
    )
    Index 0 Param 4=&MakeCustomFloatFromBool(
     Bool=false
     True Float=0
     False Float=&FNTimeOfDayScalar(Night Scale=0.02, Morning Scale=-0.01, Day Scale=0, Evening Scale=-0.01)
    )
    Module.Param0WriteEnabled=true
    Module.Param1WriteEnabled=true
    Module.Param2WriteEnabled=true
    Module.Param3WriteEnabled=true
   )
   /SetVariables(
    Particles.XY:NiagaraFloat=&FloatFromCurve(
     FloatCurve=DataInterface() {propertyValues {Curve {keys [
         {arriveTangent=0.8546, interpMode=RCIM_Cubic, leaveTangent=0.5305, tangentMode=RCTM_Break, time=0, value=0.1}
         {arriveTangent=0.4623, interpMode=RCIM_Cubic, leaveTangent=0.4623, tangentMode=RCTM_User, time=0.8, value=1}
        ]}}}
     Scale Curve=0.75
    )
    Particles.Z:NiagaraFloat=&FloatFromCurve(
     FloatCurve=DataInterface() {propertyValues {Curve {keys [
         {arriveTangent=1.8479, interpMode=RCIM_Cubic, leaveTangent=1.8479, tangentMode=RCTM_User, time=0, value=0}
         {arriveTangent=0.9995, interpMode=RCIM_Cubic, leaveTangent=0.9995, tangentMode=RCTM_User, time=0.4, value=0.4}
         {arriveTangent=0.2173, interpMode=RCIM_Cubic, leaveTangent=0.2173, tangentMode=RCTM_User, time=0.9, value=0.6}
        ]}}}
     Scale Curve=1.2
    )
   ) {isSetParameters=true}
   &ScaleMeshSize(Scale Factor=&MakeVector(X=$Particles.XY, Y=$Particles.XY, Z=$Particles.Z))
   &UpdateMeshOrientation(Rotation Vector=<0 0 1>)
   &ScaleMeshSize(Scale Factor=&MakeVector(
     X=1
     Y=1
     Z=&FloatFromCurve(
      FloatCurve=DataInterface() {propertyValues {Curve {keys [{interpMode=RCIM_Cubic, tangentMode=RCTM_Break, time=0, value=1}, <at 1 1.4>]}}}
      Scale Curve=1.2
     )
    )) {module=ScaleMeshSize001}
   &ScaleMeshSize(Scale Factor=&VectorFromFloat(Value=/Niagara/DynamicInputs/Scalability/ScalabilityDistanceBasedFloat(
      Curve:NiagaraDataInterfaceCurve=DataInterface() {propertyValues {Curve {keys [<at 0 0.4>,<at 0.03 1>,<at 1 1>]}}}
      CurveIndex:NiagaraFloat=$Emitter.Scalability.DistanceFraction
      Bias:NiagaraFloat=0
      Scale:NiagaraFloat=1
     ))) {module=ScaleMeshSize002}
   &ScaleColor(
    Scale Alpha=&FloatFromCurve(
     FloatCurve=DataInterface() {propertyValues {Curve {keys [
         {arriveTangent=8.2356, interpMode=RCIM_Cubic, leaveTangent=8.2356, tangentMode=RCTM_User, time=0, value=0}
         {arriveTangent=0.1969, interpMode=RCIM_Cubic, leaveTangent=0.1969, tangentMode=RCTM_User, time=0.15, value=1}
         {arriveTangent=-1.9699, interpMode=RCIM_Cubic, leaveTangent=-1.9699, tangentMode=RCTM_User, time=0.5, value=0.45}
         {arriveTangentWeight=0.2097, interpMode=RCIM_Cubic, tangentMode=RCTM_User, time=1, value=0}
        ]}}}
     Scale Curve=1.3
    )
    Scale RGB=&MakeVectorFromLinearColorRGB(Color=&ColorFromCurve(ColorCurve=DataInterface() {propertyValues {
        AlphaCurve {keys [<at 0 1 RCIM_Cubic>]}
        BlueCurve {keys [<at 0 3 RCIM_Cubic>, {arriveTangent=-4.0435, interpMode=RCIM_Cubic, leaveTangent=-4.0435, time=0.2507, value=1}, <at 0.72 0.0885>, <at 1 0>]}
        GreenCurve {keys [
          <at 0 1.8 RCIM_Cubic>
          {arriveTangent=-2.3769, interpMode=RCIM_Cubic, leaveTangent=-2.3769, time=0.2507, value=1}
          <at 0.72 0.0885>
          <at 1 0>
         ]}
        RedCurve {keys [
          <at 0 1.8 RCIM_Cubic>
          {arriveTangent=-2.3769, interpMode=RCIM_Cubic, leaveTangent=-2.3769, time=0.2507, value=1}
          <at 0.72 0.0885>
          <at 1 0>
         ]}
       }}))
    ScaleA=true
    ScaleRGB=true
   )
   /SetVariables(Particles.MeshPos:NiagaraPosition=&Add_Vector(A=Linked() {linkedVariable {name=Particles.Position, type=NiagaraPosition}}, B=<0 0 -25>)) {isSetParameters=true}
  ]
  rendererClasses [&NiagaraMeshRendererProperties]
  renderers [{rendererClass=&NiagaraMeshRendererProperties, rendererIndex=0}]
 }
 Light {
  emitterUpdateScript [
   &EmitterState(Life Cycle Mode=::NewEnumerator1, Loop Behavior=::NewEnumerator1, Loop Duration Mode=::NewEnumerator1, UseLoopDelay=true)
   &SpawnBurst_Instantaneous(Use Spawn Probability=false, UseLoopCountLimit=false)
  ]
  particleSpawnScript [
   &InitializeParticle(
    Color=<rgba 0.3667 0 1 1>
    Lifetime=0.25
    Mesh Scale Mode=::NewEnumerator3
    Mesh Uniform Scale=0.3
    Sprite Size Mode=::NewEnumerator3
    Uniform Sprite Size=230
    UsePositionOffset=false
    Write Material Random=true
   )
   &CameraOffset(Camera Offset Amount=50)
   &InitialMeshOrientation(Mesh Orientation Mode=::NewEnumerator1, Use Rotation Vector=false)
  ]
  particleUpdateScript [
   /SetVariables(Particles.LightExponent=8) {isSetParameters=true}
   &ParticleState()
   &ScaleSpriteSize(
    Uniform Curve Scale=2
    Uniform Curve Sprite Scale=DataInterface() {propertyValues {Curve {keys [
        {arriveTangent=-3.6138, interpMode=RCIM_Cubic, leaveTangent=-3.6138, tangentMode=RCTM_User, time=0.3, value=1}
        {interpMode=RCIM_Cubic, tangentMode=RCTM_User, time=1, value=0}
       ]}}}
   ) {module=ScaleSpriteSize002}
   &SpriteRotationRate(Rotation Rate=-40)
   &ScaleColor(
    Scale Alpha=&FloatFromCurve(FloatCurve=DataInterface() {propertyValues {Curve {keys [<at 0 0 RCIM_Cubic>,<at 0.2 2 RCIM_Cubic>,<at 1 0>]}}})
    Scale RGB=<2 2 2>
    ScaleA=true
    ScaleRGB=true
   )
   /SetVariables(Particles.SpriteSizeDarkGlow:Vector2f=&Multiply_Vector2D_ByFloat(Float=0.6, Vector2D=$Particles.SpriteSize)) {isSetParameters=true}
   /SetVariables(Particles.ColorDarkGlow:LinearColor=&ColorFromCurve(ColorCurve=DataInterface() {propertyValues {
       AlphaCurve {keys [<at 0 0 RCIM_Cubic>,<at 0.133 2.5>,<at 1 0 RCIM_Cubic>]}
       BlueCurve {keys [<at 0 0.01 RCIM_Cubic>]}
       GreenCurve {keys [<at 0 0.008 RCIM_Cubic>]}
       RedCurve {keys [<at 0 0.0087 RCIM_Cubic>]}
      }})) {isSetParameters=true}
   /SetVariables(
    Particles.DynamicMaterialParameterDarkGlow:Vector4f=&MakeVector4(
     X=&Lerp_Float(
      A=1
      B=-0.5
      Alpha=&FloatFromCurve(FloatCurve=DataInterface() {propertyValues {Curve {keys [
           {arriveTangent=1.4226, interpMode=RCIM_Cubic, leaveTangent=1.4226, tangentMode=RCTM_User, time=0, value=0}
           {interpMode=RCIM_Cubic, tangentMode=RCTM_User, time=0.8, value=1}
          ]}}})
     )
     Y=1
     Z=&Lerp_Float(
      A=-2
      B=-1
      Alpha=&FloatFromCurve(FloatCurve=DataInterface() {propertyValues {Curve {keys [<at 0 0>,<at 1 1>]}}})
     )
     W=0.54
    )
    Particles.DynamicMaterialParameterVortex:Vector4f=&MakeVector4(X=0.1, Y=1, Z=0, W=0.25)
   ) {isSetParameters=true}
   /SetVariables(Particles.LightColor=&ColorFromCurve(ColorCurve=DataInterface() {propertyValues {
       AlphaCurve {keys [<at 0.3533 1 RCIM_Cubic>,<at 1 0 RCIM_Cubic>]}
       BlueCurve {keys [<at 0 10 RCIM_Cubic>]}
       GreenCurve {keys [<at 0 0 RCIM_Cubic>]}
       RedCurve {keys [<at 0 3.3333 RCIM_Cubic>]}
      }})) {isSetParameters=true}
  ]
  rendererClasses [&NiagaraLightRendererProperties, &NiagaraSpriteRendererProperties]
  renderers [
   {rendererClass=&NiagaraLightRendererProperties, rendererIndex=0}
   {rendererClass=&NiagaraSpriteRendererProperties, rendererIndex=1}
   {rendererClass=&NiagaraSpriteRendererProperties, rendererIndex=2}
   {rendererClass=&NiagaraSpriteRendererProperties, rendererIndex=3}
  ]
 }
 SmokeWhisps {
  emitterUpdateScript [
   &EmitterState(Life Cycle Mode=::NewEnumerator1, Loop Behavior=::NewEnumerator1, Loop Delay=0.05, Loop Duration Mode=::NewEnumerator1, UseLoopDelay=false)
   &SpawnBurst_Instantaneous(Spawn Count=5, Use Spawn Probability=false, UseLoopCountLimit=false)
   &SpawnRate(
    SpawnRate=&Multiply_Float(
     A=30
     B=&Lerp_Float(
      A=0
      B=1
      Alpha=&FloatFromCurve(
       FloatCurve=DataInterface() {propertyValues {Curve {keys [{interpMode=RCIM_Cubic, tangentMode=RCTM_User, time=0.1, value=1}, {interpMode=RCIM_Cubic, tangentMode=RCTM_User, time=0.5, value=0}]}}}
       CurveIndex=$System.Timeline.NormalizedTime
      )
     )
    )
    Use Spawn Probability=false
   )
  ]
  particleSpawnScript [
   &InitializeParticle(
    Color Maximum=<rgba 1 0.4531 0.9055 1>
    Color Minimum=<rgba 0.098 0.0728 0.2 1>
    Color Mode=::NewEnumerator2
    Lifetime=&Multiply_Float(A=&RandomRangeFloat(Minimum=0.6, Maximum=1.3), B=1)
    Mass Mode=::NewEnumerator3
    Sprite Rotation Mode=::NewEnumerator2
    Sprite Size Mode=::NewEnumerator1
    Uniform Sprite Size Max=300
    Uniform Sprite Size Min=190
    UsePositionOffset=false
    Write Material Random=true
   )
   &ShapeLocation(MaskSpawnGroup=false, Offset=<0 -15 0>, Offset Mode=::NewEnumerator0, Override Seed=false, Sphere Radius=9, UseOwnerScale=true)
  ]
  particleUpdateScript [
   &ParticleState()
   &ScaleSpriteSize(
    Scale Sprite Size Mode=::NewEnumerator0
    Uniform Scale Factor=&Multiply_Float(
     A=&Lerp_Float(
      A=0.5
      B=1
      Alpha=&FloatFromCurve(FloatCurve=DataInterface() {propertyValues {Curve {keys [
           {arriveTangent=0.5115, interpMode=RCIM_Cubic, leaveTangent=0.5115, tangentMode=RCTM_User, time=0, value=0}
           {arriveTangent=2.8618, interpMode=RCIM_Cubic, leaveTangent=2.8618, tangentMode=RCTM_User, time=0.7, value=1}
          ]}}})
     )
     B=&FloatFromCurve(
      FloatCurve=DataInterface() {propertyValues {Curve {keys [{interpMode=RCIM_Cubic, tangentMode=RCTM_User, time=0, value=0.3}, {interpMode=RCIM_Cubic, tangentMode=RCTM_User, time=0.25, value=1}]}}}
      CurveIndex=$Emitter.Age
     )
    )
   )
   &ScaleColor(
    Scale Alpha=&Multiply_Float(
     A=&FloatFromCurve(FloatCurve=DataInterface() {propertyValues {Curve {keys [<at 0 0 RCIM_Cubic>,<at 0.2 1 RCIM_Cubic>,<at 1 0 RCIM_Cubic>]}}}, Scale Curve=7)
     B=&FloatFromCurve(
      FloatCurve=DataInterface() {propertyValues {Curve {keys [{interpMode=RCIM_Cubic, tangentMode=RCTM_User, time=0, value=0}, {interpMode=RCIM_Cubic, tangentMode=RCTM_User, time=0.1, value=1}]}}}
      CurveIndex=$Emitter.Age
     )
    )
    Scale RGB=&MakeVectorFromLinearColorRGB(Color=&ColorFromCurve(ColorCurve=DataInterface() {propertyValues {
        AlphaCurve {keys [<at 0 1 RCIM_Cubic>]}
        BlueCurve {keys [<at 0 4>, {arriveTangent=-7.1451, interpMode=RCIM_Cubic, time=0.1569, value=1}, <at 0.5458 0.1>, <at 1 0 RCIM_Cubic>]}
        GreenCurve {keys [<at 0 0.04>, {arriveTangent=-0.0183, interpMode=RCIM_Cubic, time=0.1569, value=0.04}, <at 0.5458 0.03>, <at 1 0 RCIM_Cubic>]}
        RedCurve {keys [<at 0 2.02>, {arriveTangent=-3.6458, interpMode=RCIM_Cubic, time=0.1569, value=0.408}, <at 0.5458 0.03>, <at 1 0 RCIM_Cubic>]}
       }}))
    ScaleA=true
    ScaleRGB=true
   )
   &GravityForce(Gravity=<0 0 -200>)
   &Drag(Drag=5, Use Linear Drag=true, Use Rotational Drag=true)
   &InheritSourceMovement(
    Applied Position Scale=&VectorFromFloat(Value=&FloatFromCurve(
      FloatCurve=DataInterface() {propertyValues {Curve {keys [
          {arriveTangent=-3.5756, interpMode=RCIM_Cubic, leaveTangent=-3.5756, tangentMode=RCTM_User, time=0, value=1}
          {interpMode=RCIM_Cubic, tangentMode=RCTM_User, time=0.5, value=0}
         ]}}}
      Scale Curve=0.3
     ))
    ApplyForce=false
    ApplyPosition=true
    ApplyPositionFalloff=false
   )
   &SolveForcesAndVelocity(Clamp Velocity=false, Limit Acceleration=false)
   &SubUVAnimation(
    End Frame Range Override=32
    "Sprite Renderer "=DataInterface() {propertyValues {SpriteRenderer=&self:SmokeWhisps_0.NiagaraSpriteRendererProperties_0}}
    UseEndFrame=true
    UseStartFrame=true
   )
   &DynamicMaterialParameters(
    Index 0 Param 1=&FloatFromCurve(FloatCurve=DataInterface() {propertyValues {Curve {keys [
         {arriveTangent=0.4272, interpMode=RCIM_Cubic, leaveTangent=0.4272, tangentMode=RCTM_User, time=0, value=0.25}
         {arriveTangent=1.2659, interpMode=RCIM_Cubic, leaveTangent=1.2659, tangentMode=RCTM_User, time=1, value=1}
        ]}}})
    Index 0 Param 2=40
    Module.Index0Param0Write=true
    Module.Index0Param1Write=true
    Module.Index0Param2Write=false
    Module.Index0Param3Write=false
   )
   &CameraOffset(Camera Offset Amount=10)
  ]
  rendererClasses [&NiagaraSpriteRendererProperties]
  renderers [{rendererClass=&NiagaraSpriteRendererProperties}]
 }
 Trail {
  emitterUpdateScript [
   &EmitterState()
   &SpawnPerUnit(
    Max Movement Threshold=1000
    Module.Use Max Movement Threshold=false
    Module.Use Movement Tolerance=false
    Spawn Spacing=100
    Use Spawn Probability=false
   ) {enabled=false}
   &SpawnRate(
    SpawnRate=&Multiply_Float(
     A=55
     B=&FloatFromCurve(
      FloatCurve=DataInterface() {propertyValues {Curve {keys [{interpMode=RCIM_Cubic, tangentMode=RCTM_User, time=0, value=1}, {interpMode=RCIM_Cubic, tangentMode=RCTM_User, time=1, value=0}]}}}
      CurveIndex=$System.Timeline.NormalizedTime
     )
    )
    Use Spawn Probability=false
   )
  ]
  particleSpawnScript [&InitializeParticle(
    Color=<rgba 0.0521 0.0665 0.0938 1>
    Mass Mode=::NewEnumerator3
    Ribbon Width=200
    Ribbon Width Mode=::NewEnumerator1
    UsePositionOffset=false
    Write Material Random=true
   )]
  particleUpdateScript [
   &ParticleState()
   &GravityForce(Gravity=<0 0 90>)
   &ScaleColor(
    Scale Alpha=&Multiply_Float(
     A=&Multiply_Float(
      A=&FloatFromCurve(
       FloatCurve=DataInterface() {propertyValues {Curve {keys [{interpMode=RCIM_Cubic, tangentMode=RCTM_User, time=0, value=1}, {interpMode=RCIM_Cubic, tangentMode=RCTM_User, time=1, value=0}]}}}
       CurveIndex=$System.Timeline.NormalizedTime
      )
      B=&FloatFromCurve(
       FloatCurve=DataInterface() {propertyValues {Curve {keys [{interpMode=RCIM_Cubic, tangentMode=RCTM_User, time=0, value=0}, {interpMode=RCIM_Cubic, tangentMode=RCTM_User, time=0.5, value=1}]}}}
       CurveIndex=&ReturnNormalizedExecIndex()
      )
     )
     B=&Multiply_Float(
      A=&FloatFromCurve(FloatCurve=DataInterface() {propertyValues {Curve {keys [{interpMode=RCIM_Cubic, tangentMode=RCTM_User, time=0, value=0}, {interpMode=RCIM_Cubic, tangentMode=RCTM_User, time=0.07, value=1}]}}})
      B=&FloatFromCurve(
       FloatCurve=DataInterface() {propertyValues {Curve {keys [{interpMode=RCIM_Cubic, tangentMode=RCTM_User, time=0, value=0}, {interpMode=RCIM_Cubic, tangentMode=RCTM_User, time=0.5, value=1}]}}}
       CurveIndex=&ReturnNormalizedExecIndex()
      )
     )
    )
    ScaleA=true
    ScaleRGB=false
   ) {module=ScaleColor001}
   &SolveForcesAndVelocity(Clamp Velocity=false, Limit Acceleration=false)
   &ScaleRibbonWidth(Ribbon Width Scale=&FloatFromCurve(FloatCurve=DataInterface() {propertyValues {Curve {keys [
         {arriveTangent=19.2206, interpMode=RCIM_Cubic, leaveTangent=19.2206, tangentMode=RCTM_User, time=0, value=0.2}
         {arriveTangent=2.25, interpMode=RCIM_Cubic, leaveTangent=2.25, time=0.1, value=2}
         <at 0.3 0.875 RCIM_Cubic>
         {arriveTangent=0.3186, interpMode=RCIM_Cubic, leaveTangent=0.3186, tangentMode=RCTM_User, time=0.7, value=2}
        ]}}}))
   &DynamicMaterialParameters(
    Index 0 Param 1=&FloatFromCurve(FloatCurve=DataInterface() {propertyValues {Curve {keys [
         {interpMode=RCIM_Cubic, leaveTangent=-9.1594, tangentMode=RCTM_Break, time=0, value=1}
         {interpMode=RCIM_Cubic, leaveTangent=-5.9384, tangentMode=RCTM_Break, time=0.3, value=0}
        ]}}})
    Index 0 Param 2=&FloatFromCurve(FloatCurve=DataInterface() {propertyValues {Curve {keys [{tangentMode=RCTM_User, time=0, value=0}, {tangentMode=RCTM_User, time=1, value=1}]}}})
    Index 0 Param 4=&FloatFromCurve(FloatCurve=DataInterface() {propertyValues {Curve {keys [
         {interpMode=RCIM_Cubic, leaveTangent=-48.5828, tangentMode=RCTM_Break, time=0, value=3}
         {arriveTangent=-2.4798, tangentMode=RCTM_User, time=0.2, value=0.25}
         <at 0.5 0>
        ]}}})
    Module.Index0Param0Write=true
    Module.Index0Param1Write=true
    Module.Index0Param2Write=true
    Module.Index0Param3Write=true
   )
  ]
  rendererClasses [&NiagaraRibbonRendererProperties]
  renderers [{rendererClass=&NiagaraRibbonRendererProperties}]
 }
 VoidOrb {
  emitterUpdateScript [
   &EmitterState(Life Cycle Mode=::NewEnumerator1, Loop Behavior=::NewEnumerator1, Loop Delay=0.5, Loop Duration Mode=::NewEnumerator1, UseLoopDelay=false)
   &SpawnBurst_Instantaneous(Spawn Count=4, Use Spawn Probability=false, UseLoopCountLimit=false)
   &SpawnRate(SpawnRate=7, Use Spawn Probability=false)
   &SpawnRate(SpawnGroup=1, SpawnRate=7, Use Spawn Probability=false) {module=SpawnRate001}
   /SetVariables(
    Emitter.CameraOfsCore=10
    Emitter.CameraOfs_Outline=-10
    Emitter.Color:LinearColor=&ColorFromCurve(
     ColorCurve=DataInterface() {propertyValues {
       AlphaCurve {keys [<at 0.096 0 RCIM_Cubic>,<at 0.3023 1 RCIM_Cubic>]}
       BlueCurve {keys [<at 0 5 RCIM_Cubic>]}
       GreenCurve {keys [<at 0 0 RCIM_Cubic>]}
       RedCurve {keys [<at 0 1.6667 RCIM_Cubic>]}
      }}
     CurveIndex=$Emitter.Age
    )
    Emitter.Color_Core:LinearColor=<rgba 0.441 0.0208 1 1>
    Emitter.Color_Outline:LinearColor=<rgba 0.3566 0 1 1>
   ) {isSetParameters=true}
   /SetVariables(Emitter.Scale:Vector3f=&VectorFromCurve(
     VectorCurve=DataInterface() {propertyValues {
       XCurve {keys [{tangentMode=RCTM_User, time=0, value=0.1}, {tangentMode=RCTM_User, time=0.2, value=1}]}
       YCurve {keys [{tangentMode=RCTM_User, time=0, value=0.1}, {tangentMode=RCTM_User, time=0.2, value=1}]}
       ZCurve {keys [{tangentMode=RCTM_User, time=0, value=0.1}, {tangentMode=RCTM_User, time=0.2, value=1}]}
      }}
     CurveIndex=$Emitter.Age
     Scale Curve=&VectorFromFloat(Value=&Multiply_Float(A=1.3, B=$System.FadeOutFastTimeline))
    )) {isSetParameters=true}
   /SetVariables(Emitter.Scale2:Vector3f=&Multiply_Vector(
     A=$Emitter.Scale
     B=&VectorFromFloat(Value=&Multiply_Float(A=0.4, B=$System.FadeOutFastTimeline))
    )) {isSetParameters=true}
   &SpawnBurst_Instantaneous(Spawn Time=8.55, Use Spawn Probability=false, UseLoopCountLimit=false) {enabled=false, module=SpawnBurst_Instantaneous001}
  ]
  particleSpawnScript [
   &InitializeParticle(
    Color=&MaskLinearColorBySpawnGroup(Masked Linear Color=<rgba 0.15 0 0.3 1>, Passthrough Linear Color=<rgba 1 0 3 2>)
    Lifetime=&RandomRangeFloat(Minimum=0.5, Maximum=1)
    Sprite Rotation Mode=::NewEnumerator2
    Sprite Size=&Vector2DFromFloat(Value=&MaskFloatBySpawnGroup(
      Masked Float=&RandomRangeFloat(Minimum=160, Maximum=190)
      Passthrough Float=&RandomRangeFloat(Minimum=140, Maximum=160)
     ))
    Sprite Size Mode=::NewEnumerator4
    Sprite UV Mode=::NewEnumerator4
    Sprite UV Scale:Vector2f=<-1 1>
    UsePositionOffset=false
    Write Material Random=true
   )
   &CameraOffset(Camera Offset Amount=-15)
   &DynamicMaterialParameters(
    Index 0 Param 1=&RandomRangeFloat(Minimum=0, Maximum=-6)
    Module.Index0Param0Write=true
    Module.Index0Param1Write=false
    Module.Index0Param2Write=false
    Module.Index0Param3Write=false
   )
   &AddVelocity(Velocity=&MaskVectorBySpawnGroup(Masked Vector=&RandomRangeVector(Minimum=<0 0 30>, Maximum=<0 0 -30>), Passthrough Vector=<0 0 0>)) {enabled=false}
   &AddVelocity(Velocity=$System.DebugSpeed) {module=AddVelocity001}
  ]
  particleUpdateScript [
   &ParticleState()
   &ScaleSpriteSize(
    Scale Sprite Size Mode=::NewEnumerator0
    Uniform Scale Factor=&FloatFromCurve(FloatCurve=DataInterface() {propertyValues {Curve {keys [<at 0 0.1 RCIM_Cubic>, {interpMode=RCIM_Cubic, leaveTangent=0.1472, time=0.4, value=1}]}}})
   )
   &ScaleColor(
    Scale Alpha=&FloatFromCurve(
     FloatCurve=DataInterface() {propertyValues {Curve {keys [
         {arriveTangent=9.8672, interpMode=RCIM_Cubic, leaveTangent=9.8672, tangentMode=RCTM_User, time=0, value=0}
         <at 0.2 1 RCIM_Cubic>
         {interpMode=RCIM_Cubic, leaveTangent=-26.9591, time=1, value=0}
        ]}}}
     Scale Curve=$System.FadeOutFastTimeline
    )
    Scale RGB=&VectorFromCurve(VectorCurve=DataInterface() {propertyValues {
       XCurve {keys [<at 0.25 1 RCIM_Cubic>,<at 1 0.5>]}
       YCurve {keys [<at 0.25 1 RCIM_Cubic>,<at 1 0.5>]}
       ZCurve {keys [<at 0.25 1 RCIM_Cubic>,<at 1 0.5>]}
      }})
    ScaleA=true
    ScaleRGB=true
   )
   /Niagara/Modules/Update/SubUV/SubUVAnimation(Number Of Frames=32)
   &DynamicMaterialParameters(
    Index 0 Param 1=&FloatFromCurve(
     FloatCurve=DataInterface() {propertyValues {Curve {keys [
         {arriveTangent=2.8718, interpMode=RCIM_Cubic, leaveTangent=2.8718, tangentMode=RCTM_User, time=0, value=0}
         {arriveTangent=-0.3942, interpMode=RCIM_Cubic, leaveTangent=-0.3942, tangentMode=RCTM_User, time=1, value=1}
        ]}}}
     Scale Curve=0.1
    )
    Index 0 Param 2=50
    Module.Index0Param0Write=true
    Module.Index0Param1Write=true
    Module.Index0Param2Write=false
    Module.Index0Param3Write=false
   ) {module=DynamicMaterialParameters001}
   &Drag(Drag=2, Use Linear Drag=true, Use Rotational Drag=true)
   &GravityForce(Gravity=<0 0 -100>) {enabled=false}
   &InheritSourceMovement(
    Applied Position Scale=&VectorFromFloat(Value=&FloatFromCurve(FloatCurve=DataInterface() {propertyValues {Curve {keys [
          {interpMode=RCIM_Cubic, tangentMode=RCTM_User, time=0.35, value=1}
          {arriveTangent=-2.5788, interpMode=RCIM_Cubic, leaveTangent=-2.5788, tangentMode=RCTM_User, time=1, value=0}
         ]}}}))
    ApplyForce=false
    ApplyPosition=true
    ApplyPositionFalloff=false
   ) {enabled=false}
   &SolveForcesAndVelocity(Clamp Velocity=false, Limit Acceleration=false)
  ]
  rendererClasses [&NiagaraSpriteRendererProperties, &NiagaraMeshRendererProperties]
  renderers [
   {rendererClass=&NiagaraSpriteRendererProperties}
   {rendererClass=&NiagaraMeshRendererProperties, rendererIndex=1}
  ]
 }
}
summary {
 systemName=NS_VoidMask_Projectile
 emitters [
  {emitterName=Light, rendererClasses [&NiagaraLightRendererProperties, &NiagaraSpriteRendererProperties]}
  {emitterName=VoidOrb, rendererClasses [&NiagaraSpriteRendererProperties, &NiagaraMeshRendererProperties]}
  {emitterName=ChromaticRefraction, rendererClasses [&NiagaraSpriteRendererProperties]}
  {emitterName=SmokeWhisps, rendererClasses [&NiagaraSpriteRendererProperties]}
  {emitterName=Trail, rendererClasses [&NiagaraRibbonRendererProperties]}
  {emitterName=GroundSandWarp, rendererClasses [&NiagaraMeshRendererProperties]}
 ]
 userVariables [
  User.DustLocation {type=NiagaraPosition, description="", defaultValue {}}
  User.GroundDust {type=NiagaraBool}
  User.ProjStopped {type=NiagaraBool}
  User.RibbonIDOffset {type=NiagaraInt32}
 ]
}
