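# Niagara system for the Void Mask reload: a one-second-scale purple-and-black burst at a hand offset with trails along the character.
# MaskTrail uses FN_BoneFlow on head, right forearm and hand bones to shape three partitioned ribbons.
# Staggered bursts add chromatic refraction, fake light and FlareVortex, which uses spawn groups and per-renderer bright/dark attributes to combine flare and vortex.
# The standalone Flare and Vortex emitters are disabled.
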NiagaraSystem /DemonMaskVoid/Assets/VFX/NS_VoidMask_Reload
dependencies {
 usedDataInterfaces [
  &NiagaraDataInterfaceCurve
  &NiagaraDataInterfaceVector2DCurve
  &NiagaraDataInterfaceColorCurve
  &NiagaraDataInterfaceSkeletalMesh
  &NiagaraDataInterfaceArrayFloat
  &NiagaraDataInterfaceArrayColor
  &NiagaraDataInterfaceArrayFloat3
 ]
 usedDynamicInputs []
 usedModules [
  &SystemState
  &self:SystemSpawnScript.SystemScriptSource.SystemScriptGraph.NiagaraNodeAssignment_0.SetVariables_#1
 ]
}
emitters {
 Copied_ChromaticRefraction {
  emitterUpdateScript [
   &EmitterState(Life Cycle Mode=::NewEnumerator1, Loop Behavior=::NewEnumerator1, Loop Delay=0.06, Loop Duration=1, UseLoopDelay=true)
   &SpawnBurst_Instantaneous(Use Spawn Probability=false, UseLoopCountLimit=false)
  ]
  particleSpawnScript [
   &PartitionParticles(Number Of Partitions=3) {enabled=false}
   &InitializeParticle(
    Lifetime=0.9
    Mesh Scale Mode=::NewEnumerator3
    Position Offset=$System.HandFXOffset
    Sprite Size Mode=::NewEnumerator3
    Uniform Sprite Size=70
    UsePositionOffset=true
    Write Material Random=true
   )
   &SystemLocation(MaskSpawnGroup=false, Offset:NiagaraPosition=$User.Offset Pos) {enabled=false}
   &CameraOffset(Camera Offset Amount=20) {enabled=false}
  ]
  particleUpdateScript [
   &ParticleState()
   &Timeline(
    Play Rate=&FloatFromCurve(
     FloatCurve=DataInterface() {propertyValues {Curve {keys [
         {interpMode=RCIM_Cubic, leaveTangent=-12.5334, tangentMode=RCTM_Break, time=0, value=4}
         {interpMode=RCIM_Cubic, tangentMode=RCTM_Break, time=0.7, value=1}
        ]}}}
     Scale Curve=0.5
    )
    Timeline End=1.0000000200408773e+20
   )
   &ScaleSpriteSize(
    Scale Sprite Size Mode=::NewEnumerator0
    Uniform Scale Factor=&FloatFromCurve(FloatCurve=DataInterface() {propertyValues {Curve {keys [
         {interpMode=RCIM_Cubic, tangentMode=RCTM_User, time=0, value=0}
         {interpMode=RCIM_Cubic, tangentMode=RCTM_User, time=0.3, value=1}
         {arriveTangent=-0.0632, tangentMode=RCTM_User, time=1, value=0}
        ]}}})
   )
   &SkeletalMeshLocation(
    Bone / Socket Index=&Add_Integer(A=$Particles.PartitionParticles.Partition, B=12)
    Bone Sampling Mode=::NewEnumerator4
    MaskSpawnGroup=false
    Sampled Position Offset=&SelectVectorFromArray(
     Direct Array Index=$Particles.PartitionParticles.Partition
     Vector Selection Array=DataInterface() {propertyValues {FloatData [<0 0 0>, <-6 0 0>, <6 0 0>]}}
    )
    Skeletal Mesh=$System.SK_Mesh
   ) {enabled=false}
   &DynamicMaterialParameters(
    Index 0 Param 1=2
    Index 0 Param 2=&Lerp_Float(
     A=1
     B=0
     Alpha=&FloatFromCurve(FloatCurve=DataInterface() {propertyValues {Curve {keys [<at 0 1>,<at 1 0>]}}})
    )
    Index 0 Param 3=&Lerp_Float(
     A=2
     B=0
     Alpha=&FloatFromCurve(FloatCurve=DataInterface() {propertyValues {Curve {keys [<at 0 1>,<at 1 0>]}}})
    )
    Index 0 Param 4=0.4
    Index 1 Param 2=50
    Index 1 Param 3=0.5
    Index 1 Param 4=1
    Index 2 Param 1=$Particles.Timeline.Time
    Index 2 Param 2=$Particles.Timeline.Time
    Index 2 Param 3=$Particles.Timeline.Time
    Index 2 Param 4=&SelectFloatFromArray(Direct Array Index=$Particles.PartitionParticles.Partition, Float Selection Array=DataInterface() {propertyValues {FloatData [1,0.3,0.3]}})
    Index 3 Param 1=&FloatFromCurve(FloatCurve=DataInterface() {propertyValues {Curve {keys [<at 0.1 1>,<at 0.4 0>]}}})
    Index 3 Param 2=200
    Module.Index0Param0Write=true
    Module.Index0Param1Write=true
    Module.Index0Param2Write=true
    Module.Index0Param3Write=true
    Module.Index1Param0Write=true
    Module.Index1Param1Write=true
    Module.Index1Param2Write=true
    Module.Index1Param3Write=true
    Module.Index2Param0Write=true
    Module.Index2Param1Write=true
    Module.Index2Param2Write=true
    Module.Index2Param3Write=true
    Module.Index3Param0Write=true
    Module.Index3Param1Write=true
    Module.Index3Param2Write=true
    Module.Index3Param3Write=true
    Write Parameter Index 1=true
    Write Parameter Index 2=true
    Write Parameter Index 3=true
   ) {module=DynamicMaterialParameters001}
   &ScaleColor(
    Scale RGB=&MakeVectorFromLinearColorRGB(Color=&ColorFromCurve(ColorCurve=DataInterface() {propertyValues {
        AlphaCurve {keys [<at 0 1 RCIM_Cubic>]}
        BlueCurve {keys [<at 0 2 RCIM_Cubic>,<at 0.9394 0.396 RCIM_Cubic>]}
        GreenCurve {keys [<at 0 0.6 RCIM_Cubic>,<at 0.9394 0.2792 RCIM_Cubic>]}
        RedCurve {keys [<at 0 1.4185 RCIM_Cubic>,<at 0.9394 0.3329 RCIM_Cubic>]}
       }}))
    ScaleA=true
    ScaleRGB=true
   ) {module=ScaleColor001}
  ]
  rendererClasses [&NiagaraSpriteRendererProperties]
  renderers [
   {rendererClass=&NiagaraSpriteRendererProperties}
   {rendererClass=&NiagaraSpriteRendererProperties, rendererIndex=1}
  ]
 }
 FakeLight {
  emitterUpdateScript [
   &EmitterState(Life Cycle Mode=::NewEnumerator1, Loop Behavior=::NewEnumerator1, Loop Delay=0.15, Loop Duration=1, UseLoopDelay=true)
   &SpawnBurst_Instantaneous(Use Spawn Probability=false, UseLoopCountLimit=false)
   &SpawnRate(SpawnRate=10, Use Spawn Probability=false) {enabled=false}
  ]
  particleSpawnScript [
   &InitializeParticle(
    Color=<rgba 12.675 0.3 30 1>
    Lifetime=1.2
    Mass Mode=::NewEnumerator3
    Position Offset=$System.HandFXOffset
    Sprite Rotation Mode=::NewEnumerator2
    Sprite Size Mode=::NewEnumerator3
    Uniform Sprite Size=55
    UsePositionOffset=true
    Write Material Random=true
   )
   &CameraOffset(Camera Offset Amount=2)
  ]
  particleUpdateScript [
   &ParticleState()
   &ScaleSpriteSize(
    Scale Sprite Size Mode=::NewEnumerator0
    Uniform Scale Factor=&Lerp_Float(
     A=0.5
     B=1
     Alpha=&FloatFromCurve(FloatCurve=DataInterface() {propertyValues {Curve {keys [
          {arriveTangent=1.2397, interpMode=RCIM_Cubic, leaveTangent=1.2397, tangentMode=RCTM_User, time=0, value=0}
          {interpMode=RCIM_Cubic, tangentMode=RCTM_User, time=0.8, value=1}
         ]}}})
    )
   ) {enabled=false}
   &ScaleColor(
    Scale Alpha=&FloatFromCurve(FloatCurve=DataInterface() {propertyValues {Curve {keys [
         <at 0 0 RCIM_Cubic>
         {arriveTangent=1.9262, interpMode=RCIM_Cubic, leaveTangent=1.9262, tangentMode=RCTM_User, time=0.6, value=1}
         <at 1 0 RCIM_Cubic>
        ]}}})
    Scale RGB=&MakeVectorFromLinearColorRGB(Color=&ColorFromCurve(ColorCurve=DataInterface() {propertyValues {
        AlphaCurve {keys [<at 0 1 RCIM_Cubic>]}
        BlueCurve {keys [<at 0 1.8654 RCIM_Cubic>]}
        GreenCurve {keys [<at 0 0 RCIM_Cubic>]}
        RedCurve {keys [<at 0 2 RCIM_Cubic>]}
       }}))
    ScaleA=true
    ScaleRGB=false
   )
   &SolveForcesAndVelocity(Clamp Velocity=false, Limit Acceleration=false)
   /SetVariables(
    Particles.LightColor=&ColorFromCurve(ColorCurve=DataInterface() {propertyValues {
       AlphaCurve {keys [<at 0.0735 0>,<at 0.1586 1 RCIM_Cubic>,<at 0.3546 0 RCIM_Cubic>]}
       BlueCurve {keys [<at 0.0735 10 RCIM_Cubic>,<at 0.2025 1>]}
       GreenCurve {keys [{arriveTangent=0.0491, interpMode=RCIM_Cubic, time=0.0735, value=0}, <at 0.2025 0>]}
       RedCurve {keys [{arriveTangent=-0.1408, interpMode=RCIM_Cubic, time=0.0735, value=3.6667}, <at 0.2025 0.1667>]}
      }})
    Particles.SpriteSizeLight:Vector2f=&Multiply_Vector2D_ByFloat(
     Float=&FloatFromCurve(FloatCurve=DataInterface() {propertyValues {Curve {keys [<at 0 0.3>,<at 0.4 1>]}}}, Scale Curve=1.1)
     Vector2D=$Particles.SpriteSize
    )
   ) {isSetParameters=true}
  ]
  rendererClasses [&NiagaraSpriteRendererProperties]
  renderers [
   {rendererClass=&NiagaraSpriteRendererProperties}
   {rendererClass=&NiagaraSpriteRendererProperties, rendererIndex=1}
  ]
 }
 Flare {
  bEnabled=false
  emitterUpdateScript [
   &EmitterState(Life Cycle Mode=::NewEnumerator1, Loop Behavior=::NewEnumerator1, Loop Delay=0.2, Loop Duration=0.5, UseLoopDelay=true)
   &SpawnBurst_Instantaneous(Spawn Time=0.1, Use Spawn Probability=false, UseLoopCountLimit=false) {enabled=false, module=SpawnBurst_Instantaneous001}
   &SpawnBurst_Instantaneous(Use Spawn Probability=false, UseLoopCountLimit=false)
  ]
  particleSpawnScript [
   &InitializeParticle(
    Lifetime=0.85
    Position Offset=$System.HandFXOffset
    Sprite Size Mode=::NewEnumerator3
    Sprite UV Mode=::NewEnumerator4
    Sprite UV Scale:Vector2f=<-1 1>
    Uniform Sprite Size=30
    UsePositionOffset=true
    Write Material Random=true
   )
   &CameraOffset(Camera Offset Amount=20)
   &DynamicMaterialParameters(
    Index 0 Param 1=&RandomRangeFloat(Minimum=0, Maximum=-6)
    Module.Index0Param0Write=false
    Module.Index0Param1Write=false
    Module.Index0Param2Write=false
    Module.Index0Param3Write=false
   ) {enabled=false}
   /SetVariables(Particles.Color2=<rgba 0 0 0 10>, Particles.SpriteSize2=<20 20>) {enabled=false, isSetParameters=true}
  ]
  particleUpdateScript [
   &ParticleState()
   &ScaleSpriteSize(
    Scale Factor=&MakeVector2D(
     X=&FloatFromCurve(FloatCurve=DataInterface() {propertyValues {Curve {keys [
          {interpMode=RCIM_Cubic, tangentMode=RCTM_User, time=0.15, value=0.2}
          {arriveTangent=7.8407, interpMode=RCIM_Cubic, leaveTangent=7.8407, tangentMode=RCTM_User, time=0.8, value=3}
         ]}}})
     Y=&FloatFromCurve(FloatCurve=DataInterface() {propertyValues {Curve {keys [
          {interpMode=RCIM_Cubic, tangentMode=RCTM_User, time=0.1, value=1}
          {arriveTangent=-3.6877, interpMode=RCIM_Cubic, leaveTangent=-3.6877, tangentMode=RCTM_User, time=0.5, value=0.01}
         ]}}})
    )
    Scale Sprite Size Mode=::NewEnumerator2
   ) {module=ScaleSpriteSize001}
   &ScaleSpriteSize(
    Uniform Curve Scale=2
    Uniform Curve Sprite Scale=DataInterface() {propertyValues {Curve {keys [
        {interpMode=RCIM_Cubic, tangentMode=RCTM_User, time=0, value=0.2}
        {arriveTangent=-0.9517, interpMode=RCIM_Cubic, leaveTangent=-0.9517, tangentMode=RCTM_User, time=0.2, value=1}
        {arriveTangent=-0.1086, interpMode=RCIM_Cubic, leaveTangent=-0.1086, tangentMode=RCTM_User, time=1, value=0}
       ]}}}
   )
   &ScaleColor(
    Scale Alpha=&FloatFromCurve(FloatCurve=DataInterface() {propertyValues {Curve {keys [
         {arriveTangent=0.834, interpMode=RCIM_Cubic, leaveTangent=0.834, tangentMode=RCTM_User, time=0, value=0}
         {arriveTangent=2.6505, interpMode=RCIM_Cubic, leaveTangent=2.6505, tangentMode=RCTM_User, time=0.3, value=1}
         {interpMode=RCIM_Cubic, tangentMode=RCTM_User, time=0.6, value=0}
        ]}}})
    Scale RGB=&MakeVectorFromLinearColorRGB(Color=&ColorFromCurve(ColorCurve=DataInterface() {propertyValues {
        AlphaCurve {keys [<at 0 1 RCIM_Cubic>]}
        BlueCurve {keys [{arriveTangent=-1.3478, interpMode=RCIM_Cubic, time=0, value=2}, <at 0.6962 0.089 RCIM_Cubic>]}
        GreenCurve {keys [{arriveTangent=0.0194, interpMode=RCIM_Cubic, time=0, value=0.0832}, <at 0.6962 0.0464 RCIM_Cubic>]}
        RedCurve {keys [{arriveTangent=-0.1379, interpMode=RCIM_Cubic, time=0, value=1.0217}, <at 0.6962 0.0511 RCIM_Cubic>]}
       }}))
    ScaleA=true
    ScaleRGB=true
   )
   /SetVariables(Particles.LightColor=&MultiplyLinearColor(
     Linear Color=$Particles.Color
     Scale Factor=&MakeVector4_FromFloat(Float=&FloatFromCurve(
       FloatCurve=DataInterface() {propertyValues {Curve {keys [
           {arriveTangent=0.834, interpMode=RCIM_Cubic, leaveTangent=0.834, tangentMode=RCTM_User, time=0, value=0}
           {arriveTangent=2.6505, interpMode=RCIM_Cubic, leaveTangent=2.6505, tangentMode=RCTM_User, time=0.5, value=1}
           {interpMode=RCIM_Cubic, tangentMode=RCTM_User, time=0.7, value=0}
          ]}}}
       Scale Curve=7
      ))
    )) {isSetParameters=true, module=SetVariables_#2}
   &DynamicMaterialParameters(
    Index 0 Param 1=&Lerp_Float(
     A=1
     B=4
     Alpha=&FloatFromCurve(FloatCurve=DataInterface() {propertyValues {Curve {keys [
          {interpMode=RCIM_Cubic, tangentMode=RCTM_User, time=0, value=0}
          {arriveTangent=0.0683, interpMode=RCIM_Cubic, leaveTangent=0.0683, tangentMode=RCTM_User, time=1, value=1}
         ]}}})
    )
    Index 0 Param 2=1
    Module.Index0Param0Write=true
    Module.Index0Param1Write=true
    Module.Index0Param2Write=false
    Module.Index0Param3Write=false
   ) {enabled=false, module=DynamicMaterialParameters001}
   /SetVariables(
    Particles.ColorDark:LinearColor=&ColorFromCurve(ColorCurve=DataInterface() {propertyValues {
       AlphaCurve {keys [<at 0 1 RCIM_Cubic>,<at 1 0 RCIM_Cubic>]}
       BlueCurve {keys [<at 0 0.01 RCIM_Cubic>]}
       GreenCurve {keys [<at 0 0.007 RCIM_Cubic>]}
       RedCurve {keys [<at 0 0.008 RCIM_Cubic>]}
      }})
    Particles.SpriteSizeDark:Vector2f=&Vector2DFromFloat(Value=&Lerp_Float(
      A=0
      B=60
      Alpha=&FloatFromCurve(FloatCurve=DataInterface() {propertyValues {Curve {keys [
           {interpMode=RCIM_Cubic, tangentMode=RCTM_User, time=0, value=0}
           {interpMode=RCIM_Cubic, tangentMode=RCTM_User, time=0.1, value=1}
           {interpMode=RCIM_Cubic, tangentMode=RCTM_User, time=0.7, value=0}
          ]}}})
     ))
   ) {isSetParameters=true, module=SetVariables_#3}
   &AccelerationForce(Acceleration=<0 0 -100>) {enabled=false}
   &SolveForcesAndVelocity(Clamp Velocity=false, Limit Acceleration=false)
  ]
  rendererClasses [&NiagaraSpriteRendererProperties]
  renderers [
   {rendererClass=&NiagaraSpriteRendererProperties}
   {rendererClass=&NiagaraSpriteRendererProperties, rendererIndex=1}
  ]
 }
 FlareVortex {
  emitterUpdateScript [
   &EmitterState(Life Cycle Mode=::NewEnumerator1, Loop Behavior=::NewEnumerator1, Loop Delay=0.12, Loop Duration=1, UseLoopDelay=true)
   &SpawnBurst_Instantaneous(Spawn Group=1, Use Spawn Probability=false, UseLoopCountLimit=false) {module=SpawnBurst_Instantaneous001}
   &SpawnBurst_Instantaneous(Spawn Time=0.08, Use Spawn Probability=false, UseLoopCountLimit=false)
  ]
  particleSpawnScript [
   /SetVariables(Particles.VisibilityTag=&MaskIntBySpawnGroup(Masked Int=0, Passthrough Int=1, Spawn Group Mask=0)) {isSetParameters=true}
   &InitializeParticle(
    Lifetime=&MaskFloatBySpawnGroup(Masked Float=0.85, Passthrough Float=1.1)
    Position Offset=$System.HandFXOffset
    Sprite Size Mode=::NewEnumerator3
    Sprite UV Mode=::NewEnumerator4
    Sprite UV Scale:Vector2f=&MaskVector2dBySpawnGroup(Masked Vector2D=<-1 1>, Passthrough Vector2D=<1 1>, Spawn Group Mask=0)
    Uniform Sprite Size=30
    UsePositionOffset=true
    Write Material Random=true
   )
   &CameraOffset(Camera Offset Amount=&MaskFloatBySpawnGroup(Masked Float=20, Passthrough Float=0))
  ]
  particleUpdateScript [
   &ParticleState()
   &ScaleSpriteSize(
    Scale Factor=&MakeVector2D(
     X=&FloatFromCurve(FloatCurve=DataInterface() {propertyValues {Curve {keys [
          {interpMode=RCIM_Cubic, tangentMode=RCTM_User, time=0.15, value=0.2}
          {arriveTangent=7.8407, interpMode=RCIM_Cubic, leaveTangent=7.8407, tangentMode=RCTM_User, time=0.8, value=3}
         ]}}})
     Y=&FloatFromCurve(FloatCurve=DataInterface() {propertyValues {Curve {keys [
          {interpMode=RCIM_Cubic, tangentMode=RCTM_User, time=0.1, value=1}
          {arriveTangent=-3.6877, interpMode=RCIM_Cubic, leaveTangent=-3.6877, tangentMode=RCTM_User, time=0.6, value=0.05}
         ]}}})
    )
    Scale Sprite Size Mode=::NewEnumerator2
   ) {module=ScaleSpriteSize001}
   &ScaleSpriteSize(
    Uniform Curve Scale=2
    Uniform Curve Sprite Scale=DataInterface() {propertyValues {Curve {keys [
        {interpMode=RCIM_Cubic, tangentMode=RCTM_User, time=0, value=0.2}
        {arriveTangent=-0.9517, interpMode=RCIM_Cubic, leaveTangent=-0.9517, tangentMode=RCTM_User, time=0.2, value=1}
        {arriveTangent=-0.1086, interpMode=RCIM_Cubic, leaveTangent=-0.1086, tangentMode=RCTM_User, time=1, value=0}
       ]}}}
   )
   &ScaleColor(
    Scale Alpha=&FloatFromCurve(FloatCurve=DataInterface() {propertyValues {Curve {keys [
         {arriveTangent=0.834, interpMode=RCIM_Cubic, leaveTangent=0.834, tangentMode=RCTM_User, time=0, value=0}
         {arriveTangent=2.6505, interpMode=RCIM_Cubic, leaveTangent=2.6505, tangentMode=RCTM_User, time=0.3, value=1}
         {interpMode=RCIM_Cubic, tangentMode=RCTM_User, time=0.7, value=0}
        ]}}})
    Scale RGB=&MakeVectorFromLinearColorRGB(Color=&ColorFromCurve(ColorCurve=DataInterface() {propertyValues {
        AlphaCurve {keys [<at 0 1 RCIM_Cubic>]}
        BlueCurve {keys [{arriveTangent=-1.3478, interpMode=RCIM_Cubic, time=0, value=2}, <at 0.6962 0.089 RCIM_Cubic>]}
        GreenCurve {keys [{arriveTangent=0.0194, interpMode=RCIM_Cubic, time=0, value=0.0832}, <at 0.6962 0.0464 RCIM_Cubic>]}
        RedCurve {keys [{arriveTangent=-0.1379, interpMode=RCIM_Cubic, time=0, value=1.0217}, <at 0.6962 0.0511 RCIM_Cubic>]}
       }}))
    ScaleA=true
    ScaleRGB=true
   )
   /SetVariables(Particles.LightColor=&MultiplyLinearColor(
     Linear Color=$Particles.Color
     Scale Factor=&MakeVector4_FromFloat(Float=&FloatFromCurve(
       FloatCurve=DataInterface() {propertyValues {Curve {keys [
           {arriveTangent=0.834, interpMode=RCIM_Cubic, leaveTangent=0.834, tangentMode=RCTM_User, time=0, value=0}
           {arriveTangent=2.6505, interpMode=RCIM_Cubic, leaveTangent=2.6505, tangentMode=RCTM_User, time=0.45, value=1}
           {interpMode=RCIM_Cubic, tangentMode=RCTM_User, time=0.75, value=0}
          ]}}}
       Scale Curve=7
      ))
    )) {isSetParameters=true, module=SetVariables_#2}
   /SetVariables(
    Particles.ColorDark:LinearColor=&ColorFromCurve(ColorCurve=DataInterface() {propertyValues {
       AlphaCurve {keys [<at 0 1 RCIM_Cubic>,<at 1 0 RCIM_Cubic>]}
       BlueCurve {keys [<at 0 0.01 RCIM_Cubic>]}
       GreenCurve {keys [<at 0 0.007 RCIM_Cubic>]}
       RedCurve {keys [<at 0 0.008 RCIM_Cubic>]}
      }})
    Particles.SpriteSizeDark:Vector2f=&Vector2DFromFloat(Value=&Lerp_Float(
      A=0
      B=60
      Alpha=&FloatFromCurve(FloatCurve=DataInterface() {propertyValues {Curve {keys [
           {interpMode=RCIM_Cubic, tangentMode=RCTM_User, time=0, value=0}
           {interpMode=RCIM_Cubic, tangentMode=RCTM_User, time=0.1, value=1}
           {interpMode=RCIM_Cubic, tangentMode=RCTM_User, time=0.7, value=0}
          ]}}})
     ))
   ) {isSetParameters=true, module=SetVariables_#3}
   &SolveForcesAndVelocity(Clamp Velocity=false, Limit Acceleration=false) {enabled=false}
   /SetVariables(
    Particles.ColorVortex:LinearColor=&MakeLinearColorFromVectorAndFloat(
     "Vector (RGB)"=&MakeVectorFromLinearColorRGB(Color=&MultiplyLinearColor(
       Linear Color=&ColorFromCurve(ColorCurve=DataInterface() {propertyValues {
          AlphaCurve {keys [<at 0 1 RCIM_Cubic>]}
          BlueCurve {keys [<at 0.2204 1 RCIM_Cubic>,<at 0.7096 0.03 RCIM_Cubic>]}
          GreenCurve {keys [<at 0.2204 1 RCIM_Cubic>,<at 0.7096 0.0052 RCIM_Cubic>]}
          RedCurve {keys [<at 0.2204 1 RCIM_Cubic>,<at 0.7096 0.0097 RCIM_Cubic>]}
         }})
       Scale Factor=&MakeVector4_FromLinearColor(LinearColor=<rgba 1.7693 0.3 3 1>)
      ))
     "Float (Alpha)"=&FloatFromCurve(FloatCurve=DataInterface() {propertyValues {Curve {keys [
          <at 0 0 RCIM_Cubic>
          {arriveTangent=-0.0178, interpMode=RCIM_Cubic, leaveTangent=-0.0178, tangentMode=RCTM_User, time=0.2, value=1}
          {interpMode=RCIM_Cubic, tangentMode=RCTM_User, time=0.4, value=1}
          {interpMode=RCIM_Cubic, tangentMode=RCTM_User, time=1, value=0}
         ]}}})
    )
    Particles.DynamicMaterialParameterVortex:Vector4f=&MakeVector4(
     X=&Lerp_Float(
      A=0
      B=1
      Alpha=&FloatFromCurve(FloatCurve=DataInterface() {propertyValues {Curve {keys [
           {interpMode=RCIM_Cubic, tangentMode=RCTM_User, time=0.5, value=0}
           {arriveTangent=2.5519, interpMode=RCIM_Cubic, leaveTangent=2.5519, tangentMode=RCTM_User, time=1, value=1}
          ]}}})
     )
     Y=&Lerp_Float(
      A=1
      B=2
      Alpha=&FloatFromCurve(FloatCurve=DataInterface() {propertyValues {Curve {keys [<at 0 0>,<at 1 1>]}}})
     )
     Z=&Lerp_Float(
      A=-0.15
      B=0.5
      Alpha=&FloatFromCurve(FloatCurve=DataInterface() {propertyValues {Curve {keys [
           {interpMode=RCIM_Cubic, tangentMode=RCTM_User, time=0, value=0}
           {arriveTangent=2.4894, interpMode=RCIM_Cubic, leaveTangent=2.4894, tangentMode=RCTM_User, time=1, value=1}
          ]}}})
     )
     W=&Lerp_Float(
      A=0.3
      B=1
      Alpha=&FloatFromCurve(FloatCurve=DataInterface() {propertyValues {Curve {keys [
           {arriveTangent=0.4065, interpMode=RCIM_Cubic, leaveTangent=0.4065, tangentMode=RCTM_User, time=0, value=0}
           {arriveTangent=2.7807, interpMode=RCIM_Cubic, leaveTangent=2.7807, tangentMode=RCTM_User, time=0.8, value=1}
          ]}}})
     )
    )
    Particles.SpriteSizeVortex:Vector2f=&Vector2DFromFloat(Value=&Multiply_Float(
      A=55
      B=&FloatFromCurve(FloatCurve=DataInterface() {propertyValues {Curve {keys [
           {arriveTangent=-0.4127, interpMode=RCIM_Cubic, leaveTangent=-0.4127, tangentMode=RCTM_User, time=0, value=1}
           {arriveTangent=-0.3575, interpMode=RCIM_Cubic, leaveTangent=-0.3575, tangentMode=RCTM_User, time=0.5, value=0.3}
          ]}}})
     ))
   ) {isSetParameters=true}
   /SetVariables(
    Particles.ColorVortexDark:LinearColor=&MultiplyLinearColor(
     Linear Color=<rgba 0 0 0 1>
     Scale Factor=&MakeFloat4FromVectorAndFloat(
      "Vector (RGB)"=<1 1 1>
      "Float (Alpha)"=&FloatFromCurve(
       FloatCurve=DataInterface() {propertyValues {Curve {keys [
           <at 0 0 RCIM_Cubic>
           {arriveTangent=-0.9414, interpMode=RCIM_Cubic, leaveTangent=-0.9414, tangentMode=RCTM_User, time=0.25, value=1}
           {interpMode=RCIM_Cubic, tangentMode=RCTM_User, time=1, value=0}
          ]}}}
       Scale Curve=1.1
      )
     )
    )
    Particles.DynamicMaterialParameterDark:Vector4f=&MakeVector4(
     X=&Lerp_Float(
      A=0
      B=1
      Alpha=&FloatFromCurve(FloatCurve=DataInterface() {propertyValues {Curve {keys [
           {interpMode=RCIM_Cubic, tangentMode=RCTM_User, time=0.4, value=0}
           {arriveTangent=0.4179, interpMode=RCIM_Cubic, leaveTangent=0.4179, tangentMode=RCTM_User, time=0.9, value=1}
          ]}}})
     )
     Y=&Lerp_Float(
      A=3
      B=1
      Alpha=&FloatFromCurve(FloatCurve=DataInterface() {propertyValues {Curve {keys [<at 0 0>,<at 0.5 1>]}}})
     )
     Z=&Lerp_Float(
      A=-10
      B=-5
      Alpha=&FloatFromCurve(FloatCurve=DataInterface() {propertyValues {Curve {keys [
           {interpMode=RCIM_Cubic, tangentMode=RCTM_User, time=0, value=0}
           {arriveTangent=2.4894, interpMode=RCIM_Cubic, leaveTangent=2.4894, tangentMode=RCTM_User, time=1, value=1}
          ]}}})
     )
     W=1
    )
    Particles.SpriteSizeDarkVortex:Vector2f=&Multiply_Vector2D_ByFloat(
     Float=&FloatFromCurve(FloatCurve=DataInterface() {propertyValues {Curve {keys [
          {arriveTangent=-5.7104, interpMode=RCIM_Cubic, leaveTangent=-5.7104, tangentMode=RCTM_User, time=0.5, value=1}
          {arriveTangent=-0.4968, interpMode=RCIM_Cubic, leaveTangent=-0.4968, tangentMode=RCTM_User, time=1, value=0}
         ]}}})
     Vector2D=$Particles.SpriteSizeVortex
    )
   ) {isSetParameters=true}
  ]
  rendererClasses [&NiagaraSpriteRendererProperties]
  renderers [
   {rendererClass=&NiagaraSpriteRendererProperties}
   {rendererClass=&NiagaraSpriteRendererProperties, rendererIndex=1}
   {rendererClass=&NiagaraSpriteRendererProperties, rendererIndex=2}
   {rendererClass=&NiagaraSpriteRendererProperties, rendererIndex=3}
  ]
 }
 MaskTrail {
  emitterSpawnScript [/SetVariables(Emitter.RibbonRandom:NiagaraFloat=&RandomRangeFloat(Minimum=0, Maximum=1)) {isSetParameters=true}]
  emitterUpdateScript [
   &EmitterState(Life Cycle Mode=::NewEnumerator1, Loop Behavior=::NewEnumerator1, Loop Delay=0.05, Loop Duration=1, UseLoopDelay=false)
   &SpawnBurst_Instantaneous(Spawn Count=3, Use Spawn Probability=false, UseLoopCountLimit=false)
  ]
  particleSpawnScript [
   &PartitionParticles(ID to Partition:NiagaraInt32=&ReturnExecIndex(), Number Of Partitions=3)
   &InitializeParticle(
    Lifetime=1.55
    Ribbon Width=20
    Ribbon Width Mode=::NewEnumerator1
    Sprite Rotation Mode=::NewEnumerator2
    Sprite Size Mode=::NewEnumerator1
    Sprite UV Mode=::NewEnumerator3
    Uniform Sprite Size Max=20
    Uniform Sprite Size Min=14
    UsePositionOffset=false
    Write Material Random=true
   )
  ]
  particleUpdateScript [
   &ParticleState()
   &FN_BoneFlow(
    Enable Position Offsets=true
    Height:NiagaraFloat=&SelectFloatFromArray(Direct Array Index=$Particles.PartitionParticles.Partition, Float Selection Array=DataInterface() {propertyValues {FloatData [0,0.5,1]}})
    Offset End:Vector3f=<-12 9 0>
    Offset Mid:Vector3f=&AddVectorToPosition(
     Position=<-40 -30 34>
     Vector=&MakeVector(
      X=&Lerp_Float(
       A=3
       B=0
       Alpha=&FloatFromCurve(FloatCurve=DataInterface() {propertyValues {Curve {keys [{interpMode=RCIM_Cubic, tangentMode=RCTM_User, time=0, value=0}, {interpMode=RCIM_Cubic, tangentMode=RCTM_User, time=1, value=1}]}}})
      )
      Y=&Lerp_Float(
       A=-40
       B=-20
       Alpha=&FloatFromCurve(FloatCurve=DataInterface() {propertyValues {Curve {keys [
            {arriveTangent=0.3142, interpMode=RCIM_Cubic, leaveTangent=0.3142, tangentMode=RCTM_User, time=0, value=0}
            {arriveTangent=0.089, interpMode=RCIM_Cubic, leaveTangent=0.089, tangentMode=RCTM_User, time=0.3, value=1}
           ]}}})
      )
      Z=&Lerp_Float(
       A=-10
       B=0
       Alpha=&FloatFromCurve(FloatCurve=DataInterface() {propertyValues {Curve {keys [
            {arriveTangent=0.3328, interpMode=RCIM_Cubic, leaveTangent=0.3328, tangentMode=RCTM_User, time=0, value=0}
            {interpMode=RCIM_Cubic, tangentMode=RCTM_User, time=0.2, value=1}
           ]}}})
      )
     )
    )
    Offset Start:Vector3f=&AddVectorToPosition(
     Position=<0 20 0>
     Vector=&MakeVector(
      X=&Lerp_Float(
       A=5
       B=-5
       Alpha=&FloatFromCurve(FloatCurve=DataInterface() {propertyValues {Curve {keys [{interpMode=RCIM_Cubic, tangentMode=RCTM_User, time=0, value=0}, {interpMode=RCIM_Cubic, tangentMode=RCTM_User, time=0.5, value=1}]}}})
      )
      Y=0
      Z=0
     )
    )
    SkeletalMesh=DataInterface() {propertyValues {
      DefaultMesh=null
      ExcludeBoneName=root
      OptionalComponentClass=null
      PreviewMesh=/Game/Characters/Player/Female/Medium/Bodies/F_MED_Default_Soldier/Meshes/F_MED_Default_Soldier
      SkinningMode=SkinOnTheFly
      SoftSourceActor=null
      SourceMode=Default
      UvSetIndex=0
      WholeMeshLOD=-1
      bExcludeBone=false
      bReadDeformedGeometry=true
      bRequireCurrentFrameData=true
      ComponentTags []
      FilteredBones [head,lowerarm_r,hand_r]
      FilteredSockets []
      MeshUserParameter {parameter {name=None, typeDefHandle {registeredTypeIndex=1}, varData []}}
      SamplingRegions []
     }}
   )
   &SolveForcesAndVelocity(Clamp Velocity=false, Limit Acceleration=false)
   &FN_RibbonMaterial(
    Color=&Multiply_LinearColorByFloat(
     Linear Color=&SelectLinearColorFromArray(
      Array Sampling Mode=::NewEnumerator1
      Direct Array Index=$Particles.PartitionParticles.Partition
      Color Selection Array=DataInterface() {propertyValues {ColorData [<rgba 0.1917 0.03 1 1>, <rgba 0.1437 0.025 0.5 1>, <rgba 0.4225 0.01 1 1>]}}
      Random Seed=?
     )
     Float=&Lerp_Float(
      A=0.5
      B=2
      Alpha=&FloatFromCurve(FloatCurve=DataInterface() {propertyValues {Curve {keys [
           {interpMode=RCIM_Cubic, tangentMode=RCTM_User, time=0.1, value=0}
           {arriveTangent=-3.6405, interpMode=RCIM_Cubic, leaveTangent=-3.6405, tangentMode=RCTM_User, time=0.26, value=0.6}
           {interpMode=RCIM_Cubic, tangentMode=RCTM_User, time=0.4, value=0}
           {arriveTangent=5.8942, interpMode=RCIM_Cubic, leaveTangent=5.8942, tangentMode=RCTM_User, time=0.7, value=1}
          ]}}})
     )
    )
    Scale Alpha:NiagaraFloat=&Lerp_Float(
     A=1
     B=10
     Alpha=&FloatFromCurve(FloatCurve=DataInterface() {propertyValues {Curve {keys [<at 0 0>,<at 0.5 1>]}}})
    )
   )
   &DynamicMaterialParameters(
    Index 0 Param 1=0.6
    Index 0 Param 2=&Lerp_Float(
     A=0.7
     B=-0.8
     Alpha=&FloatFromCurve(FloatCurve=DataInterface() {propertyValues {Curve {keys [
          {arriveTangent=2.6464, interpMode=RCIM_Cubic, leaveTangent=2.6464, tangentMode=RCTM_User, time=0, value=0}
          {interpMode=RCIM_Cubic, tangentMode=RCTM_User, time=1, value=1}
         ]}}})
    )
    Index 0 Param 3=&Lerp_Float(
     A=0.7
     B=1.6
     Alpha=&FloatFromCurve(FloatCurve=DataInterface() {propertyValues {Curve {keys [{interpMode=RCIM_Cubic, tangentMode=RCTM_User, time=0, value=0}, {interpMode=RCIM_Cubic, tangentMode=RCTM_User, time=1, value=1}]}}})
    )
    Index 0 Param 4=&Lerp_Float(
     A=0.3
     B=0.5
     Alpha=&FloatFromCurve(FloatCurve=DataInterface() {propertyValues {Curve {keys [<at 0 0>,<at 0.7 1>]}}})
    )
    Module.Index0Param0Write=true
    Module.Index0Param1Write=true
    Module.Index0Param2Write=true
    Module.Index0Param3Write=true
   )
   /SetVariables(Particles.RibbonWidth=&Multiply_Float(
     A=&FloatFromCurve(FloatCurve=DataInterface() {propertyValues {Curve {keys [
          {interpMode=RCIM_Cubic, tangentMode=RCTM_User, time=0.07, value=0.2}
          {arriveTangent=-1.2665, interpMode=RCIM_Cubic, leaveTangent=-1.2665, tangentMode=RCTM_User, time=0.25, value=1}
          <at 0.6 0.8 RCIM_Cubic>
          <at 1 1 RCIM_Cubic>
         ]}}})
     B=&Multiply_Float(
      A=&Lerp_Float(A=18, B=22, Alpha=$Emitter.RibbonRandom)
      B=&SelectFloatFromArray(Direct Array Index=$Particles.PartitionParticles.Partition, Float Selection Array=DataInterface() {propertyValues {FloatData [2,1.6,0.5]}})
     )
    )) {isSetParameters=true}
   /SetVariables(Particles.ColorBright:LinearColor=&Multiply_LinearColorByFloat(
     Linear Color=$Particles.Color
     Float=&Lerp_Float(
      A=0.7
      B=2
      Alpha=&FloatFromCurve(FloatCurve=DataInterface() {propertyValues {Curve {keys [<at 0 0 RCIM_Cubic>, {interpMode=RCIM_Cubic, tangentMode=RCTM_User, time=0.3, value=1}]}}})
     )
    )) {isSetParameters=true}
   /SetVariables(
    Particles.ColorDark:LinearColor=&MultiplyLinearColor(
     Linear Color=$Particles.Color
     Scale Factor=&MakeFloat4FromVectorAndFloat(
      "Vector (RGB)"=<0.01 0.01 0.03>
      "Float (Alpha)"=&Multiply_Float(
       A=0.6
       B=&FloatFromCurve(FloatCurve=DataInterface() {propertyValues {Curve {keys [<at 0.2 1>,<at 0.6 0>]}}})
      )
     )
    )
    Particles.RibbonWidthDark:NiagaraFloat=&Multiply_Float(A=20, B=1.7)
   ) {isSetParameters=true}
   /SetVariables(Particles.DynamicMaterialParameterDark:Vector4f=&MakeVector4(
     X=-0.1
     Y=&Lerp_Float(
      A=-0.1
      B=0
      Alpha=&FloatFromCurve(FloatCurve=DataInterface() {propertyValues {Curve {keys [<at 0 0>,<at 0.7 1 RCIM_Cubic>]}}})
     )
     Z=&Lerp_Float(
      A=0.1
      B=0.2
      Alpha=&FloatFromCurve(FloatCurve=DataInterface() {propertyValues {Curve {keys [
           {interpMode=RCIM_Cubic, tangentMode=RCTM_User, time=0, value=0}
           {arriveTangent=1.5583, interpMode=RCIM_Cubic, leaveTangent=1.5583, tangentMode=RCTM_User, time=1, value=1}
          ]}}})
     )
     W=&FloatFromCurve(FloatCurve=DataInterface() {propertyValues {Curve {keys [
          {arriveTangent=0.1393, interpMode=RCIM_Cubic, leaveTangent=0.1393, tangentMode=RCTM_User, time=0.1, value=0}
          {arriveTangent=0.3311, interpMode=RCIM_Cubic, leaveTangent=0.3311, tangentMode=RCTM_User, time=0.7, value=1}
         ]}}})
    )) {isSetParameters=true}
  ]
  rendererClasses [&NiagaraRibbonRendererProperties, &NiagaraSpriteRendererProperties]
  renderers [
   {rendererClass=&NiagaraRibbonRendererProperties}
   {rendererClass=&NiagaraRibbonRendererProperties, rendererIndex=1}
   {rendererClass=&NiagaraRibbonRendererProperties, rendererIndex=2}
   {rendererClass=&NiagaraSpriteRendererProperties, rendererIndex=3}
  ]
 }
 Vortex {
  bEnabled=false
  emitterUpdateScript [
   &EmitterState(Life Cycle Mode=::NewEnumerator1, Loop Behavior=::NewEnumerator1, Loop Delay=0.12, Loop Duration=2, UseLoopDelay=true)
   &SpawnBurst_Instantaneous(Use Spawn Probability=false, UseLoopCountLimit=false)
  ]
  particleSpawnScript [&InitializeParticle(
    Color=<rgba 1.7693 0.3 3 1>
    Lifetime=1.3
    Position Offset=$System.HandFXOffset
    Sprite Rotation Mode=::NewEnumerator2
    Sprite Size Mode=::NewEnumerator3
    Uniform Sprite Size=50
    UsePositionOffset=true
    Write Material Random=true
   )]
  particleUpdateScript [
   &ParticleState()
   &ScaleSpriteSize(Uniform Curve Sprite Scale=DataInterface() {propertyValues {Curve {keys [
        {arriveTangent=12.5362, interpMode=RCIM_Cubic, leaveTangent=12.5362, tangentMode=RCTM_User, time=0, value=0.3}
        {arriveTangent=-0.4127, interpMode=RCIM_Cubic, leaveTangent=-0.4127, tangentMode=RCTM_User, time=0.1, value=1}
        {arriveTangent=-0.3575, interpMode=RCIM_Cubic, leaveTangent=-0.3575, tangentMode=RCTM_User, time=0.4, value=0.3}
       ]}}})
   &InheritSourceMovement(ApplyForce=false, ApplyPosition=true, ApplyPositionFalloff=false) {enabled=false}
   &SolveForcesAndVelocity(Clamp Velocity=false, Limit Acceleration=false)
   &DynamicMaterialParameters(
    Index 0 Param 1=&Lerp_Float(
     A=0
     B=1
     Alpha=&FloatFromCurve(FloatCurve=DataInterface() {propertyValues {Curve {keys [
          {interpMode=RCIM_Cubic, tangentMode=RCTM_User, time=0.5, value=0}
          {arriveTangent=2.5519, interpMode=RCIM_Cubic, leaveTangent=2.5519, tangentMode=RCTM_User, time=1, value=1}
         ]}}})
    )
    Index 0 Param 2=&Lerp_Float(
     A=1
     B=2
     Alpha=&FloatFromCurve(FloatCurve=DataInterface() {propertyValues {Curve {keys [<at 0 0>,<at 1 1>]}}})
    )
    Index 0 Param 3=&Lerp_Float(
     A=-0.15
     B=0.5
     Alpha=&FloatFromCurve(FloatCurve=DataInterface() {propertyValues {Curve {keys [
          {interpMode=RCIM_Cubic, tangentMode=RCTM_User, time=0, value=0}
          {arriveTangent=2.4894, interpMode=RCIM_Cubic, leaveTangent=2.4894, tangentMode=RCTM_User, time=1, value=1}
         ]}}})
    )
    Index 0 Param 4=&Lerp_Float(
     A=0.3
     B=1
     Alpha=&FloatFromCurve(FloatCurve=DataInterface() {propertyValues {Curve {keys [
          {arriveTangent=0.4065, interpMode=RCIM_Cubic, leaveTangent=0.4065, tangentMode=RCTM_User, time=0, value=0}
          {arriveTangent=2.7807, interpMode=RCIM_Cubic, leaveTangent=2.7807, tangentMode=RCTM_User, time=0.8, value=1}
         ]}}})
    )
    Module.Index0Param0Write=true
    Module.Index0Param1Write=true
    Module.Index0Param2Write=true
    Module.Index0Param3Write=true
   )
   &ScaleColor(
    Scale Alpha=&FloatFromCurve(FloatCurve=DataInterface() {propertyValues {Curve {keys [
         <at 0 0 RCIM_Cubic>
         {arriveTangent=-0.0178, interpMode=RCIM_Cubic, leaveTangent=-0.0178, tangentMode=RCTM_User, time=0.2, value=1}
         {interpMode=RCIM_Cubic, tangentMode=RCTM_User, time=0.4, value=1}
         {interpMode=RCIM_Cubic, tangentMode=RCTM_User, time=1, value=0}
        ]}}})
    Scale RGB=&MakeVectorFromLinearColorRGB(Color=&ColorFromCurve(ColorCurve=DataInterface() {propertyValues {
        AlphaCurve {keys [<at 0 1 RCIM_Cubic>]}
        BlueCurve {keys [<at 0.2204 1 RCIM_Cubic>,<at 0.7096 0.03 RCIM_Cubic>]}
        GreenCurve {keys [<at 0.2204 1 RCIM_Cubic>,<at 0.7096 0.0052 RCIM_Cubic>]}
        RedCurve {keys [<at 0.2204 1 RCIM_Cubic>,<at 0.7096 0.0097 RCIM_Cubic>]}
       }}))
    ScaleA=true
    ScaleRGB=true
   )
   /SetVariables(
    Particles.ColorDark:LinearColor=&MultiplyLinearColor(
     Linear Color=<rgba 0 0 0 1>
     Scale Factor=&MakeFloat4FromVectorAndFloat(
      "Vector (RGB)"=<1 1 1>
      "Float (Alpha)"=&FloatFromCurve(
       FloatCurve=DataInterface() {propertyValues {Curve {keys [
           <at 0 0 RCIM_Cubic>
           {arriveTangent=-0.9414, interpMode=RCIM_Cubic, leaveTangent=-0.9414, tangentMode=RCTM_User, time=0.25, value=1}
           {interpMode=RCIM_Cubic, tangentMode=RCTM_User, time=1, value=0}
          ]}}}
       Scale Curve=1.1
      )
     )
    )
    Particles.DynamicMaterialParameterDark:Vector4f=&MakeVector4(
     X=&Lerp_Float(
      A=0
      B=1
      Alpha=&FloatFromCurve(FloatCurve=DataInterface() {propertyValues {Curve {keys [
           {interpMode=RCIM_Cubic, tangentMode=RCTM_User, time=0.4, value=0}
           {arriveTangent=2.5519, interpMode=RCIM_Cubic, leaveTangent=2.5519, tangentMode=RCTM_User, time=1, value=1}
          ]}}})
     )
     Y=&Lerp_Float(
      A=1
      B=3
      Alpha=&FloatFromCurve(FloatCurve=DataInterface() {propertyValues {Curve {keys [<at 0 0>,<at 1 1>]}}})
     )
     Z=&Lerp_Float(
      A=-10
      B=-5
      Alpha=&FloatFromCurve(FloatCurve=DataInterface() {propertyValues {Curve {keys [
           {interpMode=RCIM_Cubic, tangentMode=RCTM_User, time=0, value=0}
           {arriveTangent=2.4894, interpMode=RCIM_Cubic, leaveTangent=2.4894, tangentMode=RCTM_User, time=1, value=1}
          ]}}})
     )
     W=1
    )
    Particles.SpriteSizeDark:Vector2f=&Multiply_Vector2D_ByFloat(Float=1, Vector2D=$Particles.SpriteSize)
   ) {isSetParameters=true}
  ]
  rendererClasses [&NiagaraSpriteRendererProperties]
  renderers [
   {rendererClass=&NiagaraSpriteRendererProperties}
   {rendererClass=&NiagaraSpriteRendererProperties, rendererIndex=1}
  ]
 }
}
summary {
 systemName=NS_VoidMask_Reload
 emitters [
  {bEnabled=false, emitterName=Flare, rendererClasses [&NiagaraSpriteRendererProperties]}
  {emitterName=MaskTrail, rendererClasses [&NiagaraRibbonRendererProperties, &NiagaraSpriteRendererProperties]}
  {emitterName=FakeLight, rendererClasses [&NiagaraSpriteRendererProperties]}
  {emitterName=Copied_ChromaticRefraction, rendererClasses [&NiagaraSpriteRendererProperties]}
  {bEnabled=false, emitterName=Vortex, rendererClasses [&NiagaraSpriteRendererProperties]}
  {emitterName=FlareVortex, rendererClasses [&NiagaraSpriteRendererProperties]}
 ]
 userVariables [
  User.GlobalOpacity {type=NiagaraFloat, defaultValue=1, description=""}
  User.Offset Pos {type=NiagaraPosition, description="", defaultValue {}}
 ]
}
