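# Niagara burst effect for the Void Mask teleport spawn, purple by default.
# Short-lived streaks, flare smoke, a glow disk, rising wisps and a twisting absorb sprite form the burst, with User.Scalar sizing the streaks.
# System booleans IS_Smash and IsWanderingBoss switch or hue-shift colors in some emitters, including green Smash variants.
# FlareSmoke and TwistAbsorb write separate Dark and Flare particle attributes, layering several looks through multiple renderers.
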NiagaraSystem /DemonMaskVoid/Assets/VFX/NS_VoidMask_Teleport_Spawn
dependencies {
 usedDataInterfaces [
  &NiagaraDataInterfaceCurve
  &NiagaraDataInterfaceVectorCurve
  &NiagaraDataInterfaceArrayColor
  &NiagaraDataInterfaceArrayFloat2
  &NiagaraDataInterfaceVector2DCurve
  &NiagaraDataInterfaceColorCurve
 ]
 usedDynamicInputs [&SetBoolByFloatComparison]
 usedModules [
  &self:SystemSpawnScript.SystemScriptSource.SystemScriptGraph.NiagaraNodeAssignment_1.SetVariables_#1
  &SystemState
  &self:SystemSpawnScript.SystemScriptSource.SystemScriptGraph.NiagaraNodeAssignment_0.SetVariables_#2
 ]
}
emitters {
 DarkStreaks {
  emitterUpdateScript [
   &EmitterState(Life Cycle Mode=::NewEnumerator1, Loop Behavior=::NewEnumerator1, Loop Delay=0.02, Loop Duration=1, UseLoopDelay=false)
   &SpawnBurst_Instantaneous(Spawn Count=11, Use Spawn Probability=false, UseLoopCountLimit=false)
  ]
  particleSpawnScript [
   &PartitionParticles()
   /SetVariables(Particles.RandomFloat=&UniformRangedFloat(Minimum=0, Maximum=3)) {isSetParameters=true}
   &InitializeParticle(
    Color=&HueShiftLinearColor(
     Hue=&MakeCustomFloatFromBool(
      Bool=$System.IsWanderingBoss
      True Float=&MakeCustomFloatFromBool(Bool=$System.IS_Smash, True Float=0.45, False Float=-0.26)
      False Float=0
     )
     Saturation=&MakeCustomFloatFromBool(Bool=$System.IsWanderingBoss, True Float=0.7, False Float=1)
     Value=1
     Linear Color=&SelectLinearColorFromArray(
      Array Sampling Mode=::NewEnumerator1
      Direct Array Index=$Particles.PartitionParticles.Partition
      Color Selection Array=DataInterface() {propertyValues {ColorData [<rgba 0 0 0 1>, <rgba 0.4 0 1.2 1>]}}
      Random Seed=?
     )
    )
    Lifetime Max=0.33
    Lifetime Min=0.2
    Lifetime Mode=::NewEnumerator1
    Sprite Size=&Multiply_Vector2D_ByFloat(
     Float=&RandomRangeFloat(Minimum=0.5, Maximum=1)
     Vector2D=&SelectVector2DFromArray(
      Direct Array Index=$Particles.PartitionParticles.Partition
      Vector 2D Selection Array=DataInterface() {propertyValues {FloatData [<30 50>, <15 70>]}}
     )
    )
    Sprite Size Mode=::NewEnumerator4
    UsePositionOffset=false
    Write Material Random=true
   ) {module=InitializeParticle001}
   &ShapeLocation(
    Handle Radius=5
    Large Radius=40
    MaskSpawnGroup=false
    Non Uniform Scale=<1 1 0.1>
    Override Seed=false
    Shape Primitive=::NewEnumerator3
    UseOwnerScale=true
   ) {module=ShapeLocation001}
   &AddVelocity(
    Module.Use Falloff Distance=false
    Override Seed=false
    Velocity Mode=::NewEnumerator1
    Velocity Speed=&Lerp_Float(
     A=700
     B=1000
     Alpha=&FloatFromCurve(
      FloatCurve=DataInterface() {propertyValues {Curve {keys [{interpMode=RCIM_Cubic, tangentMode=RCTM_Break, time=0, value=0}, {interpMode=RCIM_Cubic, tangentMode=RCTM_Break, time=1, value=1}]}}}
      CurveIndex=&RandomRangeFloat(Minimum=0, Maximum=1)
      Scale Curve=0.5
     )
    )
   )
   &AddVelocity(Velocity=&RandomRangeVector(Minimum=<0 0 100>, Maximum=<0 0 500>)) {module=AddVelocity001}
   &SubUVAnimation(End Frame=3, SubUV Animation Mode:ENiagara_SubUVLookupMode=::NewEnumerator2)
  ]
  particleUpdateScript [
   &ParticleState() {module=UpdateAge}
   &ApplyOwnerScaleToAttributes()
   &Drag(Drag=6, Use Linear Drag=true, Use Rotational Drag=false)
   &SolveForcesAndVelocity(Clamp Velocity=false, Limit Acceleration=false)
   &ScaleSpriteSize(Scale Factor=&Vector2DFromFloat(Value=&FloatFromCurve(
      FloatCurve=DataInterface() {propertyValues {Curve {keys [
          {arriveTangent=1.6317, interpMode=RCIM_Cubic, leaveTangent=1.6317, tangentMode=RCTM_User, time=0, value=0.1}
          {arriveTangent=0.2754, interpMode=RCIM_Cubic, leaveTangent=0.2754, tangentMode=RCTM_User, time=1.0463, value=1}
         ]}}}
      Scale Curve=&Multiply_Float(A=2, B=$User.Scalar)
     )))
   &ScaleColor(
    Scale Alpha=&FloatFromCurve(
     FloatCurve=DataInterface() {propertyValues {Curve {keys [{interpMode=RCIM_Cubic, tangentMode=RCTM_Break, time=0, value=1}, {interpMode=RCIM_Cubic, tangentMode=RCTM_Break, time=1, value=0}]}}}
     Scale Curve=&RandomRangeFloat(Minimum=2, Maximum=2.5)
    )
    Scale RGB=&VectorFromFloat(Value=&Lerp_Float(
      A=15
      B=1
      Alpha=&FloatFromCurve(FloatCurve=DataInterface() {propertyValues {Curve {keys [{interpMode=RCIM_Cubic, tangentMode=RCTM_User, time=0.1, value=0}, {interpMode=RCIM_Cubic, tangentMode=RCTM_User, time=0.3, value=1}]}}})
     ))
    ScaleA=true
    ScaleRGB=true
   )
   &DynamicMaterialParameters(
    Index 0 Param 2=1
    Index 0 Param 3=1
    Index 0 Param 4=&FloatFromCurve(FloatCurve=DataInterface() {propertyValues {Curve {keys [<at 0.3 1>,<at 1 0.25>]}}})
    Module.Param0WriteEnabled=true
    Module.Param1WriteEnabled=true
    Module.Param2WriteEnabled=true
    Module.Param3WriteEnabled=true
   )
  ]
  rendererClasses [&NiagaraSpriteRendererProperties]
  renderers [{rendererClass=&NiagaraSpriteRendererProperties}]
 }
 FlareSmoke {
  emitterUpdateScript [
   &EmitterState(Life Cycle Mode=::NewEnumerator1, Loop Behavior=::NewEnumerator1, Loop Delay=0.04, Loop Duration=2, UseLoopDelay=false)
   &SpawnBurst_Instantaneous(Use Spawn Probability=false, UseLoopCountLimit=false)
   &SpawnRate(SpawnRate=90, Use Spawn Probability=false) {enabled=false}
  ]
  particleSpawnScript [
   &InitializeParticle(
    Lifetime=0.8
    Mass Mode=::NewEnumerator3
    Mesh Scale=&Multiply_VectorByFloat(Vector=<1 1 1>, Float=0.25)
    Mesh Scale Mode=::NewEnumerator4
    Sprite Size=<900 500>
    Sprite Size Mode=::NewEnumerator4
    UsePositionOffset=false
    Write Material Random=true
   )
   &InitialMeshOrientation(Mesh Axis To Orient=<0 0 1>, Mesh Orientation Mode=::NewEnumerator2, Use Rotation Vector=false)
   &SystemLocation(
    MaskSpawnGroup=false
    Offset=&AddVectorToPosition(Position=$System.ImpactOffset, Vector=<0 0 -3>)
    Offset Coordinate Space=::World
   )
   &CameraOffset(Camera Offset Amount=60)
   /SetVariables(Particles.CameraOffsetFlare=120) {isSetParameters=true}
  ]
  particleUpdateScript [
   &ParticleState()
   &ScaleSpriteSize(
    Scale Factor=&MakeVector2D(
     X=&FloatFromCurve(FloatCurve=DataInterface() {propertyValues {Curve {keys [{arriveTangent=27.5059, interpMode=RCIM_Cubic, leaveTangent=27.5059, tangentMode=RCTM_User, time=0.1, value=0.1}, <at 0.2 1.1>]}}})
     Y=&FloatFromCurve(FloatCurve=DataInterface() {propertyValues {Curve {keys [
          {arriveTangent=-55.1607, interpMode=RCIM_Cubic, leaveTangent=-55.1607, tangentMode=RCTM_User, time=0.02, value=3}
          {arriveTangent=-12.8632, interpMode=RCIM_Cubic, leaveTangent=-12.8632, tangentMode=RCTM_User, time=0.07, value=1.2}
          {interpMode=RCIM_Cubic, tangentMode=RCTM_User, time=0.3, value=0.2}
         ]}}})
    )
    Scale Sprite Size Mode=::NewEnumerator2
   )
   &ScaleMeshSize(Scale Factor=&MakeVector(
     X=&FloatFromCurve(FloatCurve=DataInterface() {propertyValues {Curve {keys [
          {arriveTangent=2.4641, interpMode=RCIM_Cubic, leaveTangent=2.4641, tangentMode=RCTM_User, time=0, value=0.5}
          {arriveTangent=0.1979, interpMode=RCIM_Cubic, leaveTangent=0.1979, tangentMode=RCTM_User, time=0.5, value=1}
         ]}}})
     Y=&FloatFromCurve(
      FloatCurve=DataInterface() {propertyValues {Curve {keys [
          {arriveTangent=-2.8472, interpMode=RCIM_Cubic, leaveTangent=-2.8472, tangentMode=RCTM_User, time=0, value=1}
          {interpMode=RCIM_Cubic, tangentMode=RCTM_User, time=0.6, value=0.1}
         ]}}}
      Scale Curve=2
     )
     Z=&FloatFromCurve(
      FloatCurve=DataInterface() {propertyValues {Curve {keys [
          {arriveTangent=-2.8472, interpMode=RCIM_Cubic, leaveTangent=-2.8472, tangentMode=RCTM_User, time=0, value=1}
          {interpMode=RCIM_Cubic, tangentMode=RCTM_User, time=0.6, value=0.1}
         ]}}}
      Scale Curve=2
     )
    )) {module=ScaleMeshSize001}
   &UpdateMeshOrientation(Rotation Rate=&Lerp_Float(
     A=1
     B=0.5
     Alpha=&FloatFromCurve(FloatCurve=DataInterface() {propertyValues {Curve {keys [
          {arriveTangent=2.5343, interpMode=RCIM_Cubic, leaveTangent=2.5343, tangentMode=RCTM_User, time=0, value=0}
          {interpMode=RCIM_Cubic, tangentMode=RCTM_User, time=1, value=1}
         ]}}})
    ))
   &ScaleMeshSize(Scale Factor=&Multiply_VectorByFloat(
     Vector=&MakeVector(
      X=&Lerp_Float(
       A=0.3
       B=0.7
       Alpha=&FloatFromCurve(FloatCurve=DataInterface() {propertyValues {Curve {keys [
            {arriveTangent=3.0241, interpMode=RCIM_Cubic, leaveTangent=3.0241, tangentMode=RCTM_User, time=0, value=0}
            {interpMode=RCIM_Cubic, tangentMode=RCTM_User, time=0.8, value=1}
           ]}}})
      )
      Y=&FloatFromCurve(FloatCurve=DataInterface() {propertyValues {Curve {keys [
           {arriveTangent=-2.2294, interpMode=RCIM_Cubic, leaveTangent=-2.2294, tangentMode=RCTM_User, time=0.2, value=1}
           {arriveTangent=-0.0663, interpMode=RCIM_Cubic, leaveTangent=-0.0663, tangentMode=RCTM_User, time=0.7, value=0.33}
          ]}}})
      Z=&FloatFromCurve(FloatCurve=DataInterface() {propertyValues {Curve {keys [
           {arriveTangent=-2.2294, interpMode=RCIM_Cubic, leaveTangent=-2.2294, tangentMode=RCTM_User, time=0.2, value=1}
           {arriveTangent=-0.0663, interpMode=RCIM_Cubic, leaveTangent=-0.0663, tangentMode=RCTM_User, time=0.7, value=0.33}
          ]}}})
     )
     Float=1
    )) {enabled=false}
   &ScaleColor(
    Scale Alpha=&FloatFromCurve(
     FloatCurve=DataInterface() {propertyValues {Curve {keys [
         {arriveTangent=0.4784, interpMode=RCIM_Cubic, leaveTangent=0.4784, tangentMode=RCTM_User, time=0, value=0}
         {arriveTangent=-3.8452, interpMode=RCIM_Cubic, leaveTangent=-3.8452, tangentMode=RCTM_User, time=0.1, value=1}
         {arriveTangent=-0.1713, interpMode=RCIM_Cubic, leaveTangent=-0.1713, tangentMode=RCTM_User, time=1, value=0}
        ]}}}
     Scale Curve=1.3
    )
    Scale RGB=&MakeVectorFromLinearColorRGB(Color=&HueShiftLinearColor(
      Hue=&MakeCustomFloatFromBool(
       Bool=$System.IsWanderingBoss
       True Float=&MakeCustomFloatFromBool(Bool=$System.IS_Smash, True Float=0.1, False Float=-0.24)
       False Float=0
      )
      Saturation=&MakeCustomFloatFromBool(Bool=$System.IsWanderingBoss, True Float=0.7, False Float=1)
      Value=1
      Linear Color=&MakeCustomLinearColorFromBool(
       Bool=$System.IS_Smash
       True Color=&ColorFromCurve(ColorCurve=DataInterface() {propertyValues {
          AlphaCurve {keys [<at 0 1 RCIM_Cubic>]}
          BlueCurve {keys [<at 0.0169 0.1804 RCIM_Cubic>,<at 0.6322 0>]}
          GreenCurve {keys [<at 0.0169 1 RCIM_Cubic>,<at 0.6322 0.04>]}
          RedCurve {keys [<at 0.0169 0 RCIM_Cubic>,<at 0.6322 0.0021>]}
         }})
       False Color=&ColorFromCurve(ColorCurve=DataInterface() {propertyValues {
          AlphaCurve {keys [<at 0 1 RCIM_Cubic>]}
          BlueCurve {keys [<at 0.0169 1 RCIM_Cubic>,<at 0.6322 0.04>]}
          GreenCurve {keys [<at 0.0169 0.01 RCIM_Cubic>,<at 0.6322 0.006>]}
          RedCurve {keys [<at 0.0169 0.2905 RCIM_Cubic>,<at 0.6322 0.0088>]}
         }})
      )
     ))
    ScaleA=true
    ScaleRGB=true
   )
   /SetVariables(Particles.ColorDark:LinearColor=&MakeLinearColorFromVectorAndFloat(
     "Vector (RGB)"=&MakeVectorFromLinearColorRGB(Color=&HueShiftLinearColor(
       Hue=&MakeCustomFloatFromBool(Bool=$System.IsWanderingBoss, True Float=-0.24, False Float=0)
       Saturation=&MakeCustomFloatFromBool(Bool=$System.IsWanderingBoss, True Float=0.7, False Float=1)
       Value=1
       Linear Color=&MakeCustomLinearColorFromBool(
        Bool=$System.IS_Smash
        True Color=&ColorFromCurve(ColorCurve=DataInterface() {propertyValues {
           AlphaCurve {keys [<at 0 1 RCIM_Cubic>]}
           BlueCurve {keys [<at 0 0.1115 RCIM_Cubic>,<at 0.3152 0>]}
           GreenCurve {keys [<at 0 0.3 RCIM_Cubic>,<at 0.3152 0>]}
           RedCurve {keys [<at 0 0 RCIM_Cubic>,<at 0.3152 0>]}
          }})
        False Color=&ColorFromCurve(ColorCurve=DataInterface() {propertyValues {
           AlphaCurve {keys [<at 0 1 RCIM_Cubic>]}
           BlueCurve {keys [<at 0 0.3 RCIM_Cubic>,<at 0.3152 0>]}
           GreenCurve {keys [<at 0 0 RCIM_Cubic>,<at 0.3152 0>]}
           RedCurve {keys [<at 0 0.05 RCIM_Cubic>,<at 0.3152 0>]}
          }})
       )
      ))
     "Float (Alpha)"=&FloatFromCurve(
      FloatCurve=DataInterface() {propertyValues {Curve {keys [
          {interpMode=RCIM_Cubic, tangentMode=RCTM_User, time=0, value=0}
          {arriveTangent=-2.4988, interpMode=RCIM_Cubic, leaveTangent=-2.4988, tangentMode=RCTM_User, time=0.3, value=1}
          {interpMode=RCIM_Cubic, tangentMode=RCTM_User, time=1, value=0}
         ]}}}
      Scale Curve=3
     )
    )) {isSetParameters=true}
   &SolveForcesAndVelocity(Clamp Velocity=false, Limit Acceleration=false)
   &DynamicMaterialParameters(
    Index 0 Param 1=-0.1
    Index 0 Param 2=&Lerp_Float(
     A=0
     B=1
     Alpha=&FloatFromCurve(FloatCurve=DataInterface() {propertyValues {Curve {keys [<at 0 0>,<at 1 1>]}}})
    )
    Index 0 Param 3=0.5
    Index 0 Param 4=&Lerp_Float(
     A=0.45
     B=0.8
     Alpha=&FloatFromCurve(FloatCurve=DataInterface() {propertyValues {Curve {keys [
          {arriveTangent=0.2526, interpMode=RCIM_Cubic, leaveTangent=0.2526, tangentMode=RCTM_User, time=0, value=0}
          {arriveTangent=1.7785, interpMode=RCIM_Cubic, leaveTangent=1.7785, tangentMode=RCTM_User, time=1, value=1}
         ]}}})
    )
    Module.Index0Param0Write=true
    Module.Index0Param1Write=true
    Module.Index0Param2Write=true
    Module.Index0Param3Write=true
   )
   /SetVariables(Particles.DynamicMaterialParameterDark:Vector4f=&MakeVector4(
     X=-0.5
     Y=&Lerp_Float(
      A=0.2
      B=1.5
      Alpha=&FloatFromCurve(FloatCurve=DataInterface() {propertyValues {Curve {keys [<at 0 0>,<at 1 1>]}}})
     )
     Z=1
     W=&Lerp_Float(
      A=-0.5
      B=1
      Alpha=&FloatFromCurve(FloatCurve=DataInterface() {propertyValues {Curve {keys [
           {arriveTangent=0.2526, interpMode=RCIM_Cubic, leaveTangent=0.2526, tangentMode=RCTM_User, time=0.2, value=0}
           {arriveTangent=1.7785, interpMode=RCIM_Cubic, leaveTangent=1.7785, tangentMode=RCTM_User, time=1, value=1}
          ]}}})
     )
    )) {isSetParameters=true}
   /SetVariables(
    Particles.ColorFlare:LinearColor=&MakeCustomLinearColorFromBool(
     Bool=$System.IS_Smash
     True Color=&ColorFromCurve(ColorCurve=DataInterface() {propertyValues {
        AlphaCurve {keys [<at 0 0>,<at 0.0482 1 RCIM_Cubic>,<at 0.3413 0 RCIM_Cubic>]}
        BlueCurve {keys [<at 0.0373 2.584 RCIM_Cubic>,<at 0.1975 0.0245 RCIM_Cubic>]}
        GreenCurve {keys [<at 0.0373 6.4 RCIM_Cubic>,<at 0.1975 1 RCIM_Cubic>]}
        RedCurve {keys [<at 0.0373 0 RCIM_Cubic>,<at 0.1975 0 RCIM_Cubic>]}
       }})
     False Color=&ColorFromCurve(ColorCurve=DataInterface() {propertyValues {
        AlphaCurve {keys [<at 0 0>,<at 0.0482 1 RCIM_Cubic>,<at 0.3413 0 RCIM_Cubic>]}
        BlueCurve {keys [<at 0.0373 6.4 RCIM_Cubic>,<at 0.1975 1 RCIM_Cubic>]}
        GreenCurve {keys [<at 0.0373 0 RCIM_Cubic>,<at 0.1975 0 RCIM_Cubic>]}
        RedCurve {keys [<at 0.0373 2.0267 RCIM_Cubic>,<at 0.1975 0.25 RCIM_Cubic>]}
       }})
    )
    Particles.PositionFlare:NiagaraPosition=&AddVectorToPosition(
     Position=$Particles.Position
     Vector=&MakeVector(
      X=0
      Y=0
      Z=&Lerp_Float(
       A=0
       B=20
       Alpha=&FloatFromCurve(FloatCurve=DataInterface() {propertyValues {Curve {keys [<at 0 0>,<at 0.65 1>]}}})
      )
     )
    )
   ) {isSetParameters=true}
   /SetVariables(
    Particles.ColorDark:LinearColor=&ColorFromCurve(ColorCurve=DataInterface() {propertyValues {
       AlphaCurve {keys [<at 0 0>,<at 0.0592 1.5 RCIM_Cubic>,<at 0.6507 0 RCIM_Cubic>]}
       BlueCurve {keys [<at 0 0.01 RCIM_Cubic>]}
       GreenCurve {keys [<at 0 0.007 RCIM_Cubic>]}
       RedCurve {keys [<at 0 0.008 RCIM_Cubic>]}
      }})
    Particles.SpriteSizeDark:Vector2f=&Vector2DFromFloat(Value=&Lerp_Float(
      A=0
      B=200
      Alpha=&FloatFromCurve(FloatCurve=DataInterface() {propertyValues {Curve {keys [
           {interpMode=RCIM_Cubic, tangentMode=RCTM_User, time=0, value=0}
           {interpMode=RCIM_Cubic, tangentMode=RCTM_User, time=0.1, value=1}
           {interpMode=RCIM_Cubic, tangentMode=RCTM_User, time=0.7, value=0}
          ]}}})
     ))
   ) {isSetParameters=true}
  ]
  rendererClasses [&NiagaraMeshRendererProperties, &NiagaraSpriteRendererProperties]
  renderers [
   {rendererClass=&NiagaraMeshRendererProperties, rendererIndex=0}
   {rendererClass=&NiagaraMeshRendererProperties, rendererIndex=1}
   {rendererClass=&NiagaraSpriteRendererProperties, rendererIndex=2}
   {rendererClass=&NiagaraSpriteRendererProperties, rendererIndex=3}
  ]
 }
 GlowDisk {
  emitterUpdateScript [
   &EmitterState(Life Cycle Mode=::NewEnumerator1, Loop Behavior=::NewEnumerator1, Loop Delay=0.03, Loop Duration=1, UseLoopDelay=true)
   &SpawnBurst_Instantaneous(Use Spawn Probability=false, UseLoopCountLimit=false)
  ]
  particleSpawnScript [
   &InitializeParticle(
    Color=&MakeCustomLinearColorFromBool(Bool=$System.IS_Smash, True Color=<rgba 0 2 0.8024 1>, False Color=<rgba 0.68 0.02 2 1>)
    Lifetime=0.3
    Mass Mode=::NewEnumerator3
    Mesh Scale Mode=::NewEnumerator3
    Mesh Uniform Scale=1.7
    Position Offset=$System.ImpactOffset
    Sprite Rotation Mode=::NewEnumerator2
    Sprite Size Mode=::NewEnumerator3
    Uniform Sprite Size=330
    UsePositionOffset=true
    Write Material Random=true
   )
   &SpriteFacingAndAlignment(Sprite Facing=<0 0 1>, WriteAlignment=false, WriteFacing=true)
   &InitialMeshOrientation(Mesh Orientation Mode=::NewEnumerator2, Orientation Vector=<0 0 1>, Use Rotation Vector=false)
   &AddVelocity(Velocity=<0 0 20>) {enabled=false}
   &CameraOffset(Camera Offset Amount=10)
  ]
  particleUpdateScript [
   &ParticleState()
   &ScaleMeshSize(Scale Factor=&VectorFromFloat(Value=&FloatFromCurve(FloatCurve=DataInterface() {propertyValues {Curve {keys [
          {arriveTangent=0.4737, interpMode=RCIM_Cubic, leaveTangent=0.4737, tangentMode=RCTM_User, time=0, value=0.85}
          {interpMode=RCIM_Cubic, tangentMode=RCTM_User, time=1, value=1}
         ]}}})))
   &Drag(Drag=5, Use Linear Drag=true, Use Rotational Drag=true) {enabled=false}
   &SolveForcesAndVelocity(Clamp Velocity=false, Limit Acceleration=false)
   &DynamicMaterialParameters(
    Index 0 Param 1=&FloatFromCurve(FloatCurve=DataInterface() {propertyValues {Curve {keys [
         {interpMode=RCIM_Cubic, tangentMode=RCTM_User, time=0, value=0}
         {arriveTangent=2.8206, interpMode=RCIM_Cubic, leaveTangent=2.8206, tangentMode=RCTM_User, time=1, value=1}
        ]}}})
    Index 0 Param 2=0.848
    Index 0 Param 3=&Lerp_Float(
     A=1
     B=0.1
     Alpha=&FloatFromCurve(FloatCurve=DataInterface() {propertyValues {Curve {keys [
          {arriveTangent=3.8686, interpMode=RCIM_Cubic, leaveTangent=3.8686, tangentMode=RCTM_User, time=0, value=0}
          {interpMode=RCIM_Cubic, tangentMode=RCTM_User, time=0.7, value=1}
         ]}}})
    )
    Module.Index0Param0Write=true
    Module.Index0Param1Write=true
    Module.Index0Param2Write=true
    Module.Index0Param3Write=true
   ) {module=DynamicMaterialParameters001}
   /SetVariables(Particles.ColorDark:LinearColor=&ColorFromCurve(ColorCurve=DataInterface() {propertyValues {
       AlphaCurve {keys [<at 0.4663 1.5 RCIM_Cubic>,<at 1 0 RCIM_Cubic>]}
       BlueCurve {keys [<at 0.2047 0 RCIM_Cubic>,<at 1 0.1>]}
       GreenCurve {keys [<at 0.2047 0 RCIM_Cubic>,<at 1 0.06>]}
       RedCurve {keys [<at 0.2047 0 RCIM_Cubic>,<at 1 0.05>]}
      }})) {isSetParameters=true}
   &ScaleColor(
    Scale Alpha=2
    Scale RGB=&VectorFromFloat(Value=&FloatFromCurve(FloatCurve=DataInterface() {propertyValues {Curve {keys [
          {arriveTangent=-9.3373, interpMode=RCIM_Cubic, leaveTangent=-9.3373, tangentMode=RCTM_User, time=0.1, value=2}
          {interpMode=RCIM_Cubic, tangentMode=RCTM_User, time=0.7, value=0.2}
         ]}}}))
    ScaleA=true
    ScaleRGB=true
   )
  ]
  rendererClasses [&NiagaraSpriteRendererProperties, &NiagaraMeshRendererProperties]
  renderers [
   {rendererClass=&NiagaraSpriteRendererProperties}
   {rendererClass=&NiagaraMeshRendererProperties, rendererIndex=1}
  ]
 }
 SmokeWisps {
  emitterUpdateScript [
   &EmitterState()
   &SpawnRate(SpawnRate=20, Use Spawn Probability=false) {enabled=false}
   &SpawnBurst_Instantaneous(Spawn Count=6, Spawn Group=1, Use Spawn Probability=false, UseLoopCountLimit=false)
  ]
  particleSpawnScript [
   &InitializeParticle(
    Color=&MakeCustomLinearColorFromBool(
     Bool=$System.IsWanderingBoss
     True Color=&MakeCustomLinearColorFromBool(
      Bool=$System.IS_Smash
      True Color=&RandomRangeLinearColor(Minimum=<rgba 0 0.151 0.0267 1>, Maximum=<rgba 0.0071 0.0417 0 1>)
      False Color=&RandomRangeLinearColor(Minimum=<rgba 0.2604 0.0057 0.0227 1>, Maximum=<rgba 0.0417 0.0028 0.0093 1>)
     )
     False Color=&RandomRangeLinearColor(Minimum=<rgba 0.0868 0 0.2604 1>, Maximum=<rgba 0.0029 0.0011 0.0417 1>)
    )
    Lifetime=&RandomRangeFloat(Minimum=0.5, Maximum=0.8)
    Sprite Rotation Mode=::NewEnumerator2
    Sprite Size=&Vector2DFromFloat(Value=&RandomRangeFloat(Minimum=200, Maximum=250))
    Sprite Size Mode=::NewEnumerator4
    Sprite UV Mode=::NewEnumerator3
    UsePositionOffset=false
    Write Material Random=true
   )
   &CameraOffset(Camera Offset Amount=100)
   &DynamicMaterialParameters(
    Index 0 Param 1=&RandomRangeFloat(Minimum=0, Maximum=-6)
    Module.Index0Param0Write=true
    Module.Index0Param1Write=false
    Module.Index0Param2Write=false
    Module.Index0Param3Write=false
   )
   &AddVelocity(Velocity=&RandomRangeVector(Minimum=<0 0 200>, Maximum=<0 0 20>)) {enabled=false}
  ]
  particleUpdateScript [
   &ParticleState()
   &ScaleSpriteSize(
    Scale Sprite Size Mode=::NewEnumerator0
    Uniform Scale Factor=&MaskFloatBySpawnGroup(
     Masked Float=&FloatFromCurve(FloatCurve=DataInterface() {propertyValues {Curve {keys [
          <at 0 0.25 RCIM_Cubic>
          {arriveTangent=0.75, interpMode=RCIM_Cubic, leaveTangent=0.75, time=0.4, value=0.4}
          {interpMode=RCIM_Cubic, leaveTangent=0.1472, time=1, value=1}
         ]}}})
     Passthrough Float=&FloatFromCurve(
      FloatCurve=DataInterface() {propertyValues {Curve {keys [
          {arriveTangent=4.1785, interpMode=RCIM_Cubic, leaveTangent=4.1785, tangentMode=RCTM_User, time=0, value=0.1}
          {arriveTangent=1.1179, interpMode=RCIM_Cubic, leaveTangent=1.1179, tangentMode=RCTM_User, time=0.2042, value=0.6228}
          {arriveTangent=0.0224, interpMode=RCIM_Cubic, leaveTangent=0.0224, tangentMode=RCTM_User, time=1, value=1}
         ]}}}
      Scale Curve=1.5
     )
    )
   )
   &ScaleColor(
    Scale Alpha=&FloatFromCurve(
     FloatCurve=DataInterface() {propertyValues {Curve {keys [<at 0 0 RCIM_Cubic>, <at 0.2 1 RCIM_Cubic>, {interpMode=RCIM_Cubic, leaveTangent=-26.9591, time=1, value=0}]}}}
     Scale Curve=4
    )
    Scale RGB=&MaskVectorBySpawnGroup(
     Masked Vector=&VectorFromCurve(VectorCurve=DataInterface() {propertyValues {
        XCurve {keys [<at 0.25 1 RCIM_Cubic>,<at 1 0.5>]}
        YCurve {keys [<at 0.25 1 RCIM_Cubic>,<at 1 0.5>]}
        ZCurve {keys [<at 0.25 1 RCIM_Cubic>,<at 1 0.5>]}
       }})
     Passthrough Vector=&VectorFromCurve(
      VectorCurve=DataInterface() {propertyValues {
        XCurve {keys [<at 0.1 1 RCIM_Cubic>,<at 0.5 0.5>]}
        YCurve {keys [<at 0.1 1 RCIM_Cubic>,<at 0.5 0.5>]}
        ZCurve {keys [<at 0.1 1 RCIM_Cubic>,<at 0.5 0.5>]}
       }}
      Scale Curve=<5 5 5>
     )
    )
    ScaleA=true
    ScaleRGB=true
   )
   /Niagara/Modules/Update/SubUV/SubUVAnimation(Number Of Frames=32)
   &DynamicMaterialParameters(
    Index 0 Param 1=&MaskFloatBySpawnGroup(
     Masked Float=&FloatFromCurve(
      FloatCurve=DataInterface() {propertyValues {Curve {keys [
          {arriveTangent=1.4977, interpMode=RCIM_Cubic, leaveTangent=1.4977, tangentMode=RCTM_User, time=0, value=0}
          {arriveTangent=0.4339, interpMode=RCIM_Cubic, leaveTangent=0.4339, tangentMode=RCTM_User, time=1, value=1}
         ]}}}
      Scale Curve=0.1
     )
     Passthrough Float=0
    )
    Index 0 Param 2=80
    Module.Index0Param0Write=true
    Module.Index0Param1Write=true
    Module.Index0Param2Write=false
    Module.Index0Param3Write=false
   ) {module=DynamicMaterialParameters001}
   &Drag(Use Linear Drag=true, Use Rotational Drag=true)
   &AccelerationForce(Acceleration=<0 0 300>)
   &SolveForcesAndVelocity(Clamp Velocity=false, Limit Acceleration=false)
  ]
  rendererClasses [&NiagaraSpriteRendererProperties]
  renderers [{rendererClass=&NiagaraSpriteRendererProperties}]
 }
 TwistAbsorb {
  emitterUpdateScript [
   &EmitterState(Life Cycle Mode=::NewEnumerator1, Loop Behavior=::NewEnumerator1, Loop Delay=0.1, Loop Duration=2, UseLoopDelay=true)
   &SpawnBurst_Instantaneous(Use Spawn Probability=false, UseLoopCountLimit=false)
   &SpawnBurst_Instantaneous(Spawn Group=1, Use Spawn Probability=false, UseLoopCountLimit=false) {enabled=false, module=SpawnBurst_Instantaneous001}
  ]
  particleSpawnScript [
   &InitializeParticle(
    Color=&MakeCustomLinearColorFromBool(Bool=$System.IS_Smash, True Color=<rgba 0.142 2 0 1>, False Color=<rgba 0.68 0.02 2 1>)
    Lifetime=0.5
    Mesh Scale Mode=::NewEnumerator3
    Mesh Uniform Scale=400
    Sprite Rotation Mode=::NewEnumerator2
    Sprite Size Mode=::NewEnumerator3
    Uniform Sprite Size=350
    UsePositionOffset=false
    Write Material Random=true
   )
   &CameraOffset(Camera Offset Amount=30) {module=CameraOffset001}
   &SystemLocation(MaskSpawnGroup=false, Offset=<0 0 100>)
   &InitialMeshOrientation(Mesh Orientation Mode=::NewEnumerator2, Orientation Vector=<0 0 1>, Use Rotation Vector=false)
   /SetVariables(Particles.VisibilityTag=&MaskIntBySpawnGroup(Masked Int=0, Passthrough Int=1, Spawn Group Mask=0)) {enabled=false, isSetParameters=true}
  ]
  particleUpdateScript [
   &CameraOffset(Camera Offset Amount=&Lerp_Float(
     A=600
     B=-100
     Alpha=&FloatFromCurve(FloatCurve=DataInterface() {propertyValues {Curve {keys [{interpMode=RCIM_Cubic, tangentMode=RCTM_User, time=0.35, value=0}, {interpMode=RCIM_Cubic, tangentMode=RCTM_User, time=0.7, value=1}]}}})
    )) {enabled=false}
   &ParticleState()
   &SpriteRotationRate(Rotation Rate=-40)
   &ScaleSpriteSize(Uniform Curve Sprite Scale=DataInterface() {propertyValues {Curve {keys [
        {arriveTangent=12.5362, interpMode=RCIM_Cubic, leaveTangent=12.5362, tangentMode=RCTM_User, time=0, value=0.3}
        {arriveTangent=-0.523, interpMode=RCIM_Cubic, leaveTangent=-0.523, tangentMode=RCTM_User, time=0.1, value=1}
       ]}}})
   &SolveForcesAndVelocity(Clamp Velocity=false, Limit Acceleration=false)
   &DynamicMaterialParameters(
    Index 0 Param 1=&Lerp_Float(
     A=0.1
     B=1
     Alpha=&FloatFromCurve(FloatCurve=DataInterface() {propertyValues {Curve {keys [
          {arriveTangent=0.1556, interpMode=RCIM_Cubic, leaveTangent=0.1556, tangentMode=RCTM_User, time=0, value=0}
          {arriveTangent=2.2716, interpMode=RCIM_Cubic, leaveTangent=2.2716, tangentMode=RCTM_User, time=0.75, value=1}
         ]}}})
    )
    Index 0 Param 2=1
    Index 0 Param 3=&Lerp_Float(
     A=-0.1
     B=0.5
     Alpha=&FloatFromCurve(FloatCurve=DataInterface() {propertyValues {Curve {keys [
          {arriveTangent=0.1235, interpMode=RCIM_Cubic, leaveTangent=0.1235, tangentMode=RCTM_User, time=0.4, value=0}
          {arriveTangent=1.955, interpMode=RCIM_Cubic, leaveTangent=1.955, tangentMode=RCTM_User, time=1, value=1}
         ]}}})
    )
    Index 0 Param 4=&Lerp_Float(
     A=0.2
     B=1
     Alpha=&FloatFromCurve(FloatCurve=DataInterface() {propertyValues {Curve {keys [
          {arriveTangent=0.1516, interpMode=RCIM_Cubic, leaveTangent=0.1516, tangentMode=RCTM_User, time=0, value=0}
          {arriveTangent=2.545, interpMode=RCIM_Cubic, leaveTangent=2.545, tangentMode=RCTM_User, time=0.9, value=1}
         ]}}})
    )
    Module.Index0Param0Write=true
    Module.Index0Param1Write=true
    Module.Index0Param2Write=true
    Module.Index0Param3Write=true
   )
   &ScaleColor(
    Scale Alpha=&FloatFromCurve(
     FloatCurve=DataInterface() {propertyValues {Curve {keys [
         <at 0 0 RCIM_Cubic>
         {arriveTangent=-1.4323, interpMode=RCIM_Cubic, leaveTangent=-1.4323, tangentMode=RCTM_User, time=0.4, value=1}
         {interpMode=RCIM_Cubic, tangentMode=RCTM_User, time=1, value=0}
        ]}}}
     Scale Curve=2.3
    )
    Scale RGB=&Multiply_VectorByFloat(
     Vector=&MakeVectorFromLinearColorRGB(Color=&ColorFromCurve(ColorCurve=DataInterface() {propertyValues {
         AlphaCurve {keys [<at 0 1 RCIM_Cubic>]}
         BlueCurve {keys [<at 0.1216 1 RCIM_Cubic>,<at 0.8838 0.03 RCIM_Cubic>]}
         GreenCurve {keys [<at 0.1216 1 RCIM_Cubic>,<at 0.8838 0.0018 RCIM_Cubic>]}
         RedCurve {keys [<at 0.1216 1 RCIM_Cubic>,<at 0.8838 0.0065 RCIM_Cubic>]}
        }}))
     Float=&FNTimeOfDayScalar(Night Scale=1, Morning Scale=1, Evening Scale=1)
    )
    ScaleA=true
    ScaleRGB=true
   )
   /SetVariables(
    Particles.ColorDark:LinearColor=&MultiplyLinearColor(
     Linear Color=<rgba 0.0064 0.006 0.02 1>
     Scale Factor=&MakeFloat4FromVectorAndFloat(
      "Vector (RGB)"=<1 1 1>
      "Float (Alpha)"=&FloatFromCurve(
       FloatCurve=DataInterface() {propertyValues {Curve {keys [
           <at 0 0 RCIM_Cubic>
           {arriveTangent=-2.1159, interpMode=RCIM_Cubic, leaveTangent=-2.1159, tangentMode=RCTM_User, time=0.2, value=1}
           {interpMode=RCIM_Cubic, tangentMode=RCTM_User, time=1, value=0}
          ]}}}
       Scale Curve=2
      )
     )
    )
    Particles.DynamicMaterialParameterDark:Vector4f=&MakeVector4(
     X=&Lerp_Float(
      A=0
      B=1
      Alpha=&FloatFromCurve(FloatCurve=DataInterface() {propertyValues {Curve {keys [
           {interpMode=RCIM_Cubic, tangentMode=RCTM_User, time=0, value=0}
           {arriveTangent=1.7687, interpMode=RCIM_Cubic, leaveTangent=1.7687, tangentMode=RCTM_User, time=1, value=1}
          ]}}})
     )
     Y=1
     Z=&Lerp_Float(
      A=-0.5
      B=0.3
      Alpha=&FloatFromCurve(FloatCurve=DataInterface() {propertyValues {Curve {keys [
           {arriveTangent=0.1235, interpMode=RCIM_Cubic, leaveTangent=0.1235, tangentMode=RCTM_User, time=0.2, value=0}
           {arriveTangent=1.955, interpMode=RCIM_Cubic, leaveTangent=1.955, tangentMode=RCTM_User, time=1, value=1}
          ]}}})
     )
     W=0.1
    )
    Particles.SpriteSizeDark:Vector2f=&Multiply_Vector2D_ByFloat(Float=1, Vector2D=$Particles.SpriteSize)
   ) {isSetParameters=true}
   /SetVariables(Particles.ColorFlare:LinearColor=&ColorFromCurve(ColorCurve=DataInterface() {propertyValues {
       AlphaCurve {keys [{arriveTangent=6.3978, interpMode=RCIM_Cubic, time=0.0633, value=1}, <at 0.3617 0 RCIM_Cubic>]}
       BlueCurve {keys [<at 0.0651 10 RCIM_Cubic>]}
       GreenCurve {keys [<at 0.0651 0 RCIM_Cubic>]}
       RedCurve {keys [<at 0.0651 2.5 RCIM_Cubic>]}
      }})) {isSetParameters=true}
   &ScaleMeshSize(Scale Factor=&MakeVector(
     X=&FloatFromCurve(
      FloatCurve=DataInterface() {propertyValues {Curve {keys [
          {arriveTangent=4.2714, interpMode=RCIM_Cubic, leaveTangent=4.2714, tangentMode=RCTM_User, time=0, value=0.2}
          {interpMode=RCIM_Cubic, tangentMode=RCTM_User, time=0.5, value=1}
         ]}}}
      Scale Curve=2
     )
     Y=&FloatFromCurve(
      FloatCurve=DataInterface() {propertyValues {Curve {keys [
          {interpMode=RCIM_Cubic, tangentMode=RCTM_User, time=0, value=1}
          {arriveTangent=-3.9244, interpMode=RCIM_Cubic, leaveTangent=-3.9244, tangentMode=RCTM_User, time=0.25, value=0.3}
         ]}}}
      Scale Curve=1.4
     )
     Z=1
    ))
  ]
  rendererClasses [&NiagaraSpriteRendererProperties]
  renderers [
   {rendererClass=&NiagaraSpriteRendererProperties}
   {rendererClass=&NiagaraSpriteRendererProperties, rendererIndex=1}
  ]
 }
}
summary {
 systemName=NS_VoidMask_Teleport_Spawn
 emitters [
  {emitterName=SmokeWisps, rendererClasses [&NiagaraSpriteRendererProperties]}
  {emitterName=DarkStreaks, rendererClasses [&NiagaraSpriteRendererProperties]}
  {emitterName=FlareSmoke, rendererClasses [&NiagaraMeshRendererProperties, &NiagaraSpriteRendererProperties]}
  {emitterName=GlowDisk, rendererClasses [&NiagaraSpriteRendererProperties, &NiagaraMeshRendererProperties]}
  {emitterName=TwistAbsorb, rendererClasses [&NiagaraSpriteRendererProperties]}
 ]
 userVariables [
  User.ColorVariant {type=NiagaraFloat, defaultValue=0, description=""}
  User.Scalar {type=NiagaraFloat, defaultValue=2, description=""}
  User.WanderingBoss_F {type=NiagaraFloat, defaultValue=0, description=""}
 ]
}
