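# Niagara effect for the Void Mask wandering boss relocation, combining a red flash with dark smoke.
# LongFlare spawns one wide HDR red sprite that stretches horizontally and fades, with a particle light that swells and dies.
# SmokeWhisps bursts and streams SubUV smoke from a sphere scaled by User.Global Scale, pulled by point attraction.
# TwistAbsorb, a layered triple-sprite twist with separate dark color and material attributes, is disabled.
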NiagaraSystem /DemonMaskVoid/Assets/VFX/NS_VoidMask_WanderingBossRelocation
dependencies {
 usedDataInterfaces [&NiagaraDataInterfaceCurve, &NiagaraDataInterfaceVectorCurve, &NiagaraDataInterfaceColorCurve]
 usedDynamicInputs []
 usedModules [&SystemState]
}
emitters {
 LongFlare {
  emitterUpdateScript [
   &EmitterState(Life Cycle Mode=::NewEnumerator1, Loop Behavior=::NewEnumerator1, Loop Duration=1, UseLoopDelay=true)
   &SpawnBurst_Instantaneous(Spawn Count=1, Use Spawn Probability=false)
  ]
  particleSpawnScript [
   /Niagara/Modules/Spawn/Initialization/InitializeParticle(
    Color=<rgba 10 0.2 0.3633 1>
    Lifetime=3
    Module.Write Color=true
    Module.Write Lifetime=true
    Module.Write Mass=true
    Module.Write Material Random=true
    Module.Write Position=false
    Module.Write Scale=false
    Module.Write SpriteRotation=true
    Module.Write SpriteSize=true
    Position=&SimulationPosition()
    Sprite Size=&Multiply_Vector2D_ByFloat(Float=0.45, Vector2D=&Multiply_Vector2D_ByFloat(Float=0.2509, Vector2D=<11849.3916 5571.4517>))
   )
   &DynamicMaterialParameters(
    Index 0 Param 1=0.2
    Index 0 Param 2=1
    Module.Param0WriteEnabled=true
    Module.Param0WriteEnabled1=true
    Module.Param0WriteEnabled2=true
    Module.Param0WriteEnabled3=true
    Module.Param1WriteEnabled=true
    Module.Param1WriteEnabled1=true
    Module.Param1WriteEnabled2=true
    Module.Param1WriteEnabled3=true
    Module.Param2WriteEnabled=true
    Module.Param2WriteEnabled1=true
    Module.Param2WriteEnabled2=true
    Module.Param2WriteEnabled3=true
    Module.Param3WriteEnabled=true
    Module.Param3WriteEnabled1=true
    Module.Param3WriteEnabled2=true
    Module.Param3WriteEnabled3=true
    Write Parameter Index 1=true
    Write Parameter Index 2=true
    Write Parameter Index 3=true
   )
  ]
  particleUpdateScript [
   &ParticleState()
   &CameraOffset(Camera Offset Amount=100)
   &SolveForcesAndVelocity(Clamp Velocity=false, Limit Acceleration=false)
   &ScaleColor(
    Color Value To Scale=$Particles.Color
    Scale Alpha=&FloatFromCurve(FloatCurve=DataInterface() {propertyValues {Curve {keys [
         <at 0 1 RCIM_Cubic>
         {interpMode=RCIM_Cubic, tangentMode=RCTM_Break, time=0.1, value=1}
         {arriveTangent=-3.4115, tangentMode=RCTM_User, time=1, value=0}
        ]}}})
    ScaleA=true
    ScaleRGB=true
   )
   &ScaleSpriteSize(Scale Factor=&MakeVector2D(
     X=&FloatFromCurve(
      FloatCurve=DataInterface() {propertyValues {Curve {keys [
          {arriveTangent=4.441, interpMode=RCIM_Cubic, leaveTangent=4.441, tangentMode=RCTM_User, time=0, value=0}
          {arriveTangent=-0.0405, interpMode=RCIM_Cubic, leaveTangent=-0.0405, tangentMode=RCTM_User, time=1, value=2}
         ]}}}
      Scale Curve=&RandomRangeFloat(Minimum=0.7, Maximum=1, Evaluation Type=::NewEnumerator1)
     )
     Y=1
    ))
   &ScaleSpriteSize(Scale Sprite Size Mode=::NewEnumerator0) {module=ScaleSpriteSize001}
   &FN_Light_Attributes(
    Color=<rgba 1 0 0 1>
    Create Color Attribute=true
    Create Radius Attribute=true
    Enable Particle Light Enable=true
    Light Enabled=true
    Radius=2014.3315
    Scale Color=&Multiply_Float(
     A=669820.625
     B=&FloatFromCurve(FloatCurve=DataInterface() {propertyValues {Curve {keys [
          {arriveTangent=2.7293, interpMode=RCIM_Cubic, leaveTangent=2.7293, tangentMode=RCTM_User, time=0, value=0}
          {arriveTangent=-0.1593, interpMode=RCIM_Cubic, leaveTangent=-0.1593, tangentMode=RCTM_User, time=0.6, value=0.9847}
          {arriveTangent=-0.0405, interpMode=RCIM_Cubic, leaveTangent=-0.0405, tangentMode=RCTM_User, time=1, value=0}
         ]}}})
    )
   )
  ]
  rendererClasses [&NiagaraSpriteRendererProperties, &NiagaraLightRendererProperties]
  renderers [
   {rendererClass=&NiagaraSpriteRendererProperties}
   {rendererClass=&NiagaraLightRendererProperties, rendererIndex=1}
  ]
 }
 SmokeWhisps {
  emitterUpdateScript [
   &EmitterState(Life Cycle Mode=::NewEnumerator1, Loop Behavior=::NewEnumerator1, Loop Duration=0.8, UseLoopDelay=true)
   &SpawnRate(SpawnGroup=1, SpawnRate=5, Use Spawn Probability=false) {module=SpawnRate001}
   &SpawnRate(SpawnRate=5, Use Spawn Probability=false)
   &SpawnBurst_Instantaneous(Spawn Count=10, Use Spawn Probability=false, UseLoopCountLimit=false)
  ]
  particleSpawnScript [
   /SetVariables(Particles.VisibilityTag=$Emitter.SpawnGroup) {enabled=false, isSetParameters=true}
   /Niagara/Modules/Spawn/Initialization/V2/InitializeParticle(
    Color=&MaskLinearColorBySpawnGroup(Masked Linear Color=<rgba 0.03 0.021 0.021 1>, Passthrough Linear Color=<rgba 0.05 0.025 0.0279 1>)
    Lifetime=&RandomRangeFloat(Minimum=0.8, Maximum=1.3)
    Sprite Rotation Mode=::NewEnumerator2
    Sprite Size=&Vector2DFromFloat(Value=&Multiply_Float(A=&RandomRangeFloat(Minimum=135.6973, Maximum=184.0385), B=3))
    Sprite Size Mode=::NewEnumerator4
    Sprite UV Mode=::NewEnumerator3
    UsePositionOffset=false
    Write Material Random=true
   )
   &CameraOffset(Camera Offset Amount=-100)
   &ShapeLocation(
    MaskSpawnGroup=false
    Override Seed=false
    Sphere Radius=&Multiply_Float(A=150, B=$User.Global Scale)
    UseOwnerScale=true
   )
   &DynamicMaterialParameters(
    Index 0 Param 1=&RandomRangeFloat(Minimum=0, Maximum=-6)
    Module.Index0Param0Write=true
    Module.Index0Param1Write=false
    Module.Index0Param2Write=false
    Module.Index0Param3Write=false
   )
  ]
  particleUpdateScript [
   &ParticleState()
   &ScaleSpriteSize(
    Scale Sprite Size Mode=::NewEnumerator0
    Uniform Scale Factor=&FloatFromCurve(FloatCurve=DataInterface() {propertyValues {Curve {keys [
         {arriveTangent=3.7398, interpMode=RCIM_Cubic, leaveTangent=3.7398, tangentMode=RCTM_User, time=0, value=0}
         {arriveTangent=-0.7631, interpMode=RCIM_Cubic, leaveTangent=-1.2761, tangentMode=RCTM_Break, time=0.7, value=0.9964}
         {arriveTangent=-4.1068, interpMode=RCIM_Cubic, leaveTangent=-4.1068, tangentMode=RCTM_User, time=1, value=0.4}
        ]}}})
   )
   &ScaleColor(
    Scale Alpha=&FloatFromCurve(
     FloatCurve=DataInterface() {propertyValues {Curve {keys [
         <at 0 0 RCIM_Cubic>
         <at 0.6 1.775 RCIM_Cubic>
         {arriveTangent=-12.688, interpMode=RCIM_Cubic, leaveTangent=-12.688, tangentMode=RCTM_User, time=1, value=0}
        ]}}}
     Scale Curve=1.5
    )
    Scale RGB=&VectorFromCurve(VectorCurve=DataInterface() {propertyValues {
       XCurve {keys [<at 0.25 1 RCIM_Cubic>,<at 1 0.5>]}
       YCurve {keys [<at 0.25 1 RCIM_Cubic>,<at 1 0.5>]}
       ZCurve {keys [<at 0.25 1 RCIM_Cubic>,<at 1 0.5>]}
      }})
    ScaleA=true
    ScaleRGB=true
   )
   &ScaleMeshSize(Scale Factor=&VectorFromFloat(Value=&FloatFromCurve(FloatCurve=DataInterface() {propertyValues {Curve {keys [
          {arriveTangent=3.7398, interpMode=RCIM_Cubic, leaveTangent=3.7398, tangentMode=RCTM_User, time=0, value=0}
          {arriveTangent=-0.7631, interpMode=RCIM_Cubic, leaveTangent=-1.2761, tangentMode=RCTM_Break, time=0.7, value=0.9964}
          {arriveTangent=-4.1068, interpMode=RCIM_Cubic, leaveTangent=-4.1068, tangentMode=RCTM_User, time=1, value=0}
         ]}}})))
   /Niagara/Modules/Update/SubUV/SubUVAnimation(Number Of Frames=32)
   &DynamicMaterialParameters(
    Index 0 Param 1=&FloatFromCurve(
     FloatCurve=DataInterface() {propertyValues {Curve {keys [
         {arriveTangent=2.8718, interpMode=RCIM_Cubic, leaveTangent=2.8718, tangentMode=RCTM_User, time=0, value=0}
         {arriveTangent=-0.3942, interpMode=RCIM_Cubic, leaveTangent=-0.3942, tangentMode=RCTM_User, time=1, value=1}
        ]}}}
     Scale Curve=0.1
    )
    Index 0 Param 2=50
    Module.Index0Param0Write=true
    Module.Index0Param1Write=true
    Module.Index0Param2Write=false
    Module.Index0Param3Write=false
   ) {module=DynamicMaterialParameters001}
   &PointAttractionForce(
    Attraction Radius=168975.2031
    Attraction Strength=&RandomRangeFloat(Minimum=38.6812, Maximum=52.2384)
    Kill Within Radius=false
    Module.Use Falloff=true
    UseAttractorOffset=false
   )
   &SolveForcesAndVelocity(Clamp Velocity=false, Limit Acceleration=false)
  ]
  rendererClasses [&NiagaraSpriteRendererProperties]
  renderers [{rendererClass=&NiagaraSpriteRendererProperties}]
 }
 TwistAbsorb {
  bEnabled=false
  emitterUpdateScript [
   &EmitterState(Life Cycle Mode=::NewEnumerator1, Loop Behavior=::NewEnumerator1, Loop Delay=0.6, Loop Duration=2, UseLoopDelay=true)
   &SpawnBurst_Instantaneous(Use Spawn Probability=false, UseLoopCountLimit=false)
  ]
  particleSpawnScript [
   &InitializeParticle(
    Color=<rgba 3 0.3 0.345 1>
    Lifetime=0.7
    Sprite Rotation Mode=::NewEnumerator2
    Sprite Size Mode=::NewEnumerator3
    Uniform Sprite Size=585.8763
    UsePositionOffset=false
    Write Material Random=true
   )
   &CameraOffset(Camera Offset Amount=200) {module=CameraOffset001}
   &SystemLocation(MaskSpawnGroup=false) {enabled=false}
  ]
  particleUpdateScript [
   &CameraOffset(Camera Offset Amount=&Lerp_Float(
     A=600
     B=-100
     Alpha=&FloatFromCurve(FloatCurve=DataInterface() {propertyValues {Curve {keys [{interpMode=RCIM_Cubic, tangentMode=RCTM_User, time=0.35, value=0}, {interpMode=RCIM_Cubic, tangentMode=RCTM_User, time=0.7, value=1}]}}})
    )) {enabled=false}
   &ParticleState()
   &ScaleSpriteSize(Uniform Curve Sprite Scale=DataInterface() {propertyValues {Curve {keys [
        {arriveTangent=12.5362, interpMode=RCIM_Cubic, leaveTangent=12.5362, tangentMode=RCTM_User, time=0, value=0.3}
        {arriveTangent=-0.523, interpMode=RCIM_Cubic, leaveTangent=-0.523, tangentMode=RCTM_User, time=0.1, value=1}
        <at 1 0 RCIM_Cubic>
       ]}}})
   &SolveForcesAndVelocity(Clamp Velocity=false, Limit Acceleration=false)
   &DynamicMaterialParameters(
    Index 0 Param 1=&Lerp_Float(
     A=-0.5
     B=2
     Alpha=&FloatFromCurve(FloatCurve=DataInterface() {propertyValues {Curve {keys [
          {arriveTangent=0.1556, interpMode=RCIM_Cubic, leaveTangent=0.1556, tangentMode=RCTM_User, time=0, value=0}
          {arriveTangent=2.2716, interpMode=RCIM_Cubic, leaveTangent=2.2716, tangentMode=RCTM_User, time=0.8, value=1}
         ]}}})
    )
    Index 0 Param 2=1
    Index 0 Param 3=&Lerp_Float(
     A=-0.1
     B=0.5
     Alpha=&FloatFromCurve(FloatCurve=DataInterface() {propertyValues {Curve {keys [
          {arriveTangent=0.1235, interpMode=RCIM_Cubic, leaveTangent=0.1235, tangentMode=RCTM_User, time=0.4, value=0}
          {arriveTangent=1.955, interpMode=RCIM_Cubic, leaveTangent=1.955, tangentMode=RCTM_User, time=1, value=1}
         ]}}})
    )
    Index 0 Param 4=&Lerp_Float(
     A=0
     B=1
     Alpha=&FloatFromCurve(FloatCurve=DataInterface() {propertyValues {Curve {keys [
          {arriveTangent=0.3497, interpMode=RCIM_Cubic, leaveTangent=0.3497, tangentMode=RCTM_User, time=0, value=0}
          {interpMode=RCIM_Cubic, tangentMode=RCTM_User, time=1, value=1}
         ]}}})
    )
    Module.Index0Param0Write=true
    Module.Index0Param1Write=true
    Module.Index0Param2Write=true
    Module.Index0Param3Write=true
   )
   &ScaleColor(
    Scale Alpha=&FloatFromCurve(
     FloatCurve=DataInterface() {propertyValues {Curve {keys [
         <at 0 0 RCIM_Cubic>
         {arriveTangent=-0.0178, interpMode=RCIM_Cubic, leaveTangent=-0.0178, tangentMode=RCTM_User, time=0.1, value=1}
         {interpMode=RCIM_Cubic, tangentMode=RCTM_User, time=0.3, value=1}
         {interpMode=RCIM_Cubic, tangentMode=RCTM_User, time=1, value=0}
        ]}}}
     Scale Curve=2
    )
    Scale RGB=&MakeVectorFromLinearColorRGB(Color=&ColorFromCurve(ColorCurve=DataInterface() {propertyValues {
        AlphaCurve {keys [<at 0 1 RCIM_Cubic>]}
        BlueCurve {keys [<at 0.1216 1 RCIM_Cubic>,<at 0.8838 0.03 RCIM_Cubic>]}
        GreenCurve {keys [<at 0.1216 1 RCIM_Cubic>,<at 0.8838 0.0018 RCIM_Cubic>]}
        RedCurve {keys [<at 0.1216 1 RCIM_Cubic>,<at 0.8838 0.0065 RCIM_Cubic>]}
       }}))
    ScaleA=true
    ScaleRGB=true
   )
   /SetVariables(
    Particles.ColorDark:LinearColor=&MultiplyLinearColor(
     Linear Color=<rgba 0.02 0.006 0.0121 1>
     Scale Factor=&MakeFloat4FromVectorAndFloat(
      "Vector (RGB)"=<1 1 1>
      "Float (Alpha)"=&FloatFromCurve(
       FloatCurve=DataInterface() {propertyValues {Curve {keys [
           <at 0 0 RCIM_Cubic>
           {arriveTangent=-2.1159, interpMode=RCIM_Cubic, leaveTangent=-2.1159, tangentMode=RCTM_User, time=0.2, value=1}
           {interpMode=RCIM_Cubic, tangentMode=RCTM_User, time=1, value=0}
          ]}}}
       Scale Curve=2
      )
     )
    )
    Particles.DynamicMaterialParameterDark:Vector4f=&MakeVector4(
     X=&Lerp_Float(
      A=0
      B=1
      Alpha=&FloatFromCurve(FloatCurve=DataInterface() {propertyValues {Curve {keys [
           {interpMode=RCIM_Cubic, tangentMode=RCTM_User, time=0, value=0}
           {arriveTangent=1.7687, interpMode=RCIM_Cubic, leaveTangent=1.7687, tangentMode=RCTM_User, time=1, value=1}
          ]}}})
     )
     Y=1
     Z=&Lerp_Float(
      A=-0.5
      B=0.3
      Alpha=&FloatFromCurve(FloatCurve=DataInterface() {propertyValues {Curve {keys [
           {arriveTangent=0.1235, interpMode=RCIM_Cubic, leaveTangent=0.1235, tangentMode=RCTM_User, time=0.2, value=0}
           {arriveTangent=1.955, interpMode=RCIM_Cubic, leaveTangent=1.955, tangentMode=RCTM_User, time=1, value=1}
          ]}}})
     )
     W=0.1
    )
    Particles.SpriteSizeDark:Vector2f=&Multiply_Vector2D_ByFloat(Float=1, Vector2D=$Particles.SpriteSize)
   ) {isSetParameters=true}
  ]
  rendererClasses [&NiagaraSpriteRendererProperties]
  renderers [
   {rendererClass=&NiagaraSpriteRendererProperties}
   {rendererClass=&NiagaraSpriteRendererProperties, rendererIndex=1}
   {rendererClass=&NiagaraSpriteRendererProperties, rendererIndex=2}
  ]
 }
}
summary {
 systemName=NS_VoidMask_WanderingBossRelocation
 emitters [
  {emitterName=LongFlare, rendererClasses [&NiagaraSpriteRendererProperties, &NiagaraLightRendererProperties]}
  {emitterName=SmokeWhisps, rendererClasses [&NiagaraSpriteRendererProperties]}
  {bEnabled=false, emitterName=TwistAbsorb, rendererClasses [&NiagaraSpriteRendererProperties]}
 ]
 userVariables [
  User.ElementSizeScale {type=NiagaraFloat, defaultValue=1, description=""}
  User.Global Scale {type=NiagaraFloat, defaultValue=1, description=""}
  User.WanderingBoss {type=NiagaraBool}
 ]
}
