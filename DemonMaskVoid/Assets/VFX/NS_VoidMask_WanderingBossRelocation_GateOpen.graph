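# Niagara effect for a gate opening, named for the Void Mask wandering boss relocation, mostly in red and black.
# SmokeLoopIdle layers four mesh renderers, using emitter-age curves to grow dark body, red outline and core.
# Delayed TwistAbsorb sprites shrink and darken through animated material parameters, with separate dark-layer attributes.
# Burst flares add glow, scaled for gameplay or front end, and a distance-offset flash; the flare's light-attribute module is disabled.
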NiagaraSystem /DemonMaskVoid/Assets/VFX/NS_VoidMask_WanderingBossRelocation_GateOpen
dependencies {
 usedDataInterfaces [&NiagaraDataInterfaceColorCurve, &NiagaraDataInterfaceVectorCurve, &NiagaraDataInterfaceCurve]
 usedDynamicInputs []
 usedModules [&SystemState]
}
emitters {
 Flare001 {
  emitterUpdateScript [
   &EmitterState(Life Cycle Mode=::NewEnumerator1, Loop Behavior=::NewEnumerator1, Loop Delay=0.3, Loop Duration=1, UseLoopDelay=true)
   &SpawnBurst_Instantaneous(Spawn Count=1, Use Spawn Probability=false)
  ]
  particleSpawnScript [
   &InitializeParticle(
    Color=<rgba 1.7 1.7 1.7 0.5>
    Lifetime=0.3
    Sprite Rotation Mode=::NewEnumerator2
    Sprite Size Mode=::NewEnumerator3
    Uniform Sprite Size=1579.4823
    UsePositionOffset=false
    Write Material Random=true
   )
   &DynamicMaterialParameters(
    Index 0 Param 1=&RandomRangeFloat(Minimum=0.4, Maximum=0.6)
    Index 0 Param 2=1
    Index 0 Param 3=1
    Index 0 Param 4=1
    Module.Param0WriteEnabled=true
    Module.Param1WriteEnabled=true
    Module.Param2WriteEnabled=true
    Module.Param3WriteEnabled=true
   )
   /SetVariables(Particles.Dark_Glow_CameraOffset=50) {isSetParameters=true}
   &CameraOffset(Camera Offset Amount=75)
  ]
  particleUpdateScript [
   /SetVariables(Particles.Dark_Glow=&MakeLinearColorFromVectorAndFloat(
     "Vector (RGB)"=&MakeVectorFromLinearColorRGB(Color=&MultiplyLinearColor(
       Linear Color=&ColorFromCurve(ColorCurve=DataInterface() {propertyValues {
          AlphaCurve {keys []}
          BlueCurve {keys [<at 0.1777 0.0315 RCIM_Cubic>,<at 1 0.0388 RCIM_Cubic>]}
          GreenCurve {keys [<at 0.1777 0.0629 RCIM_Cubic>,<at 1 0.0388 RCIM_Cubic>]}
          RedCurve {keys [<at 0.1777 0.45 RCIM_Cubic>,<at 1 0.8594 RCIM_Cubic>]}
         }})
       Scale Factor=&MakeVector4_FromFloat(Float=&Multiply_Float(A=0.5, B=1))
      ))
     "Float (Alpha)"=&Multiply_Float(
      A=&Multiply_Float(A=1, B=1)
      B=&FloatFromCurve(FloatCurve=DataInterface() {propertyValues {Curve {keys [<at 0 0>,<at 0.2 1>,<at 0.7 0.4>,<at 1 0>]}}})
     )
    )) {isSetParameters=true}
   /SetVariables(Particles.Dark_Glow_Size=&Multiply_Vector2(
     A=&Vector2DFromFloat(Value=180)
     B=&Vector2DFromFloat(Value=&FloatFromCurve(
       FloatCurve=DataInterface() {propertyValues {Curve {keys [<at 0 1.5 RCIM_Cubic>,<at 1 0.5>]}}}
       Scale Curve=&FN_FrontEndFloatModifier(Gameplay=1.5, Front End=0.6)
      ))
    )) {isSetParameters=true}
   &ParticleState()
   &ScaleColor(
    Scale Alpha=&Multiply_Float(
     A=1
     B=&FloatFromCurve(FloatCurve=DataInterface() {propertyValues {Curve {keys [<at 0 0>,<at 0.2 1>,<at 0.4 0.6>,<at 0.65 -0.05>]}}})
    )
    Scale RGB=&Multiply_VectorByFloat(
     Vector=&MakeVectorFromLinearColorRGB(Color=&ColorFromCurve(ColorCurve=DataInterface() {propertyValues {
         AlphaCurve {keys []}
         BlueCurve {keys [<at 0.0642 0.0507 RCIM_Cubic>,<at 0.4342 0.0304 RCIM_Cubic>]}
         GreenCurve {keys [<at 0.0642 0.0437 RCIM_Cubic>,<at 0.4342 0.0304 RCIM_Cubic>]}
         RedCurve {keys [<at 0.0642 0.7 RCIM_Cubic>,<at 0.4342 1 RCIM_Cubic>]}
        }}))
     Float=5
    )
    ScaleA=true
    ScaleRGB=true
   )
   &ScaleSpriteSize(Scale Factor=&Vector2DFromFloat(Value=&FloatFromCurve(
      FloatCurve=DataInterface() {propertyValues {Curve {keys [<at 0 1.5 RCIM_Cubic>,<at 0.8 0>]}}}
      Scale Curve=&FN_FrontEndFloatModifier(Gameplay=1.1, Front End=0.6)
     )))
   &FN_Light_Attributes(
    Color=<rgba 1.2 1.2 8 1>
    Create Color Attribute=true
    Create Exponent Attribute=true
    Create Position Attribute=true
    Create Radius Attribute=true
    Position=<20 0 0>
    Radius=100
    Scale Color=&Multiply_Float(
     A=3000
     B=&FloatFromCurve(FloatCurve=DataInterface() {propertyValues {Curve {keys [
          {arriveTangent=6.278, interpMode=RCIM_Cubic, leaveTangent=6.278, tangentMode=RCTM_User, time=0, value=0}
          {arriveTangent=-2.0311, interpMode=RCIM_Cubic, leaveTangent=-2.0311, tangentMode=RCTM_User, time=0.5, value=1}
          {arriveTangent=-2.1506, interpMode=RCIM_Cubic, tangentMode=RCTM_Break, time=1, value=0}
         ]}}})
    )
   ) {enabled=false}
   &ScaleColor(Scale RGB=&VectorFromFloat(Value=1), ScaleA=true, ScaleRGB=true) {module=ScaleColor002}
  ]
  rendererClasses [&NiagaraSpriteRendererProperties, &NiagaraLightRendererProperties]
  renderers [
   {rendererClass=&NiagaraSpriteRendererProperties}
   {rendererClass=&NiagaraSpriteRendererProperties, rendererIndex=1}
   {rendererClass=&NiagaraLightRendererProperties, rendererIndex=2}
  ]
 }
 Flash_Flare001 {
  emitterUpdateScript [
   &EmitterState(Life Cycle Mode=::NewEnumerator1, Loop Behavior=::NewEnumerator1, Loop Duration=1, UseLoopDelay=true)
   &SpawnBurst_Instantaneous(Spawn Count=1, Use Spawn Probability=false)
  ]
  particleSpawnScript [
   /Niagara/Modules/Spawn/Initialization/InitializeParticle(
    Color=<rgba 1 0.05 0.0658 0.1>
    Lifetime=3
    Module.Write Color=true
    Module.Write Lifetime=true
    Module.Write Mass=true
    Module.Write Material Random=true
    Module.Write Position=false
    Module.Write Scale=false
    Module.Write SpriteRotation=true
    Module.Write SpriteSize=true
    Position=&SimulationPosition()
    Sprite Rotation=&RandomRangeFloat(Minimum=0, Maximum=0.001)
    Sprite Size=<200 400>
   )
   &DynamicMaterialParameters(
    Index 0 Param 1=0.2
    Index 0 Param 2=1
    Module.Param0WriteEnabled=true
    Module.Param1WriteEnabled=true
    Module.Param2WriteEnabled=true
    Module.Param3WriteEnabled=true
   )
  ]
  particleUpdateScript [
   &ParticleState()
   &CameraOffset(Camera Offset Amount=&FN_LOD_Distance_Remap(Start Distance=300, End Distance=700, Minimum Output=0, Maximum Output=500, Clamp Output=true))
   &SolveForcesAndVelocity(Clamp Velocity=false, Limit Acceleration=false)
   &ScaleColor(
    Scale Alpha=&FloatFromCurve(
     FloatCurve=DataInterface() {propertyValues {Curve {keys [
         {interpMode=RCIM_Cubic, leaveTangent=4.7705, tangentMode=RCTM_Break, time=0, value=0}
         {interpMode=RCIM_Cubic, tangentMode=RCTM_Break, time=0.4, value=1}
         {arriveTangent=-2.7496, interpMode=RCIM_Cubic, leaveTangent=-2.7496, tangentMode=RCTM_User, time=1, value=0}
        ]}}}
     Scale Curve=&RandomRangeFloat(Minimum=0.5, Maximum=0.8, Evaluation Type=::NewEnumerator1)
    )
    ScaleA=true
    ScaleRGB=true
   )
   &ScaleSpriteSize(Scale Sprite Size Mode=::NewEnumerator0) {module=ScaleSpriteSize001}
   &DynamicMaterialParameters(
    Index 0 Param 1=400
    Module.Index0Param0Write=true
    Module.Index0Param1Write=true
    Module.Index0Param2Write=true
    Module.Index0Param3Write=true
   ) {module=DynamicMaterialParameters001}
  ]
  rendererClasses [&NiagaraSpriteRendererProperties]
  renderers [{rendererClass=&NiagaraSpriteRendererProperties}]
 }
 SmokeLoopIdle {
  emitterUpdateScript [
   &EmitterState(
    Inactive Response=::NewEnumerator1
    Life Cycle Mode=::NewEnumerator1
    Loop Behavior=::NewEnumerator1
    Loop Delay=0.5
    Loop Duration Mode=::NewEnumerator1
    UseLoopDelay=false
   )
   &SpawnBurst_Instantaneous(Use Spawn Probability=false, UseLoopCountLimit=false)
   /SetVariables(
    Emitter.CameraOffset=50000
    Emitter.CameraOfsCore=50000
    Emitter.CameraOfs_Outline=50000
    Emitter.Color:LinearColor=&ColorFromCurve(
     ColorCurve=DataInterface() {propertyValues {
       AlphaCurve {keys [<at 0 0 RCIM_Cubic>,<at 0.5578 2 RCIM_Cubic>]}
       BlueCurve {keys [<at 0.4494 0 RCIM_Cubic>]}
       GreenCurve {keys [<at 0.4494 0 RCIM_Cubic>]}
       RedCurve {keys [<at 0.4494 0 RCIM_Cubic>]}
      }}
     CurveIndex=$Emitter.Age
    )
    Emitter.Color_Core:LinearColor=<rgba 0.441 0.0208 1 1>
    Emitter.Color_Outline:LinearColor=&ColorFromCurve(
     ColorCurve=DataInterface() {propertyValues {
       AlphaCurve {keys [<at 0.2072 0>,<at 0.6601 1 RCIM_Cubic>]}
       BlueCurve {keys [<at 0 3.5 RCIM_Cubic>]}
       GreenCurve {keys [<at 0 0.035 RCIM_Cubic>]}
       RedCurve {keys [<at 0 0.728 RCIM_Cubic>]}
      }}
     CurveIndex=$Emitter.Age
    )
    Emitter.CoreDMP3:Vector4f=&MakeVector4(X=0.1, Y=0, Z=0, W=0)
   ) {isSetParameters=true}
   /SetVariables(
    Emitter.DynamicParam0_inky:Vector4f=&MakeVector4(
     X=&MakeFloatFromVector4(
      Vector 4=&MakeVector4(
       X=&FloatFromCurve(
        FloatCurve=DataInterface() {propertyValues {Curve {keys [<at 0 1 RCIM_Cubic>, {tangentMode=RCTM_User, time=1, value=0}]}}}
        CurveIndex=$Emitter.Age
        Scale Curve=0.5
       )
       Y=&RandomRangeFloat(Minimum=0, Maximum=1)
       Z=&FloatFromCurve(
        FloatCurve=DataInterface() {propertyValues {Curve {keys [{tangentMode=RCTM_User, time=0, value=0}, {tangentMode=RCTM_User, time=1, value=1}]}}}
        CurveIndex=$Emitter.Age
       )
       W=&FloatFromCurve(
        FloatCurve=DataInterface() {propertyValues {Curve {keys [{tangentMode=RCTM_User, time=0, value=1}, {tangentMode=RCTM_User, time=1, value=0}]}}}
        CurveIndex=$Emitter.Age
       )
      )
      Channel=::NewEnumerator0
     )
     Y=&FloatFromCurve(
      FloatCurve=DataInterface() {propertyValues {Curve {keys [{tangentMode=RCTM_User, time=0, value=0}, {tangentMode=RCTM_User, time=1, value=1}]}}}
      CurveIndex=$Emitter.Age
      Scale Curve=0.3
     )
     Z=&FloatFromCurve(
      FloatCurve=DataInterface() {propertyValues {Curve {keys [{tangentMode=RCTM_User, time=0, value=0}, {tangentMode=RCTM_User, time=1, value=1}]}}}
      CurveIndex=$Emitter.Age
     )
     W=&FloatFromCurve(
      FloatCurve=DataInterface() {propertyValues {Curve {keys [
          {interpMode=RCIM_Cubic, tangentMode=RCTM_User, time=0, value=1}
          {arriveTangent=-1.1812, interpMode=RCIM_Cubic, leaveTangent=-1.1812, tangentMode=RCTM_User, time=0.4, value=0.05}
          {interpMode=RCIM_Cubic, tangentMode=RCTM_User, time=1, value=1}
         ]}}}
      CurveIndex=$Emitter.Age
     )
    )
    Emitter.LightColor:LinearColor=<rgba 2 0 10 1>
    Emitter.LightRadius:NiagaraFloat=&Multiply_Float(
     A=&Add_Float(A=&Sine(Normalized Angle=$Emitter.Age, Period=0.78, Scale=0.2, Bias=1), B=1)
     B=100
    )
    Emitter.Scale:Vector3f=&VectorFromCurve(
     VectorCurve=DataInterface() {propertyValues {
       XCurve {keys [
         {arriveTangent=0.7455, interpMode=RCIM_Cubic, leaveTangent=0.7455, tangentMode=RCTM_User, time=0.4, value=0}
         {arriveTangent=0.7455, interpMode=RCIM_Cubic, leaveTangent=0.7455, tangentMode=RCTM_User, time=0.6, value=1.3}
         {arriveTangent=-0.5585, interpMode=RCIM_Cubic, leaveTangent=-0.5585, tangentMode=RCTM_User, time=0.8, value=1}
        ]}
       YCurve {keys [
         {arriveTangent=0.8624, interpMode=RCIM_Cubic, leaveTangent=0.8624, tangentMode=RCTM_User, time=0.4, value=0}
         {arriveTangent=0.8624, interpMode=RCIM_Cubic, leaveTangent=0.8624, tangentMode=RCTM_User, time=0.6, value=1.3}
         {arriveTangent=-2.2311, interpMode=RCIM_Cubic, leaveTangent=-2.2311, tangentMode=RCTM_User, time=0.8, value=1}
        ]}
       ZCurve {keys [
         {interpMode=RCIM_Cubic, tangentMode=RCTM_User, time=0.3, value=0}
         {arriveTangent=-1.954, interpMode=RCIM_Cubic, leaveTangent=-1.954, tangentMode=RCTM_User, time=0.55, value=1.35}
         {arriveTangent=-3.8442, interpMode=RCIM_Cubic, leaveTangent=-3.8442, tangentMode=RCTM_User, time=0.7, value=1}
        ]}
      }}
     CurveIndex=$Emitter.Age
     Scale Curve=<0.55 0.55 0.45>
    )
    Emitter.Scale2:Vector3f=&VectorFromCurve(
     VectorCurve=DataInterface() {propertyValues {XCurve {keys [<at 0.3 0>,<at 0.6 1>]}, YCurve {keys [<at 0.3 0>,<at 0.6 1>]}, ZCurve {keys [<at 0.4 0>,<at 0.5 1>]}}}
     CurveIndex=$Emitter.Age
     Scale Curve=<0.38 0.38 0.4>
    )
    Emitter.ScaleCore:Vector3f=&VectorFromCurve(
     VectorCurve=DataInterface() {propertyValues {XCurve {keys [<at 0.7 0>,<at 1.5 1>]}, YCurve {keys [<at 0.7 0>,<at 1.5 1>]}, ZCurve {keys [<at 0.7 0.8>,<at 1.5 1>]}}}
     CurveIndex=$Emitter.Age
     Scale Curve=<0.05 0.05 0.04>
    )
    Emitter.SpriteSize:Vector2f=&Multiply_Vector2D_ByFloat(
     Float=&MakeFloatFromVector(
      Vector=&Multiply_Vector(
       A=&VectorFromFloat(Value=&FloatFromCurve(
         FloatCurve=DataInterface() {propertyValues {Curve {keys [
             {arriveTangent=-1.0151, interpMode=RCIM_Cubic, leaveTangent=-1.0151, tangentMode=RCTM_User, time=0, value=1}
             {arriveTangent=-2.8172, interpMode=RCIM_Cubic, leaveTangent=-2.8172, tangentMode=RCTM_User, time=0.45, value=0.2}
            ]}}}
         CurveIndex=$Emitter.Age
        ))
       B=<1.4 1.4 0.4>
      )
      Channel=::NewEnumerator0
     )
     Vector2D=<220 200>
    )
   ) {isSetParameters=true}
   &SpawnBurst_Instantaneous(Spawn Time=8.55, Use Spawn Probability=false, UseLoopCountLimit=false) {module=SpawnBurst_Instantaneous001}
  ]
  particleSpawnScript [/Niagara/Modules/Spawn/Initialization/V2/InitializeParticle(
    Lifetime=3
    Mesh Scale Mode=::NewEnumerator3
    Ribbon Width Mode=::NewEnumerator1
    Sprite Size Mode=::NewEnumerator3
    UsePositionOffset=false
    Write Material Random=true
   )]
  particleUpdateScript [
   &ParticleState()
   &SolveForcesAndVelocity(Clamp Velocity=false, Limit Acceleration=false)
   /SetVariables(
    Particles.CameraOffset=100
    Particles.CameraOfsCore=100
    Particles.CameraOfs_Outline=100
    Particles.Color=&ColorFromCurve(
     ColorCurve=DataInterface() {propertyValues {
       AlphaCurve {keys [<at 0 0 RCIM_Cubic>,<at 0.5578 2 RCIM_Cubic>]}
       BlueCurve {keys [<at 0.4494 0 RCIM_Cubic>]}
       GreenCurve {keys [<at 0.4494 0 RCIM_Cubic>]}
       RedCurve {keys [<at 0.4494 0 RCIM_Cubic>]}
      }}
     CurveIndex=$Emitter.Age
    )
    Particles.Color_Core:LinearColor=<rgba 1 0.07 0.0855 1>
    Particles.Color_Outline:LinearColor=&ColorFromCurve(
     ColorCurve=DataInterface() {propertyValues {
       AlphaCurve {keys [<at 0.2072 0>,<at 0.6601 1 RCIM_Cubic>]}
       BlueCurve {keys [<at 0 0.4025 RCIM_Cubic>]}
       GreenCurve {keys [<at 0 0.35 RCIM_Cubic>]}
       RedCurve {keys [<at 0 3.5 RCIM_Cubic>]}
      }}
     CurveIndex=$Emitter.Age
    )
    Particles.CoreDMP3:Vector4f=&MakeVector4(X=0.1, Y=0, Z=0, W=0)
    Particles.Linear Color=&ColorFromCurve(
     ColorCurve=DataInterface() {propertyValues {
       AlphaCurve {keys [<at 0 0 RCIM_Cubic>,<at 0.5578 0.5 RCIM_Cubic>]}
       BlueCurve {keys [<at 0.4494 0 RCIM_Cubic>]}
       GreenCurve {keys [<at 0.4494 0 RCIM_Cubic>]}
       RedCurve {keys [<at 0.4494 0 RCIM_Cubic>]}
      }}
     CurveIndex=$Emitter.Age
    )
   ) {isSetParameters=true}
   /SetVariables(
    Particles.DynamicParam0_inky:Vector4f=&MakeVector4(
     X=&MakeFloatFromVector4(
      Vector 4=&MakeVector4(
       X=&FloatFromCurve(
        FloatCurve=DataInterface() {propertyValues {Curve {keys [<at 0 1 RCIM_Cubic>, {tangentMode=RCTM_User, time=1, value=0}]}}}
        CurveIndex=$Emitter.Age
        Scale Curve=0.5
       )
       Y=&RandomRangeFloat(Minimum=0, Maximum=1)
       Z=&FloatFromCurve(
        FloatCurve=DataInterface() {propertyValues {Curve {keys [{tangentMode=RCTM_User, time=0, value=0}, {tangentMode=RCTM_User, time=1, value=1}]}}}
        CurveIndex=$Emitter.Age
       )
       W=&FloatFromCurve(
        FloatCurve=DataInterface() {propertyValues {Curve {keys [{tangentMode=RCTM_User, time=0, value=1}, {tangentMode=RCTM_User, time=1, value=0}]}}}
        CurveIndex=$Emitter.Age
       )
      )
      Channel=::NewEnumerator0
     )
     Y=&FloatFromCurve(
      FloatCurve=DataInterface() {propertyValues {Curve {keys [{tangentMode=RCTM_User, time=0, value=0}, {tangentMode=RCTM_User, time=1, value=1}]}}}
      CurveIndex=$Emitter.Age
      Scale Curve=0.3
     )
     Z=&FloatFromCurve(
      FloatCurve=DataInterface() {propertyValues {Curve {keys [{tangentMode=RCTM_User, time=0, value=0}, {tangentMode=RCTM_User, time=1, value=1}]}}}
      CurveIndex=$Emitter.Age
     )
     W=&FloatFromCurve(
      FloatCurve=DataInterface() {propertyValues {Curve {keys [
          {interpMode=RCIM_Cubic, tangentMode=RCTM_User, time=0, value=1}
          {arriveTangent=-1.1812, interpMode=RCIM_Cubic, leaveTangent=-1.1812, tangentMode=RCTM_User, time=0.4, value=0.05}
          {interpMode=RCIM_Cubic, tangentMode=RCTM_User, time=1, value=1}
         ]}}}
      CurveIndex=$Emitter.Age
     )
    )
    Particles.LightColor=<rgba 2 0 10 1>
    Particles.LightRadius=&Multiply_Float(
     A=&Add_Float(A=&Sine(Normalized Angle=$Emitter.Age, Period=0.78, Scale=0.2, Bias=1), B=1)
     B=100
    )
    Particles.Scale=&VectorFromCurve(
     VectorCurve=DataInterface() {propertyValues {
       XCurve {keys [
         {arriveTangent=0.7455, interpMode=RCIM_Cubic, leaveTangent=0.7455, tangentMode=RCTM_User, time=0.4, value=0}
         {arriveTangent=0.7455, interpMode=RCIM_Cubic, leaveTangent=0.7455, tangentMode=RCTM_User, time=0.6, value=1.3}
         {arriveTangent=-0.5585, interpMode=RCIM_Cubic, leaveTangent=-0.5585, tangentMode=RCTM_User, time=0.8, value=1}
        ]}
       YCurve {keys [
         {arriveTangent=0.8624, interpMode=RCIM_Cubic, leaveTangent=0.8624, tangentMode=RCTM_User, time=0.4, value=0}
         {arriveTangent=0.8624, interpMode=RCIM_Cubic, leaveTangent=0.8624, tangentMode=RCTM_User, time=0.6, value=1.3}
         {arriveTangent=-2.2311, interpMode=RCIM_Cubic, leaveTangent=-2.2311, tangentMode=RCTM_User, time=0.8, value=1}
        ]}
       ZCurve {keys [
         {interpMode=RCIM_Cubic, tangentMode=RCTM_User, time=0.3, value=0}
         {arriveTangent=-1.954, interpMode=RCIM_Cubic, leaveTangent=-1.954, tangentMode=RCTM_User, time=0.55, value=1.35}
         {arriveTangent=-3.8442, interpMode=RCIM_Cubic, leaveTangent=-3.8442, tangentMode=RCTM_User, time=0.7, value=1}
        ]}
      }}
     CurveIndex=$Emitter.Age
     Scale Curve=&Multiply_VectorByFloat(
      Vector=<1 1 0.85>
      Float=&FloatFromCurve(
       FloatCurve=DataInterface() {propertyValues {Curve {keys [{interpMode=RCIM_Cubic, tangentMode=RCTM_Break, time=0.2, value=1}, <at 1 2>]}}}
       Scale Curve=&RandomRangeFloat(Minimum=0.8, Maximum=1, Evaluation Type=::NewEnumerator1)
      )
     )
    )
    Particles.Scale2:Vector3f=&VectorFromCurve(
     VectorCurve=DataInterface() {propertyValues {XCurve {keys [<at 0.3 0>,<at 0.6 1>]}, YCurve {keys [<at 0.3 0>,<at 0.6 1>]}, ZCurve {keys [<at 0.4 0>,<at 0.5 1>]}}}
     CurveIndex=$Emitter.Age
     Scale Curve=&Multiply_VectorByFloat(
      Vector=<1 1 1>
      Float=&FloatFromCurve(
       FloatCurve=DataInterface() {propertyValues {Curve {keys [{interpMode=RCIM_Cubic, tangentMode=RCTM_Break, time=0.2, value=1}, <at 1 2>]}}}
       Scale Curve=&RandomRangeFloat(Minimum=0.8, Maximum=1, Evaluation Type=::NewEnumerator1)
      )
     )
    )
    Particles.ScaleCore:Vector3f=&VectorFromCurve(
     VectorCurve=DataInterface() {propertyValues {XCurve {keys [<at 0.7 0>,<at 1.5 1>]}, YCurve {keys [<at 0.7 0>,<at 1.5 1>]}, ZCurve {keys [<at 0.7 0.8>,<at 1.5 1>]}}}
     CurveIndex=$Emitter.Age
     Scale Curve=<0.1 0.1 0.1>
    )
    Particles.ScaleGlow:Vector3f=&VectorFromCurve(
     VectorCurve=DataInterface() {propertyValues {
       XCurve {keys [
         {arriveTangent=0.7455, interpMode=RCIM_Cubic, leaveTangent=0.7455, tangentMode=RCTM_User, time=0.4, value=0}
         {arriveTangent=0.7455, interpMode=RCIM_Cubic, leaveTangent=0.7455, tangentMode=RCTM_User, time=0.6, value=1.3}
         {arriveTangent=-0.5585, interpMode=RCIM_Cubic, leaveTangent=-0.5585, tangentMode=RCTM_User, time=0.8, value=1}
        ]}
       YCurve {keys [
         {arriveTangent=0.8624, interpMode=RCIM_Cubic, leaveTangent=0.8624, tangentMode=RCTM_User, time=0.4, value=0}
         {arriveTangent=0.8624, interpMode=RCIM_Cubic, leaveTangent=0.8624, tangentMode=RCTM_User, time=0.6, value=1.3}
         {arriveTangent=-2.2311, interpMode=RCIM_Cubic, leaveTangent=-2.2311, tangentMode=RCTM_User, time=0.8, value=1}
        ]}
       ZCurve {keys [
         {interpMode=RCIM_Cubic, tangentMode=RCTM_User, time=0.3, value=0}
         {arriveTangent=-1.954, interpMode=RCIM_Cubic, leaveTangent=-1.954, tangentMode=RCTM_User, time=0.55, value=1.35}
         {arriveTangent=-3.8442, interpMode=RCIM_Cubic, leaveTangent=-3.8442, tangentMode=RCTM_User, time=0.7, value=1}
        ]}
      }}
     CurveIndex=$Emitter.Age
    )
    Particles.SpriteSize=&Multiply_Vector2D_ByFloat(
     Float=&MakeFloatFromVector(
      Vector=&Multiply_Vector(
       A=&VectorFromFloat(Value=&FloatFromCurve(
         FloatCurve=DataInterface() {propertyValues {Curve {keys [
             {arriveTangent=-1.0151, interpMode=RCIM_Cubic, leaveTangent=-1.0151, tangentMode=RCTM_User, time=0, value=1}
             {arriveTangent=-2.8172, interpMode=RCIM_Cubic, leaveTangent=-2.8172, tangentMode=RCTM_User, time=0.45, value=0.2}
            ]}}}
         CurveIndex=$Emitter.Age
        ))
       B=<1.4 1.4 0.4>
      )
      Channel=::NewEnumerator0
     )
     Vector2D=<220 200>
    )
   ) {isSetParameters=true}
  ]
  rendererClasses [&NiagaraMeshRendererProperties]
  renderers [
   {rendererClass=&NiagaraMeshRendererProperties, rendererIndex=0}
   {rendererClass=&NiagaraMeshRendererProperties, rendererIndex=1}
   {rendererClass=&NiagaraMeshRendererProperties, rendererIndex=2}
   {rendererClass=&NiagaraMeshRendererProperties, rendererIndex=3}
  ]
 }
 TwistAbsorb {
  emitterUpdateScript [
   &EmitterState(Life Cycle Mode=::NewEnumerator1, Loop Behavior=::NewEnumerator1, Loop Delay=0.1, Loop Duration=2, UseLoopDelay=true)
   &SpawnBurst_Instantaneous(Use Spawn Probability=false, UseLoopCountLimit=false)
  ]
  particleSpawnScript [
   &InitializeParticle(
    Color=<rgba 3 0.3 0.345 1>
    Lifetime=0.5
    Sprite Rotation Mode=::NewEnumerator2
    Sprite Size Mode=::NewEnumerator3
    Uniform Sprite Size=1010.4492
    UsePositionOffset=false
    Write Material Random=true
   )
   &CameraOffset(Camera Offset Amount=100) {module=CameraOffset001}
   &SystemLocation(MaskSpawnGroup=false) {enabled=false}
  ]
  particleUpdateScript [
   &CameraOffset(Camera Offset Amount=&Lerp_Float(
     A=600
     B=-100
     Alpha=&FloatFromCurve(FloatCurve=DataInterface() {propertyValues {Curve {keys [{interpMode=RCIM_Cubic, tangentMode=RCTM_User, time=0.35, value=0}, {interpMode=RCIM_Cubic, tangentMode=RCTM_User, time=0.7, value=1}]}}})
    )) {enabled=false}
   &ParticleState()
   &ScaleSpriteSize(Uniform Curve Sprite Scale=DataInterface() {propertyValues {Curve {keys [
        {arriveTangent=12.5362, interpMode=RCIM_Cubic, leaveTangent=12.5362, tangentMode=RCTM_User, time=0, value=0.3}
        {arriveTangent=-0.523, interpMode=RCIM_Cubic, leaveTangent=-0.523, tangentMode=RCTM_User, time=0.1, value=1}
        <at 1 0 RCIM_Cubic>
       ]}}})
   &SolveForcesAndVelocity(Clamp Velocity=false, Limit Acceleration=false)
   &DynamicMaterialParameters(
    Index 0 Param 1=&Lerp_Float(
     A=-0.5
     B=2
     Alpha=&FloatFromCurve(FloatCurve=DataInterface() {propertyValues {Curve {keys [
          {arriveTangent=0.1556, interpMode=RCIM_Cubic, leaveTangent=0.1556, tangentMode=RCTM_User, time=0, value=0}
          {arriveTangent=2.2716, interpMode=RCIM_Cubic, leaveTangent=2.2716, tangentMode=RCTM_User, time=0.8, value=1}
         ]}}})
    )
    Index 0 Param 2=1
    Index 0 Param 3=&Lerp_Float(
     A=-0.1
     B=0.5
     Alpha=&FloatFromCurve(FloatCurve=DataInterface() {propertyValues {Curve {keys [
          {arriveTangent=0.1235, interpMode=RCIM_Cubic, leaveTangent=0.1235, tangentMode=RCTM_User, time=0.4, value=0}
          {arriveTangent=1.955, interpMode=RCIM_Cubic, leaveTangent=1.955, tangentMode=RCTM_User, time=1, value=1}
         ]}}})
    )
    Index 0 Param 4=&Lerp_Float(
     A=0
     B=1
     Alpha=&FloatFromCurve(FloatCurve=DataInterface() {propertyValues {Curve {keys [
          {arriveTangent=0.3497, interpMode=RCIM_Cubic, leaveTangent=0.3497, tangentMode=RCTM_User, time=0, value=0}
          {interpMode=RCIM_Cubic, tangentMode=RCTM_User, time=1, value=1}
         ]}}})
    )
    Module.Index0Param0Write=true
    Module.Index0Param1Write=true
    Module.Index0Param2Write=true
    Module.Index0Param3Write=true
   )
   &ScaleColor(
    Scale Alpha=&FloatFromCurve(
     FloatCurve=DataInterface() {propertyValues {Curve {keys [
         <at 0 0 RCIM_Cubic>
         {arriveTangent=-0.0178, interpMode=RCIM_Cubic, leaveTangent=-0.0178, tangentMode=RCTM_User, time=0.1, value=1}
         {interpMode=RCIM_Cubic, tangentMode=RCTM_User, time=0.3, value=1}
         {interpMode=RCIM_Cubic, tangentMode=RCTM_User, time=1, value=0}
        ]}}}
     Scale Curve=2
    )
    Scale RGB=&MakeVectorFromLinearColorRGB(Color=&ColorFromCurve(ColorCurve=DataInterface() {propertyValues {
        AlphaCurve {keys [<at 0 1 RCIM_Cubic>]}
        BlueCurve {keys [<at 0.1216 1 RCIM_Cubic>,<at 0.8838 0.0027 RCIM_Cubic>]}
        GreenCurve {keys [<at 0.1216 1 RCIM_Cubic>,<at 0.8838 0.0018 RCIM_Cubic>]}
        RedCurve {keys [<at 0.1216 1 RCIM_Cubic>,<at 0.8838 0.03 RCIM_Cubic>]}
       }}))
    ScaleA=true
    ScaleRGB=true
   )
   /SetVariables(
    Particles.ColorDark:LinearColor=&MultiplyLinearColor(
     Linear Color=<rgba 0.02 0.006 0.0074 1>
     Scale Factor=&MakeFloat4FromVectorAndFloat(
      "Vector (RGB)"=<1 1 1>
      "Float (Alpha)"=&FloatFromCurve(
       FloatCurve=DataInterface() {propertyValues {Curve {keys [
           <at 0 0 RCIM_Cubic>
           {arriveTangent=-2.1159, interpMode=RCIM_Cubic, leaveTangent=-2.1159, tangentMode=RCTM_User, time=0.2, value=1}
           {interpMode=RCIM_Cubic, tangentMode=RCTM_User, time=1, value=0}
          ]}}}
       Scale Curve=2
      )
     )
    )
    Particles.DynamicMaterialParameterDark:Vector4f=&MakeVector4(
     X=&Lerp_Float(
      A=0
      B=1
      Alpha=&FloatFromCurve(FloatCurve=DataInterface() {propertyValues {Curve {keys [
           {interpMode=RCIM_Cubic, tangentMode=RCTM_User, time=0, value=0}
           {arriveTangent=1.7687, interpMode=RCIM_Cubic, leaveTangent=1.7687, tangentMode=RCTM_User, time=1, value=1}
          ]}}})
     )
     Y=1
     Z=&Lerp_Float(
      A=-0.5
      B=0.3
      Alpha=&FloatFromCurve(FloatCurve=DataInterface() {propertyValues {Curve {keys [
           {arriveTangent=0.1235, interpMode=RCIM_Cubic, leaveTangent=0.1235, tangentMode=RCTM_User, time=0.2, value=0}
           {arriveTangent=1.955, interpMode=RCIM_Cubic, leaveTangent=1.955, tangentMode=RCTM_User, time=1, value=1}
          ]}}})
     )
     W=0.1
    )
    Particles.SpriteSizeDark:Vector2f=&Multiply_Vector2D_ByFloat(Float=1, Vector2D=$Particles.SpriteSize)
   ) {isSetParameters=true}
  ]
  rendererClasses [&NiagaraSpriteRendererProperties]
  renderers [
   {rendererClass=&NiagaraSpriteRendererProperties}
   {rendererClass=&NiagaraSpriteRendererProperties, rendererIndex=1}
   {rendererClass=&NiagaraSpriteRendererProperties, rendererIndex=2}
  ]
 }
}
summary {
 systemName=NS_VoidMask_WanderingBossRelocation_GateOpen
 emitters [
  {emitterName=SmokeLoopIdle, rendererClasses [&NiagaraMeshRendererProperties]}
  {emitterName=TwistAbsorb, rendererClasses [&NiagaraSpriteRendererProperties]}
  {emitterName=Flare001, rendererClasses [&NiagaraSpriteRendererProperties, &NiagaraLightRendererProperties]}
  {emitterName=Flash_Flare001, rendererClasses [&NiagaraSpriteRendererProperties]}
 ]
 userVariables [
  User.AimVec {type=Vector3f, description="", defaultValue {struct=Vector3f, value=<0 1 1>}}
  User.bIsFrontEnd {type=NiagaraBool}
  User.bIsFrontEndPreview {type=NiagaraBool}
  User.Offset Pos {type=NiagaraPosition, description="", defaultValue {}}
 ]
}
