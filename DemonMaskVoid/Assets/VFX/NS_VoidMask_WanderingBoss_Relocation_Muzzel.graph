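# Niagara muzzle effect for the Void Mask Wandering Boss relocation, directed along User.AimVec.
# LongFlare bursts a red sprite and light that travel along the aim vector while fading.
# Tendrils and Trail burst aim-oriented meshes: tendrils spin and thin, while the trail stretches lengthwise.
# Multiple mesh renderers accompany separate dark and vortex color and dynamic-material attributes, animated mostly by curves; User.PortalPosition and Global Scale are declared but unreferenced in these modules.
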NiagaraSystem /DemonMaskVoid/Assets/VFX/NS_VoidMask_WanderingBoss_Relocation_Muzzel
dependencies {
 usedDataInterfaces [&NiagaraDataInterfaceCurve, &NiagaraDataInterfaceColorCurve]
 usedDynamicInputs []
 usedModules [&SystemState]
}
emitters {
 LongFlare {
  emitterUpdateScript [
   &EmitterState(Life Cycle Mode=::NewEnumerator1, Loop Behavior=::NewEnumerator1, Loop Duration=2.1834, UseLoopDelay=true)
   &SpawnBurst_Instantaneous(Spawn Count=1, Use Spawn Probability=false)
  ]
  particleSpawnScript [
   /Niagara/Modules/Spawn/Initialization/InitializeParticle(
    Color=<rgba 10 0.2 0.3633 1>
    Lifetime=1.0716
    Module.Write Color=true
    Module.Write Lifetime=true
    Module.Write Mass=true
    Module.Write Material Random=true
    Module.Write Position=false
    Module.Write Scale=false
    Module.Write SpriteRotation=true
    Module.Write SpriteSize=true
    Position=&SimulationPosition()
    Sprite Size=&Multiply_Vector2D_ByFloat(Float=0.898, Vector2D=&Multiply_Vector2D_ByFloat(Float=0.2509, Vector2D=<16000 3000>))
   )
   &DynamicMaterialParameters(
    Index 0 Param 1=0.2
    Index 0 Param 2=1
    Module.Param0WriteEnabled=true
    Module.Param0WriteEnabled1=true
    Module.Param0WriteEnabled2=true
    Module.Param0WriteEnabled3=true
    Module.Param1WriteEnabled=true
    Module.Param1WriteEnabled1=true
    Module.Param1WriteEnabled2=true
    Module.Param1WriteEnabled3=true
    Module.Param2WriteEnabled=true
    Module.Param2WriteEnabled1=true
    Module.Param2WriteEnabled2=true
    Module.Param2WriteEnabled3=true
    Module.Param3WriteEnabled=true
    Module.Param3WriteEnabled1=true
    Module.Param3WriteEnabled2=true
    Module.Param3WriteEnabled3=true
    Write Parameter Index 1=true
    Write Parameter Index 2=true
    Write Parameter Index 3=true
   )
  ]
  particleUpdateScript [
   &ParticleState()
   &CameraOffset(Camera Offset Amount=100)
   &SolveForcesAndVelocity(Clamp Velocity=false, Limit Acceleration=false)
   &ScaleColor(
    Color Value To Scale=$Particles.Color
    Scale Alpha=&FloatFromCurve(FloatCurve=DataInterface() {propertyValues {Curve {keys [
         {arriveTangent=-0.3704, interpMode=RCIM_Cubic, leaveTangent=-0.3704, tangentMode=RCTM_User, time=0, value=1}
         {arriveTangent=-3.3568, tangentMode=RCTM_User, time=1, value=0}
        ]}}})
    ScaleA=true
    ScaleRGB=true
   )
   &ScaleSpriteSize(Scale Factor=&MakeVector2D(
     X=&FloatFromCurve(
      FloatCurve=DataInterface() {propertyValues {Curve {keys [
          {arriveTangent=-0.1862, interpMode=RCIM_Cubic, leaveTangent=-0.1862, tangentMode=RCTM_User, time=0, value=1}
          {arriveTangent=-1.7748, interpMode=RCIM_Cubic, leaveTangent=-1.7748, tangentMode=RCTM_User, time=1, value=0.3}
         ]}}}
      Scale Curve=&RandomRangeFloat(Minimum=0.5, Maximum=1, Evaluation Type=::NewEnumerator1)
     )
     Y=1
    ))
   &ScaleSpriteSize(Scale Sprite Size Mode=::NewEnumerator0) {module=ScaleSpriteSize001}
   &FN_Light_Attributes(
    Color=<rgba 1 0 0 1>
    Create Color Attribute=true
    Create Exponent Attribute=true
    Create Radius Attribute=true
    Enable Particle Light Enable=true
    Exponent=3
    Light Enabled=true
    Radius=1095.8392
    Scale Color=&Multiply_Float(
     A=43.9151
     B=&FloatFromCurve(FloatCurve=DataInterface() {propertyValues {Curve {keys [
          {arriveTangent=2.7293, interpMode=RCIM_Cubic, leaveTangent=2.7293, tangentMode=RCTM_User, time=0, value=0}
          {arriveTangent=-0.1593, interpMode=RCIM_Cubic, leaveTangent=-0.1593, tangentMode=RCTM_User, time=0.3764, value=0.9847}
          {arriveTangent=-0.0405, interpMode=RCIM_Cubic, leaveTangent=-0.0405, tangentMode=RCTM_User, time=1, value=0}
         ]}}})
    )
   )
   &SystemLocation(
    MaskSpawnGroup=false
    Offset=&Multiply_VectorByFloat(
     Vector=$User.AimVec
     Float=&FloatFromCurve(
      FloatCurve=DataInterface() {propertyValues {Curve {keys [{interpMode=RCIM_Cubic, tangentMode=RCTM_Break, time=0, value=0}, {interpMode=RCIM_Cubic, tangentMode=RCTM_Break, time=1, value=1}]}}}
      Scale Curve=3269.5452
     )
    )
    Offset Coordinate Space=::World
   )
   &FN_Sprite_Attributes(
    Color=&MultiplyLinearColor(
     Linear Color=<rgba 1 0.1 0.115 0.1>
     Scale Factor=&MakeVector4(
      X=1
      Y=1
      Z=1
      W=&FloatFromCurve(FloatCurve=DataInterface() {propertyValues {Curve {keys [
           {interpMode=RCIM_Cubic, tangentMode=RCTM_User, time=0, value=1}
           {arriveTangent=-1.6204, interpMode=RCIM_Cubic, leaveTangent=-1.6204, tangentMode=RCTM_User, time=1, value=0}
          ]}}})
     )
    )
    Create Color Attribute=true
    Create SpriteSize Attribute=true
    SpriteSize=<2000 2000>
    Write to Module Namespace=true
   )
  ]
  rendererClasses [&NiagaraSpriteRendererProperties, &NiagaraLightRendererProperties]
  renderers [
   {rendererClass=&NiagaraSpriteRendererProperties}
   {rendererClass=&NiagaraLightRendererProperties, rendererIndex=1}
   {rendererClass=&NiagaraSpriteRendererProperties, rendererIndex=2}
  ]
 }
 Tendrils {
  emitterUpdateScript [
   &EmitterState(Life Cycle Mode=::NewEnumerator1, Loop Behavior=::NewEnumerator1, Loop Delay=0.1, Loop Duration=2, UseLoopDelay=true)
   &SpawnBurst_Instantaneous(Use Spawn Probability=false, UseLoopCountLimit=false)
  ]
  particleSpawnScript [
   /Niagara/Modules/Spawn/Initialization/V2/InitializeParticle(
    Lifetime=4
    Mass Mode=::NewEnumerator3
    Mesh Scale=&Multiply_VectorByFloat(Vector=<1 1.75 1.75>, Float=2.1737)
    Mesh Scale Mode=::NewEnumerator4
    Sprite Rotation Mode=::NewEnumerator2
    Sprite Size Mode=::NewEnumerator1
    Uniform Sprite Size Max=12
    Uniform Sprite Size Min=6
    UsePositionOffset=false
    Write Material Random=true
   )
   &InitialMeshOrientation(
    Mesh Axis To Orient=<-1 0 0>
    Mesh Orientation Mode=::NewEnumerator2
    Orientation Vector=$User.AimVec
    Rotation=&MakeVector(X=&RandomRangeFloat(Minimum=-1, Maximum=1), Y=0, Z=0)
    Use Rotation Vector=true
   )
   &SystemLocation(
    MaskSpawnGroup=false
    Offset=&AddVectorToPosition(Position=&Multiply_VectorByFloat(Vector=$User.AimVec, Float=-20), Vector=<0 0 0>)
   )
   /SetVariables(Particles.MeshOrientationVortex:Quat4f=$Particles.MeshOrientation) {isSetParameters=true}
  ]
  particleUpdateScript [
   &ParticleState()
   &ScaleMeshSize(Scale Factor=&MakeVector(
     X=1
     Y=&FloatFromCurve(FloatCurve=DataInterface() {propertyValues {Curve {keys [{interpMode=RCIM_Cubic, tangentMode=RCTM_User, time=0, value=1}, {interpMode=RCIM_Cubic, tangentMode=RCTM_User, time=1, value=0.25}]}}})
     Z=&FloatFromCurve(FloatCurve=DataInterface() {propertyValues {Curve {keys [{interpMode=RCIM_Cubic, tangentMode=RCTM_User, time=0, value=1}, {interpMode=RCIM_Cubic, tangentMode=RCTM_User, time=1, value=0.25}]}}})
    )) {module=ScaleMeshSize001}
   &UpdateMeshOrientation(
    Rotation Rate=&Lerp_Float(
     A=0.7
     B=0.2
     Alpha=&FloatFromCurve(FloatCurve=DataInterface() {propertyValues {Curve {keys [
          {arriveTangent=1.7929, interpMode=RCIM_Cubic, leaveTangent=1.7929, tangentMode=RCTM_User, time=0, value=0}
          {interpMode=RCIM_Cubic, tangentMode=RCTM_User, time=1, value=1}
         ]}}})
    )
    Rotation Vector=<1.8039 0 0>
   )
   &ScaleMeshSize(Scale Factor=&Multiply_VectorByFloat(
     Vector=&MakeVector(
      X=&Lerp_Float(
       A=0.3
       B=0.7
       Alpha=&FloatFromCurve(FloatCurve=DataInterface() {propertyValues {Curve {keys [
            {arriveTangent=3.0241, interpMode=RCIM_Cubic, leaveTangent=3.0241, tangentMode=RCTM_User, time=0, value=0}
            {interpMode=RCIM_Cubic, tangentMode=RCTM_User, time=0.8, value=1}
           ]}}})
      )
      Y=&FloatFromCurve(FloatCurve=DataInterface() {propertyValues {Curve {keys [
           {arriveTangent=-2.2294, interpMode=RCIM_Cubic, leaveTangent=-2.2294, tangentMode=RCTM_User, time=0.2, value=1}
           {arriveTangent=-0.0663, interpMode=RCIM_Cubic, leaveTangent=-0.0663, tangentMode=RCTM_User, time=0.7, value=0.33}
          ]}}})
      Z=&FloatFromCurve(FloatCurve=DataInterface() {propertyValues {Curve {keys [
           {arriveTangent=-2.2294, interpMode=RCIM_Cubic, leaveTangent=-2.2294, tangentMode=RCTM_User, time=0.2, value=1}
           {arriveTangent=-0.0663, interpMode=RCIM_Cubic, leaveTangent=-0.0663, tangentMode=RCTM_User, time=0.7, value=0.33}
          ]}}})
     )
     Float=1
    )) {enabled=false}
   &ScaleColor(
    Scale Alpha=&FloatFromCurve(FloatCurve=DataInterface() {propertyValues {Curve {keys [
         {arriveTangent=10.1427, interpMode=RCIM_Cubic, leaveTangent=10.1427, tangentMode=RCTM_User, time=0, value=0}
         {arriveTangent=-0.3382, interpMode=RCIM_Cubic, leaveTangent=-0.3382, tangentMode=RCTM_User, time=0.1, value=1}
         {arriveTangent=-3.4639, interpMode=RCIM_Cubic, leaveTangent=-3.4639, tangentMode=RCTM_User, time=1, value=0}
        ]}}})
    Scale RGB=&MakeVectorFromLinearColorRGB(Color=&ColorFromCurve(ColorCurve=DataInterface() {propertyValues {
        AlphaCurve {keys [<at 0 1 RCIM_Cubic>]}
        BlueCurve {keys [<at 0 0.1317 RCIM_Cubic>,<at 0.9918 0.0064>]}
        GreenCurve {keys [<at 0 0.1 RCIM_Cubic>,<at 0.9918 0.006>]}
        RedCurve {keys [<at 0 2 RCIM_Cubic>,<at 0.9918 0.03>]}
       }}))
    ScaleA=true
    ScaleRGB=true
   )
   /SetVariables(Particles.ColorDark:LinearColor=&MakeLinearColorFromVectorAndFloat(
     "Vector (RGB)"=&MakeVectorFromLinearColorRGB(Color=&ColorFromCurve(ColorCurve=DataInterface() {propertyValues {
         AlphaCurve {keys [<at 0 1 RCIM_Cubic>]}
         BlueCurve {keys [<at 0 0.0345 RCIM_Cubic>,<at 0.3152 0>]}
         GreenCurve {keys [<at 0 0.03 RCIM_Cubic>,<at 0.3152 0>]}
         RedCurve {keys [<at 0 0.3 RCIM_Cubic>,<at 0.3152 0>]}
        }}))
     "Float (Alpha)"=&FloatFromCurve(FloatCurve=DataInterface() {propertyValues {Curve {keys [
          {interpMode=RCIM_Cubic, tangentMode=RCTM_User, time=0.1, value=0}
          {interpMode=RCIM_Cubic, tangentMode=RCTM_User, time=0.2, value=1}
          {interpMode=RCIM_Cubic, tangentMode=RCTM_User, time=1, value=0}
         ]}}})
    )) {isSetParameters=true}
   &SolveForcesAndVelocity(Clamp Velocity=false, Limit Acceleration=false)
   &DynamicMaterialParameters(
    Index 0 Param 1=0.1
    Index 0 Param 2=&Lerp_Float(
     A=0.3
     B=2.5
     Alpha=&FloatFromCurve(FloatCurve=DataInterface() {propertyValues {Curve {keys [<at 0 0>,<at 1 1>]}}})
    )
    Index 0 Param 3=&Lerp_Float(
     A=0.5
     B=1.5
     Alpha=&FloatFromCurve(FloatCurve=DataInterface() {propertyValues {Curve {keys [<at 0 0>,<at 1 1>]}}})
    )
    Index 0 Param 4=&Lerp_Float(
     A=0.2
     B=1
     Alpha=&FloatFromCurve(FloatCurve=DataInterface() {propertyValues {Curve {keys [
          {arriveTangent=0.2526, interpMode=RCIM_Cubic, leaveTangent=0.2526, tangentMode=RCTM_User, time=0, value=0}
          {arriveTangent=1.7785, interpMode=RCIM_Cubic, leaveTangent=1.7785, tangentMode=RCTM_User, time=1, value=1}
         ]}}})
    )
    Module.Index0Param0Write=true
    Module.Index0Param1Write=true
    Module.Index0Param2Write=true
    Module.Index0Param3Write=true
   )
   /SetVariables(Particles.DynamicMaterialParameterDark:Vector4f=&MakeVector4(
     X=1
     Y=&Lerp_Float(
      A=0.2
      B=2
      Alpha=&FloatFromCurve(FloatCurve=DataInterface() {propertyValues {Curve {keys [<at 0 0>,<at 1 1>]}}})
     )
     Z=&Lerp_Float(
      A=0.5
      B=1
      Alpha=&FloatFromCurve(FloatCurve=DataInterface() {propertyValues {Curve {keys [<at 0 0>,<at 0.7 1>]}}})
     )
     W=&Lerp_Float(
      A=0
      B=1
      Alpha=&FloatFromCurve(FloatCurve=DataInterface() {propertyValues {Curve {keys [
           {arriveTangent=0.2526, interpMode=RCIM_Cubic, leaveTangent=0.2526, tangentMode=RCTM_User, time=0.2, value=0}
           {arriveTangent=1.7785, interpMode=RCIM_Cubic, leaveTangent=1.7785, tangentMode=RCTM_User, time=1, value=1}
          ]}}})
     )
    )) {isSetParameters=true}
   /SetVariables(Particles.DynamicMaterialParameterVortex:Vector4f=&MakeVector4(
     X=&Lerp_Float(
      A=0.4
      B=-2
      Alpha=&FloatFromCurve(FloatCurve=DataInterface() {propertyValues {Curve {keys [
           {arriveTangent=5.2247, interpMode=RCIM_Cubic, leaveTangent=5.2247, tangentMode=RCTM_User, time=0, value=0}
           {interpMode=RCIM_Cubic, tangentMode=RCTM_User, time=0.4, value=1}
          ]}}})
     )
     Y=&Lerp_Float(
      A=1
      B=4
      Alpha=&FloatFromCurve(FloatCurve=DataInterface() {propertyValues {Curve {keys [<at 0 0>,<at 0.5 1>]}}})
     )
     Z=&Lerp_Float(
      A=0
      B=1
      Alpha=&FloatFromCurve(FloatCurve=DataInterface() {propertyValues {Curve {keys [<at 0 0>,<at 0.3 1>]}}})
     )
     W=-0.1
    )) {isSetParameters=true}
   /SetVariables(Particles.ScaleVortex:Vector3f=&Multiply_Vector(A=<1 1 1>, B=<3 4 4>)) {isSetParameters=true}
   /SetVariables(Particles.ColorVortex:LinearColor=&ColorFromCurve(ColorCurve=DataInterface() {propertyValues {
       AlphaCurve {keys [<at 0 0>, {arriveTangent=1.0121, interpMode=RCIM_Cubic, leaveTangent=1.0121, time=0.046, value=2}, <at 0.988 1>]}
       BlueCurve {keys [<at 0.1136 2 RCIM_Cubic>,<at 0.5005 0.02 RCIM_Cubic>]}
       GreenCurve {keys [<at 0.1136 0 RCIM_Cubic>,<at 0.5005 0.01 RCIM_Cubic>]}
       RedCurve {keys [<at 0.1136 0.5413 RCIM_Cubic>,<at 0.5005 0.0112 RCIM_Cubic>]}
      }})) {isSetParameters=true}
   &CameraOffset(Camera Offset Amount=100)
  ]
  rendererClasses [&NiagaraMeshRendererProperties]
  renderers [
   {rendererClass=&NiagaraMeshRendererProperties, rendererIndex=0}
   {rendererClass=&NiagaraMeshRendererProperties, rendererIndex=1}
   {rendererClass=&NiagaraMeshRendererProperties, rendererIndex=2}
  ]
 }
 Trail {
  emitterUpdateScript [
   &EmitterState(Life Cycle Mode=::NewEnumerator1, Loop Behavior=::NewEnumerator1, Loop Duration=1, UseLoopDelay=true)
   &SpawnBurst_Instantaneous(Use Spawn Probability=false, UseLoopCountLimit=false)
  ]
  particleSpawnScript [
   &InitializeParticle(
    Color=<rgba 1 0.1 0.115 1>
    Lifetime=1.3
    Mesh Scale Mode=::NewEnumerator3
    Mesh Uniform Scale=1.2363
    UsePositionOffset=false
    Write Material Random=true
   )
   &InitialMeshOrientation(
    Mesh Orientation Mode=::NewEnumerator2
    Orientation Vector=$User.AimVec
    Rotation=&MakeVector(X=&RandomRangeFloat(Minimum=-1, Maximum=1), Y=0, Z=0)
    Use Rotation Vector=true
   )
  ]
  particleUpdateScript [
   &ParticleState()
   &ScaleMeshSize(Scale Factor=&MakeVector(
     X=&FloatFromCurve(
      FloatCurve=DataInterface() {propertyValues {Curve {keys [{interpMode=RCIM_Cubic, leaveTangent=3.428, tangentMode=RCTM_Break, time=0, value=0}, <at 1 1>]}}}
      Scale Curve=1.8
     )
     Y=1
     Z=1
    ))
   &ScaleMeshSize(Scale Factor=<-1.5121 0.4927 0.4521>) {module=ScaleMeshSize002}
   &CameraOffset(Camera Offset Amount=100)
   /SetVariables(Particles.ColorDark:LinearColor=&MakeLinearColorFromVectorAndFloat(
     "Vector (RGB)"=&MakeVectorFromLinearColorRGB(Color=&ColorFromCurve(ColorCurve=DataInterface() {propertyValues {
         AlphaCurve {keys [<at 0 1 RCIM_Cubic>]}
         BlueCurve {keys [<at 0 0.005 RCIM_Cubic>,<at 0.3152 0>]}
         GreenCurve {keys [<at 0 0 RCIM_Cubic>,<at 0.3152 0>]}
         RedCurve {keys [<at 0 0.3 RCIM_Cubic>,<at 0.3152 0>]}
        }}))
     "Float (Alpha)"=&FloatFromCurve(
      FloatCurve=DataInterface() {propertyValues {Curve {keys [
          {interpMode=RCIM_Cubic, tangentMode=RCTM_User, time=0.1, value=0}
          {interpMode=RCIM_Cubic, tangentMode=RCTM_User, time=0.2, value=1}
          {interpMode=RCIM_Cubic, tangentMode=RCTM_User, time=1, value=0}
         ]}}}
      Scale Curve=1.5
     )
    )) {isSetParameters=true}
   &DynamicMaterialParameters(
    Index 0 Param 1=-0.3
    Index 0 Param 2=0.488
    Index 0 Param 3=1
    Index 0 Param 4=&FloatFromCurve(FloatCurve=DataInterface() {propertyValues {Curve {keys [<at 0 1>,<at 0.2 0>,<at 1 1>]}}})
    Module.Index0Param0Write=true
    Module.Index0Param1Write=true
    Module.Index0Param2Write=true
    Module.Index0Param3Write=true
   )
   /SetVariables(Particles.DynamicMaterialParameterDark:Vector4f=&MakeVector4(X=-0.2, Y=0.3, Z=1, W=0.1)) {isSetParameters=true}
  ]
  rendererClasses [&NiagaraMeshRendererProperties]
  renderers [
   {rendererClass=&NiagaraMeshRendererProperties, rendererIndex=0}
   {rendererClass=&NiagaraMeshRendererProperties, rendererIndex=1}
  ]
 }
}
summary {
 systemName=NS_VoidMask_WanderingBoss_Relocation_Muzzel
 emitters [
  {emitterName=Tendrils, rendererClasses [&NiagaraMeshRendererProperties]}
  {emitterName=Trail, rendererClasses [&NiagaraMeshRendererProperties]}
  {emitterName=LongFlare, rendererClasses [&NiagaraSpriteRendererProperties, &NiagaraLightRendererProperties]}
 ]
 userVariables [
  User.AimVec {type=Vector3f, description="", defaultValue {struct=Vector3f, value=<0 1 0>}}
  User.Global Scale {type=NiagaraFloat, defaultValue=1, description=""}
  User.PortalPosition {type=NiagaraPosition, description="", defaultValue {}}
 ]
}
