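# Niagara ribbon trail for the Wand Boss Void Mask teleport, stretching between User.OriginPos and the system's location.
# Two delayed burst emitters, one bright and one dark, each re-lerp ten ribbon points every update from a raised origin to the emitter position.
# Curves grow width over life and along the ribbon, and fade alpha over life and toward both ends.
# User.Color Variant blends each layer between two tints, and random U offsets feed the ribbon material.
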NiagaraSystem /DemonMaskVoid/Assets/VFX/NS_WandBoss_VoidMask_Teleport_WrapTrail
dependencies {
 usedDataInterfaces [&NiagaraDataInterfaceCurve, &NiagaraDataInterfaceColorCurve, &NiagaraDataInterfaceArrayColor]
 usedDynamicInputs [&SelectLinearColorFromArray]
 usedModules [
  &self:SystemSpawnScript.SystemScriptSource.SystemScriptGraph.NiagaraNodeAssignment_0.SetVariables_#1
  &SystemState
 ]
}
emitters {
 Trail {
  emitterUpdateScript [
   &EmitterState(Life Cycle Mode=::NewEnumerator1, Loop Behavior=::NewEnumerator1, Loop Delay=0.2, Loop Duration=1.2, UseLoopDelay=true)
   &SpawnBurst_Instantaneous(Spawn Count=10, Use Spawn Probability=false, UseLoopCountLimit=false)
  ]
  particleSpawnScript [
   /SetVariables(Particles.RibbonLinkOrder:NiagaraFloat=&ReturnNormalizedExecIndex()) {isSetParameters=true, module=SetVariables_#2}
   &InitializeParticle(
    Color Minimum=<rgba 0.7 0.7 0.7 1>
    Color Mode=::NewEnumerator2
    Lifetime Max=1.2
    Lifetime Min=1.21
    Lifetime Mode=::NewEnumerator1
    Position=&LerpPosition(A=$User.OriginPos, B=$Engine.Emitter.SimulationPosition, Alpha=&ReturnNormalizedExecIndex())
    Position Mode=::NewEnumerator1
    Ribbon Width=350
    Ribbon Width Mode=::NewEnumerator1
    Sprite Size Mode=::NewEnumerator3
    Uniform Sprite Size=5
    UsePositionOffset=false
    Write Material Random=true
   )
   &FN_RibbonMaterial(Color Pan U Offset:NiagaraFloat=&RandomRangeFloat(Minimum=0, Maximum=1))
  ]
  particleUpdateScript [
   &ParticleState()
   /SetVariables(Particles.Position=&LerpPosition(
     A=&AddVectorToPosition(Position=$User.OriginPos, Vector=<0 0 240>)
     B=$Engine.Emitter.SimulationPosition
     Alpha=&ReturnNormalizedExecIndex()
    )) {isSetParameters=true, module=SetVariables_#3}
   &ScaleRibbonWidth(Ribbon Width Scale=&FloatFromCurve(
     FloatCurve=DataInterface() {propertyValues {Curve {keys [
         {arriveTangent=4.1785, interpMode=RCIM_Cubic, leaveTangent=4.1785, tangentMode=RCTM_User, time=0, value=0}
         {arriveTangent=1.1179, interpMode=RCIM_Cubic, leaveTangent=1.1179, tangentMode=RCTM_User, time=0.2042, value=0.6228}
         {arriveTangent=0.0224, interpMode=RCIM_Cubic, leaveTangent=0.0224, tangentMode=RCTM_User, time=1, value=1}
        ]}}}
     Scale Curve=&FloatFromCurve(
      FloatCurve=DataInterface() {propertyValues {Curve {keys [<at 0 1 RCIM_Cubic>, {arriveTangent=2.9588, interpMode=RCIM_Cubic, leaveTangent=2.9588, tangentMode=RCTM_User, time=1, value=2}]}}}
      CurveIndex=&ReturnNormalizedExecIndex()
     )
    ))
   &ScaleColor(
    Scale Alpha=&FloatFromCurve(FloatCurve=DataInterface() {propertyValues {Curve {keys [<at 0.0079 0>,<at 0.2176 1>,<at 0.9371 0>,<at 1 0>]}}})
    Scale RGB=&Lerp_Vector(A=<30 0.78 1.267>, B=<1 15 1>, Alpha=&MakeFloatFromInt(Int=$User.Color Variant))
    ScaleA=true
    ScaleRGB=true
   )
   &ScaleColor(
    Scale Alpha=&FloatFromCurve(
     FloatCurve=DataInterface() {propertyValues {Curve {keys [
         {interpMode=RCIM_Cubic, tangentMode=RCTM_User, time=0, value=0}
         {interpMode=RCIM_Cubic, tangentMode=RCTM_User, time=0.7, value=0.95}
         {arriveTangent=-3.1667, interpMode=RCIM_Cubic, leaveTangent=-3.1667, time=0.9, value=0.75}
         {arriveTangent=-6.1359, tangentMode=RCTM_User, time=1, value=0}
        ]}}}
     CurveIndex=$Particles.RibbonLinkOrder
    )
    ScaleA=true
    ScaleRGB=true
   ) {module=ScaleColor001}
  ]
  rendererClasses [&NiagaraRibbonRendererProperties]
  renderers [{rendererClass=&NiagaraRibbonRendererProperties}]
 }
 Trail002 {
  emitterUpdateScript [
   &EmitterState(Life Cycle Mode=::NewEnumerator1, Loop Behavior=::NewEnumerator1, Loop Delay=0.2, Loop Duration=1.5167, UseLoopDelay=true)
   &SpawnBurst_Instantaneous(Spawn Count=10, Use Spawn Probability=false, UseLoopCountLimit=false)
  ]
  particleSpawnScript [
   /SetVariables(Particles.RibbonLinkOrder:NiagaraFloat=&ReturnNormalizedExecIndex()) {isSetParameters=true, module=SetVariables_#2}
   &InitializeParticle(
    Color Minimum=<rgba 0.7 0.7 0.7 1>
    Color Mode=::NewEnumerator2
    Lifetime Max=1.5
    Lifetime Min=1.5
    Lifetime Mode=::NewEnumerator1
    Position=&LerpPosition(A=$User.OriginPos, B=$Engine.Emitter.SimulationPosition, Alpha=&ReturnNormalizedExecIndex())
    Position Mode=::NewEnumerator1
    Ribbon Width=&RandomRangeFloat(Minimum=250, Maximum=250)
    Ribbon Width Mode=::NewEnumerator1
    Sprite Size Mode=::NewEnumerator3
    Uniform Sprite Size=5
    UsePositionOffset=false
    Write Material Random=true
   )
   &FN_RibbonMaterial(Color Pan U Offset:NiagaraFloat=&RandomRangeFloat(Minimum=0, Maximum=1))
  ]
  particleUpdateScript [
   &ParticleState()
   /SetVariables(Particles.Position=&LerpPosition(
     A=&AddVectorToPosition(Position=$User.OriginPos, Vector=<0 0 240>)
     B=$Engine.Emitter.SimulationPosition
     Alpha=&ReturnNormalizedExecIndex()
    )) {isSetParameters=true, module=SetVariables_#3}
   &ScaleRibbonWidth(Ribbon Width Scale=&FloatFromCurve(
     FloatCurve=DataInterface() {propertyValues {Curve {keys [
         {arriveTangent=4.1785, interpMode=RCIM_Cubic, leaveTangent=4.1785, tangentMode=RCTM_User, time=0, value=0}
         {arriveTangent=1.1179, interpMode=RCIM_Cubic, leaveTangent=1.1179, tangentMode=RCTM_User, time=0.2042, value=0.6228}
         {arriveTangent=0.0224, interpMode=RCIM_Cubic, leaveTangent=0.0224, tangentMode=RCTM_User, time=1, value=1}
        ]}}}
     Scale Curve=&FloatFromCurve(
      FloatCurve=DataInterface() {propertyValues {Curve {keys [<at 0 1 RCIM_Cubic>,<at 1 1.3 RCIM_Cubic>]}}}
      CurveIndex=&ReturnNormalizedExecIndex()
     )
    ))
   &ScaleColor(
    Scale Alpha=&FloatFromCurve(FloatCurve=DataInterface() {propertyValues {Curve {keys [<at 0.0079 0>,<at 0.1502 1>,<at 0.9371 0>,<at 1 0>]}}})
    Scale RGB=&Lerp_Vector(A=<0.0208 0.0069 0.0071>, B=<0.001 0.02 0.0001>, Alpha=&MakeFloatFromInt(Int=$User.Color Variant))
    ScaleA=true
    ScaleRGB=true
   )
   &ScaleColor(
    Scale Alpha=&FloatFromCurve(
     FloatCurve=DataInterface() {propertyValues {Curve {keys [
         {arriveTangent=6.7993, interpMode=RCIM_Cubic, leaveTangent=6.7993, tangentMode=RCTM_User, time=0, value=0}
         {arriveTangent=2, interpMode=RCIM_Cubic, leaveTangent=2, time=0.1231, value=0.7144}
         {interpMode=RCIM_Cubic, tangentMode=RCTM_User, time=0.5, value=1}
         {arriveTangent=-2, interpMode=RCIM_Cubic, leaveTangent=-2, time=0.9124, value=0.75}
         {arriveTangent=-12.7858, tangentMode=RCTM_User, time=1, value=0}
        ]}}}
     CurveIndex=$Particles.RibbonLinkOrder
     Scale Curve=1.1
    )
    ScaleA=true
    ScaleRGB=true
   ) {module=ScaleColor001}
  ]
  rendererClasses [&NiagaraRibbonRendererProperties]
  renderers [{rendererClass=&NiagaraRibbonRendererProperties}]
 }
}
summary {
 systemName=NS_WandBoss_VoidMask_Teleport_WrapTrail
 emitters [
  {emitterName=Trail002, rendererClasses [&NiagaraRibbonRendererProperties]}
  {emitterName=Trail, rendererClasses [&NiagaraRibbonRendererProperties]}
 ]
 userVariables [
  User.Color Variant {type=NiagaraInt32}
  User.OriginPos {type=NiagaraPosition, description="", defaultValue {}}
 ]
}
